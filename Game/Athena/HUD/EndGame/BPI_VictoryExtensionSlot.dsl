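; Blueprint interface for victory extension slot widgets on the Athena end-game HUD.
; It declares OverrideVisualData, which takes a BP Victory Extension Slot Data object to replace a slot's visuals, and PlayIntroAnimation, which triggers the slot's intro animation.
; The asset holds only these function signatures, leaving behavior to implementing classes.
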(fn OverrideVisualData
  (params (ExtensionSlotData :type BP%20Victory%20Extension%20Slot%20Data%20Type%20Object%20Reference)))

(fn PlayIntroAnimation)
