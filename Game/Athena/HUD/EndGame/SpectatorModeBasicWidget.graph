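# Basic spectator-mode HUD widget from the Athena end-game UI, laying out a bottom bar for cycling between spectator targets.
# A half-transparent black image backs a centered row of previous and next UIKit quiet buttons around an uppercase target-name placeholder, alongside an empty extra-buttons box.
# The event graph holds only unconnected PreConstruct, Construct and Tick events, so the widget itself has no switching or text-update logic.
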WidgetBlueprint /Game/Athena/HUD/EndGame/SpectatorModeBasicWidget
graphs [EventGraph {nodes [
   Event PreConstruct()
   Event Construct()
   Event Tick()
  ]}]
variables []
widgets {
 info {inheritedWidgetCount=0, parentClass=&UserWidget, rootWidgetClass=&Overlay, widgetCount=9}
 widgets [
  Overlay_103 {widgetClassPath=&Overlay, widget=Overlay_103}
  Overlay_0 {
   widgetClassPath=&Overlay
   parent=Overlay_103
   widget=Overlay_0
   slot {class=OverlaySlot, properties {horizontalAlignment=HAlign_Fill, verticalAlignment=VAlign_Bottom}}
  }
  Image_72 {
   widgetClassPath=&Image
   bIsVariable=true
   parent=Overlay_0
   widget=Image_72
   properties {
    brush {
     bIsDynamicallyLoaded=false
     drawAs=Image
     imageType=FullColor
     mirroring=NoMirror
     resourceName=None
     resourceObject=null
     tiling=NoTile
     imageSize=<64 64>
     margin=<ltrb 0 0 0 0>
     outlineSettings {
      bUseBrushTransparency=false
      roundingType=HalfHeightRadius
      width=0
      color {colorUseRule=UseColor_Specified, specifiedColor=<rgba 0 0 0 0>}
      cornerRadii=<0 0 0 0>
     }
     tintColor {colorUseRule=UseColor_Specified, specifiedColor=<rgba 1 1 1 1>}
     uVRegion {bIsValid=false, max=<0 0>, min=<0 0>}
    }
    colorAndOpacity=<rgba 0 0 0 0.5>
   }
   slot {class=OverlaySlot, properties {horizontalAlignment=HAlign_Fill, verticalAlignment=VAlign_Fill}}
  }
  Overlay_2 {
   widgetClassPath=&Overlay
   parent=Overlay_0
   widget=Overlay_2
   slot {class=OverlaySlot, properties {horizontalAlignment=HAlign_Fill, verticalAlignment=VAlign_Bottom, padding=<ltrb 0 25 0 25>}}
  }
  ExtraButtonsBar {
   widgetClassPath=&HorizontalBox
   parent=Overlay_2
   widget=ExtraButtonsBar
   slot {class=OverlaySlot, properties {padding=<ltrb 25 0 25 0>}}
  }
  ExtraButtonsBar_1 {
   widgetClassPath=&HorizontalBox
   parent=Overlay_2
   widget=ExtraButtonsBar_1
   slot {class=OverlaySlot, properties {horizontalAlignment=HAlign_Center, verticalAlignment=VAlign_Center}}
  }
  Button_PreviousTarget {
   widgetClassPath=/FortUILibrary/UIKit/Buttons/WBP_UIKit_Button_Quiet
   bIsVariable=true
   parent=ExtraButtonsBar_1
   widget=Button_PreviousTarget
   properties {enableOnWidgetStyled=true, hasText=true, minHeight=48, text=""}
   slot {class=HorizontalBoxSlot}
  }
  Text_SpectatorTarget {
   widgetClassPath=&TextBlock
   bIsVariable=true
   parent=ExtraButtonsBar_1
   widget=Text_SpectatorTarget
   properties {
    text="{TargetName}"
    textTransformPolicy=ToUpper
    font {
     bForceMonospaced=false
     bMaterialIsStencil=false
     fontMaterial=null
     fontObject=&Burbank
     letterSpacing=0
     monospacedWidth=1
     size=24
     skewAmount=0
     typefaceFontName=BurbankRegular-Black
     outlineSettings {
      bApplyOutlineToDropShadows=false
      bMiteredCorners=false
      bSeparateFillAlpha=false
      outlineMaterial=/Game/UI/Foundation/Materials/M-FontShadow
      outlineSize=6
      outlineColor=<rgba 0.0052 0.0052 0.0056 0.5>
     }
    }
    renderTransformPivot=<0.5 0>
    shadowColorAndOpacity=<rgba 0.0052 0.0052 0.0056 0.4>
    shadowOffset=<8 8>
   }
   slot {class=HorizontalBoxSlot, properties {horizontalAlignment=HAlign_Center, verticalAlignment=VAlign_Center, padding=<ltrb 50 0 50 0>}}
  }
  Button_NextTarget {
   widgetClassPath=&WBP_UIKit_Button_Quiet
   bIsVariable=true
   parent=ExtraButtonsBar_1
   widget=Button_NextTarget
   properties {enableOnWidgetStyled=true, hasText=true, minHeight=48, text=""}
   slot {class=HorizontalBoxSlot}
  }
 ]
}
