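# Emote menu touch button widget for the Bacchus HUD, derived from the native EmoteMenuButton.
# It layers an emote icon, a radial cooldown button, a collapsed background and an invisible internal button.
# On each HUD state update it calls the parent, then looks up the owning player's external emote controller component.
# If that component is valid, a bang marker image shows only while the external emote list is non-empty.
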WidgetBlueprint /Game/Athena/HUD/Bacchus/Button_EmoteMenu
graphs [EventGraph {nodes [
   n1 = Event OnHUDStateUpdate!() -> (OutputDelegate, then, NewState:"Fort HUD State Structure (by ref)") {is=title K2Node_Event}
   > Parent: OnHUDStateUpdate!(execute, NewState:"Fort HUD State Structure (by ref)"=@n1.NewState) -> (then) {is=title K2Node_CallParentFunction}
   n2 = GetComponentByClass!(
    self:Actor Object Reference=GetOwningPlayer()
    ComponentClass:Actor Component Class Reference=/Script/FortniteGame.FortControllerComponent_ExternalEmote
   ) -> (ReturnValue:Fort Controller Component External Emote Object Reference)
   > Is Valid(InputObject=@n2)
   n3 = GetExternalEmotes!(self:Fort Controller Component External Emote Object Reference=@n2) -> (ReturnValue:Array of External Emote Category Structures)
   > SetVisibility(
    self=Get ExternalEmoteBang() -> (ExternalEmoteBang:Image Object Reference)
    InVisibility=Select(
     Option 0=Collapsed
     Option 1=HitTestInvisible
     Index=Is Array Non Empty!(InputArray:Array of External Emote Category Structures=@n3) -> (Non-Empty:bool=~) {is=title K2Node_MacroInstance}
    )
   )
  ]}]
variables []
widgets {
 info {inheritedWidgetCount=8, parentClass=/Script/FortniteUI.EmoteMenuButton, rootWidgetClass=&Overlay, widgetCount=15}
 widgets [
  Overlay_0 {widgetClassPath=&Overlay, widget=Overlay_0}
  SizeBox_ButtonSize {
   widgetClassPath=&SizeBox
   bInherited=true
   bIsVariable=true
   parent=Overlay_0
   widget=SizeBox_ButtonSize
   properties {bOverride_HeightOverride=true, bOverride_WidthOverride=true, heightOverride=200, widthOverride=200}
   slot {class=OverlaySlot, properties {horizontalAlignment=HAlign_Center, verticalAlignment=VAlign_Center, padding=<ltrb 4 2 4 2>}}
  }
  InvalidationBox_0 {
   widgetClassPath=/Script/UMG.InvalidationBox
   parent=SizeBox_ButtonSize
   widget=InvalidationBox_0
   slot {class=SizeBoxSlot, properties {horizontalAlignment=HAlign_Center, verticalAlignment=VAlign_Center}}
  }
  SizeBox_ImageSize {
   widgetClassPath=&SizeBox
   bInherited=true
   bIsVariable=true
   parent=InvalidationBox_0
   widget=SizeBox_ImageSize
   slot {class=PanelSlot}
  }
  ScaleBox_4 {widgetClassPath=&ScaleBox, parent=SizeBox_ImageSize, widget=ScaleBox_4, slot {class=SizeBoxSlot}}
  SizeBox_1 {
   widgetClassPath=&SizeBox
   parent=ScaleBox_4
   widget=SizeBox_1
   properties {bOverride_HeightOverride=true, bOverride_WidthOverride=true, heightOverride=128, widthOverride=128}
   slot {class=ScaleBoxSlot}
  }
  Overlay_1 {widgetClassPath=&Overlay, parent=SizeBox_1, widget=Overlay_1, slot {class=SizeBoxSlot}}
  Image_Background {
   widgetClassPath=&Image
   bInherited=true
   bIsVariable=true
   parent=Overlay_1
   widget=Image_Background
   properties {
    visibility=Collapsed
    brush {
     bIsDynamicallyLoaded=false
     drawAs=Image
     imageType=FullColor
     mirroring=NoMirror
     resourceName=None
     resourceObject=/Game/Athena/HUD/Bacchus/Material/MI_UI_ButtonBG_Dark
     tiling=NoTile
     imageSize=<128 128>
     margin=<ltrb 0 0 0 0>
     outlineSettings {
      bUseBrushTransparency=false
      roundingType=HalfHeightRadius
      width=0
      color {colorUseRule=UseColor_Specified, specifiedColor=<rgba 0 0 0 0>}
      cornerRadii=<0 0 0 0>
     }
     tintColor {colorUseRule=UseColor_Specified, specifiedColor=<rgba 1 1 1 1>}
     uVRegion {bIsValid=false, max=<0 0>, min=<0 0>}
    }
   }
   slot {class=OverlaySlot, properties {horizontalAlignment=HAlign_Fill, verticalAlignment=VAlign_Fill}}
  }
  SizeBox_0 {
   widgetClassPath=&SizeBox
   parent=Overlay_1
   widget=SizeBox_0
   properties {
    bOverride_MaxDesiredHeight=true
    bOverride_MaxDesiredWidth=true
    bOverride_MinDesiredHeight=true
    bOverride_MinDesiredWidth=true
    maxAspectRatio=2
    maxDesiredHeight=128
    maxDesiredWidth=128
    minDesiredHeight=60
    minDesiredWidth=60
   }
   slot {class=OverlaySlot, properties {horizontalAlignment=HAlign_Center, verticalAlignment=VAlign_Center}}
  }
  Image_Icon {
   widgetClassPath=&Image
   bInherited=true
   bIsVariable=true
   parent=SizeBox_0
   widget=Image_Icon
   properties {
    visibility=HitTestInvisible
    brush {
     bIsDynamicallyLoaded=false
     drawAs=Image
     imageType=FullColor
     mirroring=NoMirror
     resourceName=None
     resourceObject=/Game/Athena/HUD/Atlas/Bacchus/T_UI_Emote_Sprite
     tiling=NoTile
     imageSize=<96 96>
     margin=<ltrb 0 0 0 0>
     outlineSettings {
      bUseBrushTransparency=false
      roundingType=HalfHeightRadius
      width=0
      color {colorUseRule=UseColor_Specified, specifiedColor=<rgba 0 0 0 0>}
      cornerRadii=<0 0 0 0>
     }
     tintColor {colorUseRule=UseColor_Specified, specifiedColor=<rgba 1 1 1 1>}
     uVRegion {bIsValid=false, max=<0 0>, min=<0 0>}
    }
    colorAndOpacity=<rgba 1 1 1 0.8>
   }
   slot {class=SizeBoxSlot}
  }
  RadialButton_ButtonCooldown {
   widgetClassPath=&/RadialBacchusButton
   bInherited=true
   bIsVariable=true
   parent=Overlay_1
   widget=RadialButton_ButtonCooldown
   properties {elementTag {tagName=HUD.Bacchus.TouchButton}}
   slot {class=OverlaySlot}
  }
  ExternalEmoteBang {
   widgetClassPath=&Image
   bIsVariable=true
   parent=Overlay_1
   widget=ExternalEmoteBang
   properties {
    visibility=Collapsed
    brush {
     bIsDynamicallyLoaded=false
     drawAs=Image
     imageType=NoImage
     mirroring=NoMirror
     resourceName=None
     resourceObject=/Game/Athena/UI/Challenges/Art/MapIcons/M_UI_MarkerState-Bang
     tiling=NoTile
     imageSize=<50 50>
     margin=<ltrb 0 0 0 0>
     outlineSettings {
      bUseBrushTransparency=false
      roundingType=HalfHeightRadius
      width=0
      color {colorUseRule=UseColor_Specified, specifiedColor=<rgba 0 0 0 0>}
      cornerRadii=<0 0 0 0>
     }
     tintColor {colorUseRule=UseColor_Specified, specifiedColor=<rgba 1 1 1 1>}
     uVRegion {bIsValid=false, max=<0 0>, min=<0 0>}
    }
   }
   slot {class=OverlaySlot}
  }
  Button_InternalButton {
   widgetClassPath=&Button
   bInherited=true
   parent=Overlay_0
   widget=Button_InternalButton
   properties {
    visibility=HitTestInvisible
    widgetStyle {
     clickedSlateSound {resourceObject=null}
     disabled {
      bIsDynamicallyLoaded=false
      drawAs=NoDrawType
      imageType=NoImage
      mirroring=NoMirror
      resourceName=None
      resourceObject=null
      tiling=NoTile
      imageSize=<0 0>
      margin=<ltrb 0 0 0 0>
      outlineSettings {
       bUseBrushTransparency=false
       roundingType=HalfHeightRadius
       width=0
       color {colorUseRule=UseColor_Specified, specifiedColor=<rgba 0 0 0 0>}
       cornerRadii=<0 0 0 0>
      }
      tintColor {colorUseRule=UseColor_Specified, specifiedColor=<rgba 1 1 1 0>}
      uVRegion {bIsValid=false, max=<0 0>, min=<0 0>}
     }
     disabledForeground {colorUseRule=UseColor_Specified, specifiedColor=<rgba 0.5271 0.5271 0.5271 1>}
     hovered {
      bIsDynamicallyLoaded=false
      drawAs=RoundedBox
      imageType=NoImage
      mirroring=NoMirror
      resourceName=None
      resourceObject=null
      tiling=NoTile
      imageSize=<32 32>
      margin=<ltrb 0 0 0 0>
      outlineSettings {
       bUseBrushTransparency=true
       roundingType=FixedRadius
       width=1
       color {colorUseRule=UseColor_Specified, specifiedColor=<rgba 0.7243 0.7243 0.7243 1>}
       cornerRadii=<4 4 4 4>
      }
      tintColor {colorUseRule=UseColor_Specified, specifiedColor=<rgba 1 1 1 0>}
      uVRegion {bIsValid=false, max=<0 0>, min=<0 0>}
     }
     hoveredForeground {colorUseRule=UseColor_Specified, specifiedColor=<rgba 0 1 1 1>}
     hoveredSlateSound {resourceObject=null}
     normal {
      bIsDynamicallyLoaded=false
      drawAs=RoundedBox
      imageType=NoImage
      mirroring=NoMirror
      resourceName=None
      resourceObject=null
      tiling=NoTile
      imageSize=<32 32>
      margin=<ltrb 0 0 0 0>
      outlineSettings {
       bUseBrushTransparency=true
       roundingType=FixedRadius
       width=1
       color {colorUseRule=UseColor_Specified, specifiedColor=<rgba 0.6951 0.6951 0.6951 1>}
       cornerRadii=<4 4 4 4>
      }
      tintColor {colorUseRule=UseColor_Specified, specifiedColor=<rgba 1 1 1 0>}
      uVRegion {bIsValid=false, max=<0 0>, min=<0 0>}
     }
     normalForeground {colorUseRule=UseColor_Specified, specifiedColor=<rgba 0 1 1 1>}
     normalPadding=<ltrb 12 1.5 12 1.5>
     pressed {
      bIsDynamicallyLoaded=false
      drawAs=RoundedBox
      imageType=NoImage
      mirroring=NoMirror
      resourceName=None
      resourceObject=null
      tiling=NoTile
      imageSize=<32 32>
      margin=<ltrb 0 0 0 0>
      outlineSettings {
       bUseBrushTransparency=true
       roundingType=FixedRadius
       width=1
       color {colorUseRule=UseColor_Specified, specifiedColor=<rgba 0.7243 0.7243 0.7243 1>}
       cornerRadii=<4 4 4 4>
      }
      tintColor {colorUseRule=UseColor_Specified, specifiedColor=<rgba 1 1 1 0>}
      uVRegion {bIsValid=false, max=<0 0>, min=<0 0>}
     }
     pressedForeground {colorUseRule=UseColor_Specified, specifiedColor=<rgba 0 1 1 1>}
     pressedPadding=<ltrb 12 1.5 12 1.5>
     pressedSlateSound {resourceObject=null}
    }
   }
   slot {class=OverlaySlot, properties {horizontalAlignment=HAlign_Fill, verticalAlignment=VAlign_Fill}}
  }
  ScaleBox_HUDWidgetScale {widgetClassPath=&ScaleBox, bInherited=true, widget=null}
  TutorialHighlight {widgetClassPath=&TutorialHighlight}
 ]
}
