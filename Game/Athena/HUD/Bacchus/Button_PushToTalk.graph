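# Mobile HUD button (BacchusButton subclass) for switching push-to-talk voice chat state in Athena.
# Touch start captures input, shows the pressed icon and either calls TogglePTTState or flips Enabled/MicDisabled, depending on the Fort.Rollback.MobilePushToTalkToggle console variable.
# Touch end only releases capture, and capture loss clears the pressed look.
# Construct binds AthenaHUDContext PTT delegates; Update Visibility swaps talk/mute sprites and collapses the button unless PTT is allowed on mobile Athena.
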WidgetBlueprint /Game/Athena/HUD/Bacchus/Button_PushToTalk
graphs [
 EventGraph {nodes [
   PTTClicked() {is=title K2Node_CustomEvent}
   > n1 = GetConsoleVariableBoolValue!(execute, VariableName:String=Fort.Rollback.MobilePushToTalkToggle) -> (then, ReturnValue:bool)
   > n2 = Branch(Condition=@n1)
   n3 = Get AthenaHUDContext!() -> (ReturnValue:Athena HUDContext Object Reference) {is=title K2Node_GetBlueprintContextEx}
   > TogglePTTState(self:Athena HUDContext Object Reference=@n3)
   SetPTTState!(
    execute=@n2.else
    self:Athena HUDContext Object Reference=@n3
    NewPushToTalkState:EPTTState Enum=Select!(
     Enabled:EPTTState Enum=MicDisabled
     MicDisabled:EPTTState Enum=Enabled
     Index:EPTTState Enum=GetPTTState!(self:Athena HUDContext Object Reference=@n3) -> (ReturnValue:EPTTState Enum=Enabled)
    ) -> (ReturnValue:EPTTState Enum)
   ) -> (then)
   n4 = On Can PTT() {is=title K2Node_CustomEvent}
   > Update Visibility!(execute, self, PTT State:EPTTState Enum=GetPTTState(self=Get AthenaHUDContext()) -> (ReturnValue=Enabled)) -> (then)
   Event Construct()
   > Assign On Can PTTChange!(execute:Exec, self:Athena HUDContext Object Reference=Get AthenaHUDContext(), Delegate:"Delegate (by ref)"=@n4) -> (then:Exec) {is=title K2Node_AssignDelegate}
   > Bind Event to On PTTState Change(self:Athena HUDContext Object Reference=Get AthenaHUDContext(), Delegate=Create Event()) {is=title K2Node_AddDelegate}
   > Update Visibility(PTT State=GetPTTState(self=Get AthenaHUDContext()) -> (ReturnValue=Enabled))
   Event OnMouseCaptureLost() {is=title K2Node_Event}
   > SetIconPressed!(execute, self:Bacchus Button Object Reference, bIsPressed:bool=false) -> (then)
   Break Fort HUD State!(FortHUDState:"Fort HUD State Structure (by ref)"=Event OnHUDStateUpdate!() -> (OutputDelegate, then, NewState:"Fort HUD State Structure (by ref)") {is=title K2Node_Event}.NewState) -> (
    bInBuildMode:bool=~
    bInCombatMode:bool=~
    bInEditMode:bool=~
    bInCreativeMode:bool=~
    bIsParachuteOpen:bool=~
    bIsFreeFalling:bool=~
    bInLockedBus:bool=~
    bInUnlockedBus:bool=~
    bOnTargeting:bool=~
    bOnUsingScopeTargeting:bool=~
    bOnCanTarget:bool=~
    bOnCanUseScopeTargeting:bool=~
    bOnCanUseSecondaryAbility:bool=~
    bCanOpenChute:bool=~
    bCrouching:bool=~
    bDBNO:bool=~
    bIsControllingRCPawn:bool=~
    bIsInVehicle:bool=~
    bIsDrivingVehicle:bool=~
    bCanSwapSeats:bool=~
    bIsCarryingDBNO:bool=~
    bIsCarryingHeldObject:bool=~
    bIsZiplining:bool=~
    bIsCreativeFlying:bool=~
    bImprovedDBNO:bool=~
   ) {is=title K2Node_BreakStruct}
  ]}
 NewMacro_0 {nodes [
   Inputs()
   > SetVisibility(InVisibility=Collapsed)
   Outputs()
  ]}
 OnTouchEnded {nodes [
   OnTouchEnded!() -> (then, MyGeometry:Geometry Structure, InTouchEvent:"Pointer Event Structure (by ref)") {is=title K2Node_FunctionEntry}
   > n1 = Branch(Condition=HasMouseCapture!(self:Widget Object Reference) -> (ReturnValue:bool))
   > Return Node!(
    execute
    ReturnValue:Event Reply Structure=ReleaseMouseCapture!(Reply:"Event Reply Structure (by ref)"=Handled()) -> (ReturnValue:Event Reply Structure)
   )
   Return Node(execute=@n1.else, ReturnValue=Unhandled())
  ]}
 OnTouchStarted {nodes [
   OnTouchStarted!() -> (then, MyGeometry:Geometry Structure, InTouchEvent:"Pointer Event Structure (by ref)") {is=title K2Node_FunctionEntry}
   > n1 = Branch(Condition=NOT Boolean(A=HasMouseCapture()))
   > PTTClicked() {is=title K2Node_CallFunction}
   > SetIconPressed(bIsPressed=true)
   > Return Node(ReturnValue=CaptureMouse!(Reply:"Event Reply Structure (by ref)"=Handled(), CapturingWidget:Widget Object Reference=Self-Reference()) -> (ReturnValue:Event Reply Structure))
   Return Node(execute=@n1.else, ReturnValue=Unhandled())
  ]}
 Update Visibility {nodes [
   n1 = Update Visibility!() -> (then, PTT State:EPTTState Enum=Enabled) {is=title K2Node_FunctionEntry}
   > SetButtonSprite!(
    execute
    self:Bacchus Button Object Reference
    NewSprite:Paper Sprite Object Reference=Select(
     Enabled:Paper Sprite Object Reference=/Game/Athena/HUD/Atlas/Bacchus/Talk_Normal_Sprite
     MicDisabled:Paper Sprite Object Reference=/Game/Athena/HUD/Atlas/Bacchus/Mute_Normal_Sprite
     Index=@n1
    ) -> (ReturnValue:Paper Sprite Object Reference)
   ) -> (then)
   n2 = Get AthenaHUDContext()
   n3 = GetSubGame!() -> (ReturnValue:ESubGame Enum=Campaign)
   > n4 = Branch(Condition=AND Boolean!(
     A=GetCanPTT!(self:Athena HUDContext Object Reference=@n2) -> (ReturnValue:bool)
     B=IsMobileGame()
     C:bool="Equal (Enum)"(A:ESubGame Enum=@n3, B:ESubGame Enum=Athena)
    ) -> (ReturnValue))
   > SetVisibility(InVisibility=SelfHitTestInvisible)
   SetVisibility(execute=@n4.else, InVisibility=Collapsed)
  ]}
]
variables []
widgets {
 info {inheritedWidgetCount=8, parentClass=/Script/FortniteUI.BacchusButton, rootWidgetClass=&Overlay, widgetCount=14}
 widgets [
  Overlay_0 {widgetClassPath=&Overlay, widget=Overlay_0}
  SizeBox_ButtonSize {
   widgetClassPath=&SizeBox
   bInherited=true
   bIsVariable=true
   parent=Overlay_0
   widget=SizeBox_ButtonSize
   properties {bOverride_HeightOverride=true, bOverride_WidthOverride=true, heightOverride=200, widthOverride=200}
   slot {class=OverlaySlot, properties {horizontalAlignment=HAlign_Center, verticalAlignment=VAlign_Center, padding=<ltrb 4 2 4 2>}}
  }
  InvalidationBox_0 {
   widgetClassPath=/Script/UMG.InvalidationBox
   parent=SizeBox_ButtonSize
   widget=InvalidationBox_0
   slot {class=SizeBoxSlot, properties {horizontalAlignment=HAlign_Center, verticalAlignment=VAlign_Center}}
  }
  SizeBox_ImageSize {
   widgetClassPath=&SizeBox
   bInherited=true
   bIsVariable=true
   parent=InvalidationBox_0
   widget=SizeBox_ImageSize
   slot {class=PanelSlot}
  }
  ScaleBox_4 {widgetClassPath=&ScaleBox, parent=SizeBox_ImageSize, widget=ScaleBox_4, slot {class=SizeBoxSlot}}
  SizeBox_1 {
   widgetClassPath=&SizeBox
   parent=ScaleBox_4
   widget=SizeBox_1
   properties {bOverride_HeightOverride=true, bOverride_WidthOverride=true, heightOverride=128, widthOverride=128}
   slot {class=ScaleBoxSlot}
  }
  Overlay_1 {widgetClassPath=&Overlay, parent=SizeBox_1, widget=Overlay_1, slot {class=SizeBoxSlot}}
  Image_Background {
   widgetClassPath=&Image
   bInherited=true
   bIsVariable=true
   parent=Overlay_1
   widget=Image_Background
   properties {brush {
     bIsDynamicallyLoaded=false
     drawAs=Image
     imageType=FullColor
     mirroring=NoMirror
     resourceName=None
     resourceObject=/Game/Athena/HUD/Bacchus/Material/MI_UI_ButtonBG_Dark
     tiling=NoTile
     imageSize=<128 128>
     margin=<ltrb 0 0 0 0>
     outlineSettings {
      bUseBrushTransparency=false
      roundingType=HalfHeightRadius
      width=0
      color {colorUseRule=UseColor_Specified, specifiedColor=<rgba 0 0 0 0>}
      cornerRadii=<0 0 0 0>
     }
     tintColor {colorUseRule=UseColor_Specified, specifiedColor=<rgba 1 1 1 1>}
     uVRegion {bIsValid=false, max=<0 0>, min=<0 0>}
    }}
   slot {class=OverlaySlot, properties {horizontalAlignment=HAlign_Fill, verticalAlignment=VAlign_Fill}}
  }
  SizeBox_0 {
   widgetClassPath=&SizeBox
   parent=Overlay_1
   widget=SizeBox_0
   properties {
    bOverride_MaxDesiredHeight=true
    bOverride_MaxDesiredWidth=true
    bOverride_MinDesiredHeight=true
    bOverride_MinDesiredWidth=true
    maxAspectRatio=2
    maxDesiredHeight=128
    maxDesiredWidth=128
    minDesiredHeight=60
    minDesiredWidth=60
   }
   slot {class=OverlaySlot, properties {horizontalAlignment=HAlign_Center, verticalAlignment=VAlign_Center}}
  }
  Image_Icon {
   widgetClassPath=&Image
   bInherited=true
   bIsVariable=true
   parent=SizeBox_0
   widget=Image_Icon
   properties {
    visibility=HitTestInvisible
    brush {
     bIsDynamicallyLoaded=false
     drawAs=Image
     imageType=NoImage
     mirroring=NoMirror
     resourceName=None
     resourceObject=/Game/Athena/HUD/Atlas/Bacchus/Confirm_Normal_L_Sprite
     tiling=NoTile
     imageSize=<72 72>
     margin=<ltrb 0 0 0 0>
     outlineSettings {
      bUseBrushTransparency=false
      roundingType=HalfHeightRadius
      width=0
      color {colorUseRule=UseColor_Specified, specifiedColor=<rgba 0 0 0 0>}
      cornerRadii=<0 0 0 0>
     }
     tintColor {colorUseRule=UseColor_Specified, specifiedColor=<rgba 1 1 1 1>}
     uVRegion {bIsValid=false, max=<0 0>, min=<0 0>}
    }
    colorAndOpacity=<rgba 1 1 1 0.8>
   }
   slot {class=SizeBoxSlot}
  }
  RadialButton_ButtonCooldown {
   widgetClassPath=&/RadialBacchusButton
   bInherited=true
   bIsVariable=true
   parent=Overlay_1
   widget=RadialButton_ButtonCooldown
   properties {elementTag {tagName=HUD.Bacchus.TouchButton}}
   slot {class=OverlaySlot}
  }
  Button_InternalButton {
   widgetClassPath=&Button
   bInherited=true
   parent=Overlay_0
   widget=Button_InternalButton
   properties {
    visibility=HitTestInvisible
    widgetStyle {
     clickedSlateSound {resourceObject=null}
     disabled {
      bIsDynamicallyLoaded=false
      drawAs=NoDrawType
      imageType=NoImage
      mirroring=NoMirror
      resourceName=None
      resourceObject=null
      tiling=NoTile
      imageSize=<0 0>
      margin=<ltrb 0 0 0 0>
      outlineSettings {
       bUseBrushTransparency=false
       roundingType=HalfHeightRadius
       width=0
       color {colorUseRule=UseColor_Specified, specifiedColor=<rgba 0 0 0 0>}
       cornerRadii=<0 0 0 0>
      }
      tintColor {colorUseRule=UseColor_Specified, specifiedColor=<rgba 1 1 1 0>}
      uVRegion {bIsValid=false, max=<0 0>, min=<0 0>}
     }
     disabledForeground {colorUseRule=UseColor_Specified, specifiedColor=<rgba 0.5271 0.5271 0.5271 1>}
     hovered {
      bIsDynamicallyLoaded=false
      drawAs=RoundedBox
      imageType=NoImage
      mirroring=NoMirror
      resourceName=None
      resourceObject=null
      tiling=NoTile
      imageSize=<32 32>
      margin=<ltrb 0 0 0 0>
      outlineSettings {
       bUseBrushTransparency=true
       roundingType=FixedRadius
       width=1
       color {colorUseRule=UseColor_Specified, specifiedColor=<rgba 0.7243 0.7243 0.7243 1>}
       cornerRadii=<4 4 4 4>
      }
      tintColor {colorUseRule=UseColor_Specified, specifiedColor=<rgba 1 1 1 0>}
      uVRegion {bIsValid=false, max=<0 0>, min=<0 0>}
     }
     hoveredForeground {colorUseRule=UseColor_Specified, specifiedColor=<rgba 0 1 1 1>}
     hoveredSlateSound {resourceObject=null}
     normal {
      bIsDynamicallyLoaded=false
      drawAs=RoundedBox
      imageType=NoImage
      mirroring=NoMirror
      resourceName=None
      resourceObject=null
      tiling=NoTile
      imageSize=<32 32>
      margin=<ltrb 0 0 0 0>
      outlineSettings {
       bUseBrushTransparency=true
       roundingType=FixedRadius
       width=1
       color {colorUseRule=UseColor_Specified, specifiedColor=<rgba 0.6951 0.6951 0.6951 1>}
       cornerRadii=<4 4 4 4>
      }
      tintColor {colorUseRule=UseColor_Specified, specifiedColor=<rgba 1 1 1 0>}
      uVRegion {bIsValid=false, max=<0 0>, min=<0 0>}
     }
     normalForeground {colorUseRule=UseColor_Specified, specifiedColor=<rgba 0 1 1 1>}
     normalPadding=<ltrb 12 1.5 12 1.5>
     pressed {
      bIsDynamicallyLoaded=false
      drawAs=RoundedBox
      imageType=NoImage
      mirroring=NoMirror
      resourceName=None
      resourceObject=null
      tiling=NoTile
      imageSize=<32 32>
      margin=<ltrb 0 0 0 0>
      outlineSettings {
       bUseBrushTransparency=true
       roundingType=FixedRadius
       width=1
       color {colorUseRule=UseColor_Specified, specifiedColor=<rgba 0.7243 0.7243 0.7243 1>}
       cornerRadii=<4 4 4 4>
      }
      tintColor {colorUseRule=UseColor_Specified, specifiedColor=<rgba 1 1 1 0>}
      uVRegion {bIsValid=false, max=<0 0>, min=<0 0>}
     }
     pressedForeground {colorUseRule=UseColor_Specified, specifiedColor=<rgba 0 1 1 1>}
     pressedPadding=<ltrb 12 1.5 12 1.5>
     pressedSlateSound {resourceObject=null}
    }
   }
   slot {class=OverlaySlot, properties {horizontalAlignment=HAlign_Fill, verticalAlignment=VAlign_Fill}}
  }
  ScaleBox_HUDWidgetScale {widgetClassPath=&ScaleBox, bInherited=true, widget=null}
  TutorialHighlight {widgetClassPath=&TutorialHighlight}
 ]
}
