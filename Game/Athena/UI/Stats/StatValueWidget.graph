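# Widget Blueprint on AthenaProfileStatWidgetBase that shows one Athena profile stat as an animated, sheared Burbank number inside the inherited ProfileStatBox.
# On OnStatChanged, it gets the stat box's base stat view and the UI style for its Stat gameplay tag.
# The style sets the numeric type and starting value (MinimalNotableValue) of the CommonNumericTextBlock.
# The number then briefly interpolates from that value to the stat read from the view.
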WidgetBlueprint /Game/Athena/UI/Stats/StatValueWidget
graphs [EventGraph {nodes [
   Event OnStatChanged() {is=title K2Node_Event}
   n1 = Get ProfileStatBox() -> (ProfileStatBox:Athena Profile Stat Box Object Reference)
   > n2 = GetBaseStatView!(execute, self:Athena Profile Stat Box Object Reference=@n1) -> (then, ReturnValue:Athena Base Stat View Object Reference)
   n3 = Get Stat() -> (Stat:Gameplay Tag Structure)
   n4 = Break Fort UIStat Style!(FortUIStatStyle:"Fort UIStat Style Structure (by ref)"=GetStatStyle!(self:Athena Profile Stat Box Object Reference=@n1, InStat:"Gameplay Tag Structure (by ref)"=@n3) -> (ReturnValue:"Fort UIStat Style Structure (by ref)")) -> (Name:Text, HighestText:Text, Icon:Slate Brush Structure, NumericType:ECommonNumericType Enum, MinimalNotableValue:float=~) {is=title K2Node_BreakStruct}
   n5 = Get Value() -> (Value:Common Numeric Text Block Object Reference)
   > SetNumericType!(execute, self:Common Numeric Text Block Object Reference=@n5, InNumericType:ECommonNumericType Enum=@n4.NumericType) -> (then)
   > SetCurrentValue!(execute, self:Common Numeric Text Block Object Reference=@n5, NewValue:float=@n4.MinimalNotableValue) -> (then)
   > InterpolateToValue!(
    execute
    self:Common Numeric Text Block Object Reference=@n5
    TargetValue:float=GetStat!(self:Athena Base Stat View Object Reference=@n2, InStat:"Gameplay Tag Structure (by ref)"=@n3) -> (ReturnValue:float)
    MaximumInterpolationDuration:float=0.5
    MinimumChangeRate:float=0.5
    OutroOffset:float=0
   ) -> (then)
  ]}]
variables []
widgets {
 info {
  inheritedWidgetCount=1
  parentClass=/Script/FortniteUI.AthenaProfileStatWidgetBase
  rootWidgetClass=/Script/FortniteUI.AthenaProfileStatBox
  widgetCount=2
 }
 widgets [
  ProfileStatBox {widgetClassPath=&AthenaProfileStatBox, bInherited=true, bIsVariable=true, widget=ProfileStatBox}
  Value {
   widgetClassPath=/Script/CommonUI.CommonNumericTextBlock
   bIsVariable=true
   parent=ProfileStatBox
   widget=Value
   properties {
    text=Text Block
    font {
     bForceMonospaced=false
     bMaterialIsStencil=false
     fontMaterial=null
     fontObject=&Burbank
     letterSpacing=0
     monospacedWidth=1
     size=22
     skewAmount=0
     typefaceFontName=BurbankRegular-Black
     outlineSettings {
      bApplyOutlineToDropShadows=false
      bMiteredCorners=false
      bSeparateFillAlpha=false
      outlineMaterial=null
      outlineSize=0
      outlineColor=<rgba 0 0 0 1>
     }
    }
    renderTransform {angle=0, scale=<1 1>, shear=<-8 0>, translation=<0 0>}
    shadowOffset=<0 0>
   }
   slot {class=SizeBoxSlot}
  }
 ]
}
