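# MatchStatsGrid is an end-game HUD widget that lists a player's match stats as labeled rows, such as eliminations, accuracy, damage and materials.
# An AthenaProfileStatBox root holds a grid that pairs each text label with a StatValueWidget bound to a GameplayStat tag.
# SetStats passes an Athena match stat view object to the stat box through SetStatView, without respecting the parent stat view.
# The event graph is empty.
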WidgetBlueprint /Game/Athena/HUD/EndGame/MatchStatsGrid
graphs [
 EventGraph {links [], nodes []}
 SetStats {nodes [
   n1 = SetStats!() -> (then, AthenaMatchStats:Athena Match Stat View Object Reference) {is=title K2Node_FunctionEntry}
   > SetStatView!(
    execute
    self:Athena Profile Stat Box Object Reference=Get ProfileStatBox() -> (ProfileStatBox:Athena Profile Stat Box Object Reference)
    InStatView:Athena Base Stat View Object Reference=@n1
    RespectParentStatView:bool=false
   ) -> (then)
  ]}
]
variables []
widgets {
 info {inheritedWidgetCount=0, parentClass=&UserWidget, rootWidgetClass=/Script/FortniteUI.AthenaProfileStatBox, widgetCount=27}
 widgets [
  ProfileStatBox {widgetClassPath=&AthenaProfileStatBox, bIsVariable=true, widget=ProfileStatBox}
  StatGrid {widgetClassPath=&GridPanel, parent=ProfileStatBox, widget=StatGrid, properties {columnFill [0,0,0,1]}, slot {class=SizeBoxSlot}}
  Spacer_85 {widgetClassPath=&Spacer, parent=StatGrid, widget=Spacer_85, properties {size=<0 1>}, slot {class=GridSlot, properties {column=2}}}
  CommonTextBlock_23 {
   widgetClassPath=&CommonTextBlock
   parent=StatGrid
   widget=CommonTextBlock_23
   properties {
    text=Eliminations
    font {
     bForceMonospaced=false
     bMaterialIsStencil=false
     fontMaterial=null
     fontObject=&BurbankSmall
     letterSpacing=0
     monospacedWidth=1
     size=20
     skewAmount=0
     typefaceFontName=Bold
     outlineSettings {
      bApplyOutlineToDropShadows=false
      bMiteredCorners=false
      bSeparateFillAlpha=false
      outlineMaterial=null
      outlineSize=0
      outlineColor=<rgba 0 0 0 1>
     }
    }
    shadowOffset=<0 0>
   }
   slot {class=GridSlot, properties {verticalAlignment=VAlign_Center, padding=<ltrb 0 0 30 0>}}
  }
  KillStat {
   widgetClassPath=/Game/Athena/UI/Stats/StatValueWidget
   parent=StatGrid
   widget=KillStat
   properties {stat {tagName=GameplayStat.Profile.Match.PersonalElimination}}
   slot {class=GridSlot, properties {column=1, padding=<ltrb 0 0 20 0>}}
  }
  CommonTextBlock_25 {
   widgetClassPath=&CommonTextBlock
   parent=StatGrid
   widget=CommonTextBlock_25
   properties {
    text=Distance Traveled
    font {
     bForceMonospaced=false
     bMaterialIsStencil=false
     fontMaterial=null
     fontObject=&BurbankSmall
     letterSpacing=0
     monospacedWidth=1
     size=20
     skewAmount=0
     typefaceFontName=Bold
     outlineSettings {
      bApplyOutlineToDropShadows=false
      bMiteredCorners=false
      bSeparateFillAlpha=false
      outlineMaterial=null
      outlineSize=0
      outlineColor=<rgba 0 0 0 1>
     }
    }
    shadowOffset=<0 0>
   }
   slot {class=GridSlot, properties {row=6, verticalAlignment=VAlign_Center, padding=<ltrb 0 0 30 0>}}
  }
  DistanceStat {
   widgetClassPath=&StatValueWidget
   parent=StatGrid
   widget=DistanceStat
   properties {stat {tagName=GameplayStat.Profile.Match.TravelDistanceGround}}
   slot {class=GridSlot, properties {column=1, row=6}}
  }
  CommonTextBlock_27 {
   widgetClassPath=&CommonTextBlock
   parent=StatGrid
   widget=CommonTextBlock_27
   properties {
    text=Assists
    font {
     bForceMonospaced=false
     bMaterialIsStencil=false
     fontMaterial=null
     fontObject=&BurbankSmall
     letterSpacing=0
     monospacedWidth=1
     size=20
     skewAmount=0
     typefaceFontName=Bold
     outlineSettings {
      bApplyOutlineToDropShadows=false
      bMiteredCorners=false
      bSeparateFillAlpha=false
      outlineMaterial=null
      outlineSize=0
      outlineColor=<rgba 0 0 0 1>
     }
    }
    shadowOffset=<0 0>
   }
   slot {class=GridSlot, properties {row=1, verticalAlignment=VAlign_Center, padding=<ltrb 0 0 30 0>}}
  }
  AssistsStat {
   widgetClassPath=&StatValueWidget
   parent=StatGrid
   widget=AssistsStat
   properties {stat {tagName=GameplayStat.Profile.Match.Assists}}
   slot {class=GridSlot, properties {column=1, row=1, padding=<ltrb 0 0 20 0>}}
  }
  CommonTextBlock_29 {
   widgetClassPath=&CommonTextBlock
   parent=StatGrid
   widget=CommonTextBlock_29
   properties {
    text=Head Shots
    font {
     bForceMonospaced=false
     bMaterialIsStencil=false
     fontMaterial=null
     fontObject=&BurbankSmall
     letterSpacing=0
     monospacedWidth=1
     size=20
     skewAmount=0
     typefaceFontName=Bold
     outlineSettings {
      bApplyOutlineToDropShadows=false
      bMiteredCorners=false
      bSeparateFillAlpha=false
      outlineMaterial=null
      outlineSize=0
      outlineColor=<rgba 0 0 0 1>
     }
    }
    shadowOffset=<0 0>
   }
   slot {class=GridSlot, properties {row=5, verticalAlignment=VAlign_Center, padding=<ltrb 0 0 30 0>}}
  }
  HeadshotsStat {
   widgetClassPath=&StatValueWidget
   parent=StatGrid
   widget=HeadshotsStat
   properties {stat {tagName=GameplayStat.Profile.Weapon.CriticalHitsToHostiles}}
   slot {class=GridSlot, properties {column=1, row=5, padding=<ltrb 0 0 20 0>}}
  }
  CommonTextBlock_31 {
   widgetClassPath=&CommonTextBlock
   parent=StatGrid
   widget=CommonTextBlock_31
   properties {
    text=Revives
    font {
     bForceMonospaced=false
     bMaterialIsStencil=false
     fontMaterial=null
     fontObject=&BurbankSmall
     letterSpacing=0
     monospacedWidth=1
     size=20
     skewAmount=0
     typefaceFontName=Bold
     outlineSettings {
      bApplyOutlineToDropShadows=false
      bMiteredCorners=false
      bSeparateFillAlpha=false
      outlineMaterial=null
      outlineSize=0
      outlineColor=<rgba 0 0 0 1>
     }
    }
    shadowOffset=<0 0>
   }
   slot {class=GridSlot, properties {row=2, verticalAlignment=VAlign_Center, padding=<ltrb 0 0 30 0>}}
  }
  RevivesStat {
   widgetClassPath=&StatValueWidget
   parent=StatGrid
   widget=RevivesStat
   properties {stat {tagName=GameplayStat.Profile.Match.Revives}}
   slot {class=GridSlot, properties {column=1, row=2, padding=<ltrb 0 0 20 0>}}
  }
  CommonTextBlock_33 {
   widgetClassPath=&CommonTextBlock
   parent=StatGrid
   widget=CommonTextBlock_33
   properties {
    text=Materials Gathered
    font {
     bForceMonospaced=false
     bMaterialIsStencil=false
     fontMaterial=null
     fontObject=&BurbankSmall
     letterSpacing=0
     monospacedWidth=1
     size=20
     skewAmount=0
     typefaceFontName=Bold
     outlineSettings {
      bApplyOutlineToDropShadows=false
      bMiteredCorners=false
      bSeparateFillAlpha=false
      outlineMaterial=null
      outlineSize=0
      outlineColor=<rgba 0 0 0 1>
     }
    }
    shadowOffset=<0 0>
   }
   slot {class=GridSlot, properties {row=7, verticalAlignment=VAlign_Center, padding=<ltrb 0 0 30 0>}}
  }
  MaterialsGatheredStat {
   widgetClassPath=&StatValueWidget
   parent=StatGrid
   widget=MaterialsGatheredStat
   properties {stat {tagName=GameplayStat.Profile.Match.MaterialsGathered}}
   slot {class=GridSlot, properties {column=1, row=7}}
  }
  CommonTextBlock_35 {
   widgetClassPath=&CommonTextBlock
   parent=StatGrid
   widget=CommonTextBlock_35
   properties {
    text=Accuracy
    font {
     bForceMonospaced=false
     bMaterialIsStencil=false
     fontMaterial=null
     fontObject=&BurbankSmall
     letterSpacing=0
     monospacedWidth=1
     size=20
     skewAmount=0
     typefaceFontName=Bold
     outlineSettings {
      bApplyOutlineToDropShadows=false
      bMiteredCorners=false
      bSeparateFillAlpha=false
      outlineMaterial=null
      outlineSize=0
      outlineColor=<rgba 0 0 0 1>
     }
    }
    shadowOffset=<0 0>
   }
   slot {class=GridSlot, properties {row=3, verticalAlignment=VAlign_Center, padding=<ltrb 0 0 30 0>}}
  }
  AccuracyStat {
   widgetClassPath=&StatValueWidget
   parent=StatGrid
   widget=AccuracyStat
   properties {stat {tagName=GameplayStat.Profile.Match.Accuracy}}
   slot {class=GridSlot, properties {column=1, row=3, padding=<ltrb 0 0 20 0>}}
  }
  CommonTextBlock_37 {
   widgetClassPath=&CommonTextBlock
   parent=StatGrid
   widget=CommonTextBlock_37
   properties {
    text=Materials Used
    font {
     bForceMonospaced=false
     bMaterialIsStencil=false
     fontMaterial=null
     fontObject=&BurbankSmall
     letterSpacing=0
     monospacedWidth=1
     size=20
     skewAmount=0
     typefaceFontName=Bold
     outlineSettings {
      bApplyOutlineToDropShadows=false
      bMiteredCorners=false
      bSeparateFillAlpha=false
      outlineMaterial=null
      outlineSize=0
      outlineColor=<rgba 0 0 0 1>
     }
    }
    shadowOffset=<0 0>
   }
   slot {class=GridSlot, properties {row=8, verticalAlignment=VAlign_Center, padding=<ltrb 0 0 30 0>}}
  }
  MaterialsUsedStat {
   widgetClassPath=&StatValueWidget
   parent=StatGrid
   widget=MaterialsUsedStat
   properties {stat {tagName=GameplayStat.Profile.Match.MaterialsUsed}}
   slot {class=GridSlot, properties {column=1, row=8}}
  }
  CommonTextBlock_2 {
   widgetClassPath=&CommonTextBlock
   parent=StatGrid
   widget=CommonTextBlock_2
   properties {
    text=Damage To Players
    font {
     bForceMonospaced=false
     bMaterialIsStencil=false
     fontMaterial=null
     fontObject=&BurbankSmall
     letterSpacing=0
     monospacedWidth=1
     size=20
     skewAmount=0
     typefaceFontName=Bold
     outlineSettings {
      bApplyOutlineToDropShadows=false
      bMiteredCorners=false
      bSeparateFillAlpha=false
      outlineMaterial=null
      outlineSize=0
      outlineColor=<rgba 0 0 0 1>
     }
    }
    shadowOffset=<0 0>
   }
   slot {class=GridSlot, properties {row=4, verticalAlignment=VAlign_Center, padding=<ltrb 0 0 30 0>}}
  }
  DamageToPlayersStat {
   widgetClassPath=&StatValueWidget
   parent=StatGrid
   widget=DamageToPlayersStat
   properties {stat {tagName=GameplayStat.Profile.Weapon.TotalDamageToHostiles}}
   slot {class=GridSlot, properties {column=1, row=4}}
  }
  CommonTextBlock_3 {
   widgetClassPath=&CommonTextBlock
   parent=StatGrid
   widget=CommonTextBlock_3
   properties {
    text=Damage Taken
    font {
     bForceMonospaced=false
     bMaterialIsStencil=false
     fontMaterial=null
     fontObject=&BurbankSmall
     letterSpacing=0
     monospacedWidth=1
     size=20
     skewAmount=0
     typefaceFontName=Bold
     outlineSettings {
      bApplyOutlineToDropShadows=false
      bMiteredCorners=false
      bSeparateFillAlpha=false
      outlineMaterial=null
      outlineSize=0
      outlineColor=<rgba 0 0 0 1>
     }
    }
    shadowOffset=<0 0>
   }
   slot {class=GridSlot, properties {row=9, verticalAlignment=VAlign_Center, padding=<ltrb 0 0 30 0>}}
  }
  DamageReceivedStat {
   widgetClassPath=&StatValueWidget
   parent=StatGrid
   widget=DamageReceivedStat
   properties {stat {tagName=GameplayStat.Profile.Match.DamageTaken}}
   slot {class=GridSlot, properties {column=1, row=9}}
  }
  StructuralDamage {
   widgetClassPath=&CommonTextBlock
   parent=StatGrid
   widget=StructuralDamage
   properties {
    text=Damage To Structures
    font {
     bForceMonospaced=false
     bMaterialIsStencil=false
     fontMaterial=null
     fontObject=&BurbankSmall
     letterSpacing=0
     monospacedWidth=1
     size=20
     skewAmount=0
     typefaceFontName=Bold
     outlineSettings {
      bApplyOutlineToDropShadows=false
      bMiteredCorners=false
      bSeparateFillAlpha=false
      outlineMaterial=null
      outlineSize=0
      outlineColor=<rgba 0 0 0 1>
     }
    }
    shadowOffset=<0 0>
   }
   slot {class=GridSlot, properties {row=11, verticalAlignment=VAlign_Center, padding=<ltrb 0 0 30 0>}}
  }
  StructuralDamageStat {
   widgetClassPath=&StatValueWidget
   parent=StatGrid
   widget=StructuralDamageStat
   properties {stat {tagName=GameplayStat.Profile.Weapon.NormalDamageToStructures}}
   slot {class=GridSlot, properties {column=1, row=11}}
  }
  CommonTextBlock_5 {
   widgetClassPath=&CommonTextBlock
   parent=StatGrid
   widget=CommonTextBlock_5
   properties {
    text=Hits
    font {
     bForceMonospaced=false
     bMaterialIsStencil=false
     fontMaterial=null
     fontObject=&BurbankSmall
     letterSpacing=0
     monospacedWidth=1
     size=20
     skewAmount=0
     typefaceFontName=Bold
     outlineSettings {
      bApplyOutlineToDropShadows=false
      bMiteredCorners=false
      bSeparateFillAlpha=false
      outlineMaterial=null
      outlineSize=0
      outlineColor=<rgba 0 0 0 1>
     }
    }
    shadowOffset=<0 0>
   }
   slot {class=GridSlot, properties {row=10, verticalAlignment=VAlign_Center, padding=<ltrb 0 0 30 0>}}
  }
  HitsStat {
   widgetClassPath=&StatValueWidget
   parent=StatGrid
   widget=HitsStat
   properties {stat {tagName=GameplayStat.Profile.Weapon.TotalHitsToHostiles}}
   slot {class=GridSlot, properties {column=1, row=10, padding=<ltrb 0 0 20 0>}}
  }
 ]
}
