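# End-game HUD feed item widget that displays one line of XP reason text.
# A minimum-width SizeBox holds a right-justified RichTextBlock, styled by XpReasonStyleSheet and filled through SetXpText.
# PlayPresentEffect runs the Present_PopIn animation at a given playback speed, fading the text in while it shrinks and untilts to rest.
# The event graph is empty, so both functions depend on external callers.
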WidgetBlueprint /Game/Athena/HUD/EndGame/XpReasonFeedItem
animations [Present_PopIn {
  binding_count=1
  display_label=""
  length_seconds=1
  object=&self:Present_PopIn
  sequence {bindings [XpText {
     id=#1
     possessed_class=RichTextBlock
     tracks [
      {
       type=MovieScene2DTransformTrack
       sections [{
         type=MovieScene2DTransformSection
         end_frame=20
         start_frame=0
         channels [
          Translation.X {type=MovieSceneScriptingFloatChannel}
          Translation.Y {type=MovieSceneScriptingFloatChannel}
          Angle {type=MovieSceneScriptingFloatChannel, keys [[0,10],[20,0]]}
          Scale.X {type=MovieSceneScriptingFloatChannel, keys [[0,2],[20,1]]}
          Scale.Y {type=MovieSceneScriptingFloatChannel, keys [[0,2],[20,1]]}
          Shear.X {type=MovieSceneScriptingFloatChannel, const [0,1]}
          Shear.Y {type=MovieSceneScriptingFloatChannel, const [0,1]}
         ]
        }]
      }
      {
       type=MovieSceneFloatTrack
       display_name=Render Opacity
       sections [{type=MovieSceneFloatSection, end_frame=20, start_frame=0, channels [None {type=MovieSceneScriptingFloatChannel, keys [[0,0],[20,1]]}]}]
      }
     ]
    }]}
 }]
graphs [
 EventGraph {links [], nodes []}
 PlayPresentEffect {nodes [
   n1 = PlayPresentEffect!() -> (then, PlaybackSpeed:double=1) {is=title K2Node_FunctionEntry}
   > PlayAnimation(InAnimation=Get Present_PopIn() -> (Present_PopIn:Widget Animation Object Reference), PlaybackSpeed=@n1)
  ]}
 SetXpText {nodes [
   n1 = SetXpText!() -> (then, Text:Text) {is=title K2Node_FunctionEntry}
   > SetText(self=Get XpText() -> (XpText:Rich Text Block Object Reference), InText=@n1)
  ]}
]
variables []
widgets {
 info {inheritedWidgetCount=0, parentClass=&CommonUserWidget, rootWidgetClass=&SizeBox, widgetCount=2}
 widgets [
  SizeBox_0 {widgetClassPath=&SizeBox, widget=SizeBox_0, properties {bOverride_MinDesiredWidth=true, minDesiredWidth=200}}
  XpText {
   widgetClassPath=/Script/UMG.RichTextBlock
   bIsVariable=true
   parent=SizeBox_0
   widget=XpText
   properties {justification=Right, textStyleSet=&/XpReasonStyleSheet}
   slot {class=SizeBoxSlot}
  }
 ]
}
