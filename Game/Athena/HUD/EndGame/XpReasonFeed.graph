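# End-game HUD widget that shows a vertical feed of match XP reward reasons as XpReasonFeedItem entries in a dynamic entry box.
# Play Xp Reward adds one entry, sets its text, plays its present effect sped up, and holds an Animating flag through a short delay.
# IsReadyForNextXpReward reports when that flag is clear.
# ImmediatelyShowXpRewards rebuilds the feed without effects, and Destruct stops animations and clears it.
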WidgetBlueprint /Game/Athena/HUD/EndGame/XpReasonFeed
graphs [
 EventGraph {nodes [
   Reset Feed() {is=title K2Node_CustomEvent}
   > Reset!(
    execute
    self:Dynamic Entry Box Object Reference=Get DynamicXpFeed() -> (DynamicXpFeed:Dynamic Entry Box Object Reference)
    bDeleteWidgets:bool=true
   ) -> (then)
   n1 = Play Xp Reward!() -> (OutputDelegate, then, Reward:"Athena Match Xp Reward Structure (by ref)") {is=title K2Node_CustomEvent}
   > Set Animating(Animating:bool=true) -> (Output_Get:bool)
   > n2 = Create Entry!(execute, self:Dynamic Entry Box Object Reference=Get DynamicXpFeed()) -> (then, ReturnValue:User Widget Object Reference)
   > n3 = Cast To XpReasonFeedItem!(execute, Object=@n2) -> (then, CastFailed, AsXp Reason Feed Item:Xp Reason Feed Item Object Reference) {is=title K2Node_DynamicCast}
   > SetXpText!(
    execute
    self:Xp Reason Feed Item Object Reference=@n3
    Text:Text=Break Athena Match Xp Reward!(AthenaMatchXpReward:"Athena Match Xp Reward Structure (by ref)"=@n1.Reward) -> (Text:Text, Amount:int) {is=title K2Node_BreakStruct}
   ) -> (then)
   > ForceLayoutPrepass(self:Widget Object Reference)
   > PlayPresentEffect!(execute, self:Xp Reason Feed Item Object Reference=@n3, PlaybackSpeed:double=4) -> (then)
   > Delay(Duration=0.3)
   > Set Animating(Animating=false)
   Event Destruct()
   > StopAnimationsAndLatentActions(self:User Widget Object Reference)
   > Set Animating(Animating=false)
   > Reset Feed() {is=title K2Node_CallFunction}
  ]}
 ImmediatelyShowXpRewards {nodes [
   n1 = ImmediatelyShowXpRewards!() -> (then, XpRewards:Array of Athena Match Xp Reward Structures) {is=title K2Node_FunctionEntry}
   > Reset(self=Get DynamicXpFeed(), bDeleteWidgets=true)
   > n2 = For Each Loop(Array:Array of Athena Match Xp Reward Structures=@n1) -> (Array Element:Athena Match Xp Reward Structure)
   > n3 = Create Entry(self=Get DynamicXpFeed())
   > n4 = Cast To XpReasonFeedItem(Object=@n3)
   > SetXpText(self=@n4, Text=Break Athena Match Xp Reward(AthenaMatchXpReward=@n2))
  ]}
 IsReadyForNextXpReward {nodes [
   IsReadyForNextXpReward() {is=title K2Node_FunctionEntry}
   > Return Node!(execute, Ready:bool=NOT Boolean(A=Get Animating() -> (Animating:bool)))
  ]}
]
variables [Animating {type=boolean}]
widgets {
 info {inheritedWidgetCount=0, parentClass=&UserWidget, rootWidgetClass=&SizeBox, widgetCount=3}
 widgets [
  SizeBox_0 {widgetClassPath=&SizeBox, widget=SizeBox_0, properties {heightOverride=100, widthOverride=100}}
  Overlay_0 {widgetClassPath=&Overlay, parent=SizeBox_0, widget=Overlay_0, slot {class=SizeBoxSlot}}
  DynamicXpFeed {
   widgetClassPath=&DynamicEntryBox
   bIsVariable=true
   parent=Overlay_0
   widget=DynamicXpFeed
   properties {entryBoxType=Vertical, entryWidgetClass=&/XpReasonFeedItem.XpReasonFeedItem_C}
   slot {class=OverlaySlot, properties {horizontalAlignment=HAlign_Fill, verticalAlignment=VAlign_Fill}}
  }
 ]
}
