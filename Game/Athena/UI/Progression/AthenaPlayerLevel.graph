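# Athena player level widget showing rank banner, level number and XP bar, with named slots for extra content.
# It initializes from account or level info and animates XP rewards, with Tick using an animation-driven XpLerp to update the bar material and text.
# Level crossings in either direction pop the number, update the banner and broadcast LevelChanged, then replay leftover XP using that level's threshold from the Athena profile.
# At max level it shows MAX.
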WidgetBlueprint /Game/Athena/UI/Progression/AthenaPlayerLevel
animations [
 AnimateXp {
  binding_count=1
  display_label=""
  length_seconds=1
  object=&self:AnimateXp
  sequence {bindings [AthenaPlayerLevel_C_86 {
     id=#1
     possessed_class=AthenaPlayerLevel_C
     tracks [{
       type=MovieSceneFloatTrack
       display_name=Xp Lerp
       sections [{type=MovieSceneFloatSection, end_frame=20, start_frame=0, channels [None {type=MovieSceneScriptingFloatChannel, keys [[0,0],[20,1]]}]}]
      }]
    }]}
 }
 AnimateXpAndLevel {
  binding_count=2
  display_label=""
  length_seconds=1.25
  object=&self:AnimateXpAndLevel
  sequence {
   bindings [
    LevelText {
     id=#2
     possessed_class=CommonTextBlock
     tracks [{
       type=MovieScene2DTransformTrack
       sections [{
         type=MovieScene2DTransformSection
         end_frame=25
         start_frame=15
         channels [
          Translation.X {type=MovieSceneScriptingFloatChannel}
          Translation.Y {type=MovieSceneScriptingFloatChannel}
          Angle {type=MovieSceneScriptingFloatChannel}
          Scale.X {type=MovieSceneScriptingFloatChannel, keys [[15,1],[20,1.5],[25,1]]}
          Scale.Y {type=MovieSceneScriptingFloatChannel, keys [[15,1],[20,1.5],[25,1]]}
          Shear.X {type=MovieSceneScriptingFloatChannel}
          Shear.Y {type=MovieSceneScriptingFloatChannel}
         ]
        }]
      }]
    }
    AthenaPlayerLevel_C_86 {
     id=#1
     possessed_class=AthenaPlayerLevel_C
     tracks [{
       type=MovieSceneFloatTrack
       display_name=Xp Lerp
       sections [{type=MovieSceneFloatSection, end_frame=25, start_frame=0, channels [None {type=MovieSceneScriptingFloatChannel, keys [[0,0],[20,1],[25,1]]}]}]
      }]
    }
   ]
   tracks [{type=MovieSceneEventTrack, sections [{type=MovieSceneEventSection}]}]
  }
 }
]
graphs [
 AnimateBar {nodes [
   AnimateBar() {is=title K2Node_FunctionEntry}
   n1 = "To Float (Integer)"(InInt=Get CurrentXp() -> (CurrentXp:int=0))
   n2 = "To Float (Integer)"(InInt=Get TargetXp() -> (TargetXp:int=0))
   n3 = Get XpLerp() -> (XpLerp:double)
   > UpdateXp!(
    execute
    self
    LevelXp:int=Truncate(A=Lerp(A=@n1, B=@n2, Alpha=@n3))
    LevelXpForLevel:int=Get LevelXpForLevel() -> (LevelXpForLevel:int=0)
    Level:int=Get CurrentLevel() -> (CurrentLevel:int=0)
    MaxLevel:int=Get MaxLevel() -> (MaxLevel:int=0)
   ) -> (then)
  ]}
 EventGraph {nodes [
   EventTrack_Leveled() {is=title K2Node_CustomEvent}
   n1 = "integer > integer"(A=Get TargetXp() -> (TargetXp=0), B=0) {is=title K2Node_CallFunction}
   > Set CurrentLevel(CurrentLevel:int=int + int(A=SelectInt(A=1, B=-1, bPickA=@n1), B=Get CurrentLevel() -> (CurrentLevel=0)) -> (ReturnValue=0) {is=title K2Node_CommutativeAssociativeBinaryOperator}) -> (Output_Get:int=0)
   > UpdateLevel!(execute, self, Value:int=Get CurrentLevel() -> (CurrentLevel=0)) -> (then)
   > SetBannerSurroundLevel!(
    execute
    self:Player Rank Emblem Object Reference=Get PlayerBanner() -> (PlayerBanner:Player Rank Emblem Object Reference)
    Level:int=Get CurrentLevel() -> (CurrentLevel=0)
    Animate:bool=true
   ) -> (then)
   > Call Level Changed!(execute:Exec, self:Athena Player Level Object Reference, Level:int=Get CurrentLevel() -> (CurrentLevel=0)) -> (then:Exec) {is=title K2Node_CallDelegate}
   n2 = Play Xp Reward!() -> (OutputDelegate, then, Reward:"Athena Match Xp Reward Structure (by ref)") {is=title K2Node_CustomEvent}
   > Play XP Reward Directly!(
    execute
    self
    XpToAdd:int=Break Athena Match Xp Reward!(AthenaMatchXpReward:"Athena Match Xp Reward Structure (by ref)"=@n2.Reward) -> (Text:Text, Amount:int) {is=title K2Node_BreakStruct}.Amount
   ) -> (then)
   n3 = Play XP Reward Directly!() -> (OutputDelegate, then, XpToAdd:int) {is=title K2Node_CustomEvent}
   > n4 = Branch(Condition="integer < integer"!(A:int=Get CurrentLevel() -> (CurrentLevel=0), B:int=Get MaxLevel() -> (MaxLevel=0)) -> (ReturnValue:bool) {is=title K2Node_CallFunction})
   > Set Animating(Animating:bool=true) -> (Output_Get:bool)
   > Set XpToAdd(XpToAdd:int=@n3.XpToAdd) -> (Output_Get:int=0)
   > n5 = Set TargetXp(TargetXp:int=int + int(A=Get CurrentXp() -> (CurrentXp=0), B=Get XpToAdd() -> (XpToAdd:int=0)) -> (ReturnValue=0)) -> (Output_Get:int=0)
   > n6 = Branch(Condition="integer >= integer"!(A:int=@n5, B:int=Get LevelXpForLevel() -> (LevelXpForLevel=0)) -> (ReturnValue:bool) {is=title K2Node_CallFunction})
   > PlayAnimation(InAnimation=Get AnimateXpAndLevel() -> (AnimateXpAndLevel:Widget Animation Object Reference))
   n7 = Get LevelXpForLevel() -> (LevelXpForLevel=0)
   > Set WrapAroundXp(WrapAroundXp:int=int - int(A=Get TargetXp() -> (TargetXp=0), B=@n7) {is=title K2Node_CallFunction}) -> (Output_Get:int=0)
   > Set TargetXp(TargetXp=@n7) -> (Output_Get=0)
   n8 = Branch(execute=@n6.else, Condition="integer < integer"(A=Get TargetXp() -> (TargetXp=0), B=0))
   > PlayAnimation(InAnimation=Get AnimateXpAndLevel())
   n9 = GetAthenaProfile!(self:Fort Mcp Context Object Reference=Get FortMcpContext!() -> (ReturnValue:Fort Mcp Context Object Reference) {is=title K2Node_GetBlueprintContextEx}) -> (ReturnValue:Fort Mcp Profile Athena Object Reference)
   > Is Valid(InputObject=@n9)
   n10 = GetSeasonLevelInfo!(self:Fort Mcp Profile Athena Object Reference=@n9, SeasonLevel:int=int - int(A=Get CurrentLevel() -> (CurrentLevel=0), B=1)) -> (ReturnValue:Athena Season Reward Level Info Structure)
   > Set WrapAroundXp(WrapAroundXp=int + int(
     A=Get TargetXp() -> (TargetXp=0)
     B=Break Athena Season Reward Level Info!(AthenaSeasonRewardLevelInfo:"Athena Season Reward Level Info Structure (by ref)"=@n10) -> (Track:EAthenaSeasonRewardTrack Enum, Level:int, XpToNextLevel:int, Rewards:Array of Athena Reward Item Reference Structures) {is=title K2Node_BreakStruct}.XpToNextLevel
    ) -> (ReturnValue=0)) -> (Output_Get=0)
   > Set TargetXp(TargetXp=0) -> (Output_Get=0)
   PlayAnimation(execute=@n8.else, InAnimation=Get AnimateXp() -> (AnimateXp:Widget Animation Object Reference))
   Set XpToAdd(execute=@n4.else, XpToAdd=0) -> (Output_Get=0)
   > Set TargetXp(TargetXp=0) -> (Output_Get=0)
   > Set WrapAroundXp(WrapAroundXp=0) -> (Output_Get=0)
   Event Tick()
   > Branch(Condition=Get Animating() -> (Animating:bool))
   > AnimateBar() {is=title K2Node_CallFunction}
   Event Destruct()
   > StopAnimationsAndLatentActions(self:User Widget Object Reference)
   > Set Animating(Animating=false)
   n11 = Event PreConstruct()
   > Branch(Condition=@n11.IsDesignTime)
   > Set TargetXp(TargetXp=1) -> (Output_Get=0)
   "Animation Finished (AnimateXp)"!() -> (OutputDelegate:Delegate, then:Exec) {is=title K2Node_WidgetAnimationEvent}
   > Set Animating(Animating=false)
   > Set CurrentXp(CurrentXp:int=Get TargetXp() -> (TargetXp=0)) -> (Output_Get:int=0)
   > AnimateBar()
   "Animation Finished (AnimateXpAndLevel)"!() -> (OutputDelegate:Delegate, then:Exec) {is=title K2Node_WidgetAnimationEvent}
   > Set Animating(Animating=false)
   > Set CurrentXp(CurrentXp=0) -> (Output_Get=0)
   > Play XP Reward Directly(XpToAdd=Get WrapAroundXp() -> (WrapAroundXp:int=0)) {is=title K2Node_CallFunction}
   n12 = GetAthenaProfile(self=Get FortMcpContext())
   > Is Valid(InputObject=@n12)
   n13 = Get CurrentLevel() -> (CurrentLevel=0)
   > Set LevelXpForLevel(LevelXpForLevel:int=Break Athena Season Reward Level Info(AthenaSeasonRewardLevelInfo=GetSeasonLevelInfo(self=@n12, SeasonLevel=@n13)).XpToNextLevel) -> (Output_Get:int=0)
  ]}
 InitializeFromAccount {nodes [
   n1 = InitializeFromAccount!() -> (then, FortPublicAccountInfo:"Fort Public Account Info Structure (by ref)") {is=title K2Node_FunctionEntry}
   > InitializeFromLevelInfo!(
    execute
    self
    LevelInfo:Athena Level Info Structure=Break Fort Public Account Info!(FortPublicAccountInfo:"Fort Public Account Info Structure (by ref)"=@n1) -> (Level:int, MaxLevel:int, LevelXp:int, LevelXpForLevel:int, BattleRoyaleLevel:Athena Level Info Structure) {is=title K2Node_BreakStruct}.BattleRoyaleLevel
   ) -> (then)
  ]}
 InitializeFromLevelInfo {nodes [
   n1 = InitializeFromLevelInfo!() -> (then, LevelInfo:Athena Level Info Structure) {is=title K2Node_FunctionEntry}
   n2 = Break Athena Level Info!(AthenaLevelInfo:"Athena Level Info Structure (by ref)"=@n1) -> (AccountLevel:int, Level:int, MaxLevel:int, LevelXp:int, LevelXpForLevel:int, BookMaxLevel:int, BookLevelXpForLevel:int) {is=title K2Node_BreakStruct}
   > Set CurrentLevel(CurrentLevel=@n2.Level) -> (Output_Get=0)
   > Set MaxLevel(MaxLevel:int=@n2.MaxLevel) -> (Output_Get:int=0)
   > Set CurrentXp(CurrentXp=@n2.LevelXp) -> (Output_Get=0)
   > Set LevelXpForLevel(LevelXpForLevel=@n2.LevelXpForLevel) -> (Output_Get=0)
   > n3 = Sequence()
   > UpdateLevel(Value=Get CurrentLevel() -> (CurrentLevel=0))
   UpdateXp(
    execute=@n3.then_1
    LevelXp=Get CurrentXp() -> (CurrentXp=0)
    LevelXpForLevel=Get LevelXpForLevel() -> (LevelXpForLevel=0)
    Level=Get CurrentLevel() -> (CurrentLevel=0)
    MaxLevel=Get MaxLevel() -> (MaxLevel=0)
   )
  ]}
 IsReadyForNextXpReward {nodes [
   IsReadyForNextXpReward() {is=title K2Node_FunctionEntry}
   > Return Node!(execute, Ready:bool=NOT Boolean(A=Get Animating()))
  ]}
 LevelChanged {nodes [LevelChanged!() -> (then, Level:int) {is=title K2Node_FunctionEntry}]}
 UpdateLevel {nodes [
   n1 = UpdateLevel!() -> (then, Value:int) {is=title K2Node_FunctionEntry}
   > "SetText (Text)"(self=Get LevelText() -> (LevelText:Common Text Object Reference), InText="To Text (Integer)"(Value=@n1))
  ]}
 UpdateXp {nodes [
   n1 = UpdateXp!() -> (then, LevelXp:int, LevelXpForLevel:int, Level:int, MaxLevel:int) {is=title K2Node_FunctionEntry}
   > n2 = GetDynamicMaterial(self=Get XpBar() -> (XpBar:Image Object Reference))
   > n3 = Branch(Condition="integer >= integer"(A=@n1.Level, B=@n1.MaxLevel))
   > SetScalarParameterValue(self=@n2, ParameterName=Progress, Value=1.0f)
   > "SetText (Text)"(self=Get XpText() -> (XpText:Common Text Object Reference), InText=Format Text(Format="LOCTEXT(\"MAX\")"))
   SetScalarParameterValue(
    execute=@n3.else
    self=@n2
    ParameterName=Progress
    Value=float / float(A="To Float (Integer)"(InInt=@n1), B="To Float (Integer)"(InInt=@n1.LevelXpForLevel)) {is=title K2Node_CallFunction}
   )
   > "SetText (Text)"(
    self=Get XpText()
    InText=Format Text!(Format="LOCTEXT(\"{xp_owned} / {xp_needed}\")", xp_owned:int=@n1, xp_needed:int=@n1.LevelXpForLevel) -> (Result)
   )
  ]}
]
variables [
 Animating {type=boolean}
 CurrentLevel {type=integer, default=1}
 CurrentXp {type=integer}
 LevelChanged {}
 LevelXpForLevel {type=integer}
 MaxLevel {type=integer}
 TargetXp {type=integer}
 WrapAroundXp {type=integer}
 XpLerp {type=number}
 XpToAdd {type=integer}
]
widgets {
 info {inheritedWidgetCount=0, parentClass=&CommonUserWidget, rootWidgetClass=&HorizontalBox, widgetCount=18}
 widgets [
  HorizontalBox_0 {widgetClassPath=&HorizontalBox, widget=HorizontalBox_0, properties {flowDirectionPreference=Culture}}
  PlayerBanner {
   widgetClassPath=&/PlayerRankEmblem
   bIsVariable=true
   parent=HorizontalBox_0
   widget=PlayerBanner
   properties {flowDirectionPreference=Culture}
   slot {class=HorizontalBoxSlot, properties {horizontalAlignment=HAlign_Left, verticalAlignment=VAlign_Top, padding=<ltrb -9 -13 -7 -23>}}
  }
  VerticalBox_0 {
   widgetClassPath=&VerticalBox
   parent=HorizontalBox_0
   widget=VerticalBox_0
   slot {class=HorizontalBoxSlot, properties {horizontalAlignment=HAlign_Left, padding=<ltrb 4 8 0 8>, size {sizeRule=Fill, value=1}}}
  }
  HorizontalBox_2 {
   widgetClassPath=&HorizontalBox
   parent=VerticalBox_0
   widget=HorizontalBox_2
   slot {class=VerticalBoxSlot, properties {verticalAlignment=VAlign_Bottom, padding=<ltrb 0 0 0 -10>}}
  }
  CommonTextBlock_1 {
   widgetClassPath=&CommonTextBlock
   parent=HorizontalBox_2
   widget=CommonTextBlock_1
   properties {
    text=LEVEL
    font {
     bForceMonospaced=false
     bMaterialIsStencil=false
     fontMaterial=null
     fontObject=&BurbankSmall
     letterSpacing=0
     monospacedWidth=1
     size=24
     skewAmount=0
     typefaceFontName=Black
     outlineSettings {
      bApplyOutlineToDropShadows=false
      bMiteredCorners=false
      bSeparateFillAlpha=false
      outlineMaterial=null
      outlineSize=0
      outlineColor=<rgba 0 0 0 1>
     }
    }
    shadowOffset=<0 0>
   }
   slot {class=HorizontalBoxSlot, properties {verticalAlignment=VAlign_Bottom, padding=<ltrb 0 0 8 -4>}}
  }
  LevelText {
   widgetClassPath=&CommonTextBlock
   bIsVariable=true
   parent=HorizontalBox_2
   widget=LevelText
   properties {
    text=25
    font {
     bForceMonospaced=false
     bMaterialIsStencil=false
     fontMaterial=null
     fontObject=&Burbank
     letterSpacing=0
     monospacedWidth=1
     size=48
     skewAmount=0
     typefaceFontName=Burbank-Black
     outlineSettings {
      bApplyOutlineToDropShadows=false
      bMiteredCorners=false
      bSeparateFillAlpha=false
      outlineMaterial=null
      outlineSize=0
      outlineColor=<rgba 0 0 0 1>
     }
    }
    shadowOffset=<0 0>
   }
   slot {class=HorizontalBoxSlot, properties {padding=<ltrb 0 0 0 -16>}}
  }
  TotalContentSlot {
   widgetClassPath=&NamedSlot
   bIsVariable=true
   parent=HorizontalBox_2
   widget=TotalContentSlot
   properties {bExposeOnInstanceOnly=true}
   slot {class=HorizontalBoxSlot, properties {horizontalAlignment=HAlign_Right, size {sizeRule=Fill, value=1}}}
  }
  SizeBox_3 {
   widgetClassPath=&SizeBox
   parent=VerticalBox_0
   widget=SizeBox_3
   properties {bOverride_MinDesiredWidth=true, minDesiredWidth=288}
   slot {class=VerticalBoxSlot, properties {verticalAlignment=VAlign_Top}}
  }
  HorizontalBox_1 {widgetClassPath=&HorizontalBox, parent=SizeBox_3, widget=HorizontalBox_1, slot {class=SizeBoxSlot}}
  VerticalBox_2 {
   widgetClassPath=&VerticalBox
   parent=HorizontalBox_1
   widget=VerticalBox_2
   slot {class=HorizontalBoxSlot, properties {size {sizeRule=Fill, value=1}}}
  }
  HorizontalBox_3 {
   widgetClassPath=&HorizontalBox
   parent=VerticalBox_2
   widget=HorizontalBox_3
   slot {class=VerticalBoxSlot, properties {padding=<ltrb 0 16 0 0>}}
  }
  XpBar {
   widgetClassPath=&Image
   bIsVariable=true
   parent=HorizontalBox_3
   widget=XpBar
   properties {
    bFlipForRightToLeftFlowDirection=true
    brush {
     bIsDynamicallyLoaded=false
     drawAs=Image
     imageType=NoImage
     mirroring=NoMirror
     resourceName=None
     resourceObject=&/MI_Level_Bar_Neutral
     tiling=NoTile
     imageSize=<16 16>
     margin=<ltrb 0 0 0 0>
     outlineSettings {
      bUseBrushTransparency=false
      roundingType=HalfHeightRadius
      width=0
      color {colorUseRule=UseColor_Specified, specifiedColor=<rgba 0 0 0 0>}
      cornerRadii=<0 0 0 0>
     }
     tintColor {colorUseRule=UseColor_Specified, specifiedColor=<rgba 1 1 1 1>}
     uVRegion {bIsValid=false, max=<0 0>, min=<0 0>}
    }
   }
   slot {class=HorizontalBoxSlot, properties {verticalAlignment=VAlign_Center, padding=<ltrb 0 0 8 0>, size {sizeRule=Fill, value=1}}}
  }
  SizeBox_0 {
   widgetClassPath=&SizeBox
   parent=HorizontalBox_3
   widget=SizeBox_0
   properties {bOverride_HeightOverride=true, heightOverride=40}
   slot {class=HorizontalBoxSlot}
  }
  LevelRewardArea {
   widgetClassPath=&NamedSlot
   bIsVariable=true
   parent=SizeBox_0
   widget=LevelRewardArea
   properties {bExposeOnInstanceOnly=true}
   slot {class=SizeBoxSlot, properties {horizontalAlignment=HAlign_Left, verticalAlignment=VAlign_Center}}
  }
  XpText {
   widgetClassPath=&CommonTextBlock
   bIsVariable=true
   parent=VerticalBox_2
   widget=XpText
   properties {
    text=0/123456
    colorAndOpacity {colorUseRule=UseColor_Specified, specifiedColor=<rgba 0.491 1 0.0999 1>}
    font {
     bForceMonospaced=false
     bMaterialIsStencil=false
     fontMaterial=null
     fontObject=&BurbankSmall
     letterSpacing=0
     monospacedWidth=1
     size=15
     skewAmount=0
     typefaceFontName=Bold
     outlineSettings {
      bApplyOutlineToDropShadows=false
      bMiteredCorners=false
      bSeparateFillAlpha=false
      outlineMaterial=null
      outlineSize=0
      outlineColor=<rgba 0 0 0 1>
     }
    }
    renderTransform {angle=0, scale=<1 1>, shear=<-8 0>, translation=<0 0>}
    shadowOffset=<0 0>
   }
   slot {class=VerticalBoxSlot, properties {horizontalAlignment=HAlign_Left, padding=<ltrb 0 -4 0 -4>}}
  }
  Border_0 {
   widgetClassPath=&Border
   parent=VerticalBox_2
   widget=Border_0
   properties {background {
     bIsDynamicallyLoaded=false
     drawAs=Border
     imageType=NoImage
     mirroring=NoMirror
     resourceName=None
     resourceObject=null
     tiling=NoTile
     imageSize=<32 32>
     margin=<ltrb 0.2 0.2 0.2 0.2>
     outlineSettings {
      bUseBrushTransparency=false
      roundingType=HalfHeightRadius
      width=0
      color {colorUseRule=UseColor_Specified, specifiedColor=<rgba 0 0 0 0>}
      cornerRadii=<0 0 0 0>
     }
     tintColor {colorUseRule=UseColor_Specified, specifiedColor=<rgba 1 0 0 0>}
     uVRegion {bIsValid=false, max=<0 0>, min=<0 0>}
    }}
   slot {class=VerticalBoxSlot, properties {size {sizeRule=Fill, value=1}}}
  }
  SizeBox_4 {
   widgetClassPath=&SizeBox
   parent=Border_0
   widget=SizeBox_4
   properties {bOverride_HeightOverride=true, heightOverride=40}
   slot {class=BorderSlot}
  }
  ExtraContentArea {
   widgetClassPath=&NamedSlot
   bIsVariable=true
   parent=SizeBox_4
   widget=ExtraContentArea
   properties {bExposeOnInstanceOnly=true}
   slot {class=SizeBoxSlot, properties {horizontalAlignment=HAlign_Right}}
  }
 ]
}
