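# Widget that shows a small level-up reward in the progression UI as an item icon beside a quantity.
# SetupReward collapses the row when there is no reward, otherwise makes it visible without hit testing.
# With a reward, it breaks the item/quantity pair to set the count text and applies the item's small preview image from a soft texture.
# The LevelRewarded input is unused, and the event graph is empty.
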WidgetBlueprint /Game/Athena/UI/Progression/SmallLevelUpReward
graphs [
 EventGraph {links [], nodes []}
 SetupReward {nodes [
   n1 = SetupReward!() -> (then, HasReward:bool=~, RewardItem:Fort Item Instance Quantity Pair Structure, LevelRewarded:int) {is=title K2Node_FunctionEntry}
   > SetVisibility(
    self=Get RewardContainer() -> (RewardContainer:Horizontal Box Object Reference)
    InVisibility=Select(Option 0=Collapsed, Option 1=SelfHitTestInvisible, Index=@n1)
   )
   > Branch(Condition=@n1)
   > n2 = Sequence!(execute) -> (then_0, then_1, then_2:Exec, then_3:Exec)
   n3 = Break Fort Item Instance Quantity Pair!(FortItemInstanceQuantityPair:"Fort Item Instance Quantity Pair Structure (by ref)"=@n1.RewardItem) -> (Item:Fort Item Object Reference, InventoryType:EFortInventoryType Enum, Quantity:int) {is=title K2Node_BreakStruct}
   "SetText (Text)"(
    execute=@n2.then_2
    self=Get MiniReward1CountTB() -> (MiniReward1CountTB:Common Text Object Reference)
    InText="To Text (Integer)"(Value=@n3.Quantity)
   )
   SetBrushFromSoftTexture!(
    execute=@n2.then_3
    self:Image Object Reference=Get MiniReward1Image() -> (MiniReward1Image:Image Object Reference)
    SoftTexture:Texture 2D Soft Object Reference=GetSmallPreviewImage!(self:Fort Item Object Reference=@n3) -> (ReturnValue:Texture 2D Soft Object Reference)
    bMatchSize:bool=false
   ) -> (then)
  ]}
]
variables []
widgets {
 info {inheritedWidgetCount=0, parentClass=&UserWidget, rootWidgetClass=&HorizontalBox, widgetCount=4}
 widgets [
  RewardContainer {widgetClassPath=&HorizontalBox, bIsVariable=true, widget=RewardContainer}
  SizeBox_1 {
   widgetClassPath=&SizeBox
   parent=RewardContainer
   widget=SizeBox_1
   slot {class=HorizontalBoxSlot, properties {padding=<ltrb 0 0 4 0>}}
  }
  MiniReward1Image {
   widgetClassPath=&Image
   bIsVariable=true
   parent=SizeBox_1
   widget=MiniReward1Image
   properties {
    visibility=SelfHitTestInvisible
    brush {
     bIsDynamicallyLoaded=false
     drawAs=Image
     imageType=NoImage
     mirroring=NoMirror
     resourceName=None
     resourceObject=/Game/UI/Foundation/Textures/Icons/Items/T-FNBR-SeasonalXP
     tiling=NoTile
     imageSize=<32 32>
     margin=<ltrb 0 0 0 0>
     outlineSettings {
      bUseBrushTransparency=false
      roundingType=HalfHeightRadius
      width=0
      color {colorUseRule=UseColor_Specified, specifiedColor=<rgba 0 0 0 0>}
      cornerRadii=<0 0 0 0>
     }
     tintColor {colorUseRule=UseColor_Specified, specifiedColor=<rgba 1 1 1 1>}
     uVRegion {bIsValid=false, max=<0 0>, min=<0 0>}
    }
   }
   slot {class=SizeBoxSlot, properties {horizontalAlignment=HAlign_Center, verticalAlignment=VAlign_Center}}
  }
  MiniReward1CountTB {
   widgetClassPath=&CommonTextBlock
   bIsVariable=true
   parent=RewardContainer
   widget=MiniReward1CountTB
   properties {
    style=/FortUILibrary/Styles/Typography/Foundation/TextStyle-Base-XS-B-70pc.TextStyle-Base-XS-B-70pc_C
    text=500
    colorAndOpacity {colorUseRule=UseColor_Specified, specifiedColor=<rgba 1 1 1 0.7>}
    font {
     bForceMonospaced=false
     bMaterialIsStencil=false
     fontMaterial=null
     fontObject=&NotoSans
     letterSpacing=0
     monospacedWidth=1
     size=16
     skewAmount=0
     typefaceFontName=NotoSans-Bold
     outlineSettings {
      bApplyOutlineToDropShadows=false
      bMiteredCorners=false
      bSeparateFillAlpha=false
      outlineMaterial=null
      outlineSize=0
      outlineColor=<rgba 0 0 0 1>
     }
    }
    shadowOffset=<0 0>
   }
   slot {class=HorizontalBoxSlot}
  }
 ]
}
