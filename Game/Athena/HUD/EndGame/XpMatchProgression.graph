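# End-of-match XP progression widget that animates total score, seasonal XP and the player's level bar from a match report.
# Play Xp Rewards resets the feed, interpolates the totals and arms a Tick-driven sequencer.
# The sequencer feeds one reward at a time to PlayerLevelBar and XpReasonFeed once both are ready, then calls Finished Playing Animation.
# Level changes refresh the LevelUpReward preview, hosted in the bar's named slots, from the Athena profile.
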WidgetBlueprint /Game/Athena/HUD/EndGame/XpMatchProgression
graphs [
 EventGraph {nodes [
   n1 = "Level Changed (PlayerLevelBar)"!() -> (OutputDelegate, then, Level:int=0) {is=title K2Node_ComponentBoundEvent}
   > UpdateRewardForLevel!(execute, self, SeasonLevel:int=@n1.Level) -> (then)
   n2 = Play Xp Rewards!() -> (OutputDelegate, then, Report:Athena Player Match Report Object Reference) {is=title K2Node_CustomEvent}
   > n3 = Set Report(Report:Athena Player Match Report Object Reference=@n2.Report) -> (Output_Get:Athena Player Match Report Object Reference)
   > n4 = GetXpRewards!(execute, self:Athena Player Match Report Object Reference=@n3) -> (then, XpRewards:Array of Athena Match Xp Reward Structures)
   > Set XpRewards(XpRewards:Array of Athena Match Xp Reward Structures=@n4) -> (Output_Get:Array of Athena Match Xp Reward Structures)
   > Set PlayingLeveling(PlayingLeveling:bool=true) -> (Output_Get:bool)
   > Set XpRewardIndex(XpRewardIndex:int=0) -> (Output_Get:int=0)
   > Reset Feed(self:Xp Reason Feed Object Reference=Get XpReasonFeed() -> (XpReasonFeed:Xp Reason Feed Object Reference))
   > SetCurrentValue!(
    execute
    self:Common Numeric Text Block Object Reference=[
     Get TotalScore() -> (TotalScore:Common Numeric Text Block Object Reference)
     Get TotalXp() -> (TotalXp:Common Numeric Text Block Object Reference)
    ]
    NewValue:float=0
   ) -> (then)
   n5 = Get Report() -> (Report:Athena Player Match Report Object Reference)
   > InterpolateToValue!(
    execute
    self:Common Numeric Text Block Object Reference=Get TotalScore()
    TargetValue:float="To Float (Integer)"(InInt=GetTotalScoreEarned!(self:Athena Player Match Report Object Reference=@n5) -> (ReturnValue:int=0))
    MaximumInterpolationDuration:float=1
    MinimumChangeRate:float=1
    OutroOffset:float=0
   ) -> (then)
   n6 = Get Report()
   > InterpolateToValue(
    self=Get TotalXp()
    TargetValue="To Float (Integer)"(InInt=GetTotalSeasonalXpEarned!(self:Athena Player Match Report Object Reference=@n6) -> (ReturnValue:int=0))
    MaximumInterpolationDuration=1
    MinimumChangeRate=1
    OutroOffset=0
   )
   Event Tick()
   > TickRewardAnimation()
   Event Destruct()
   > StopAnimationsAndLatentActions(self:User Widget Object Reference)
   > Set PlayingLeveling(PlayingLeveling=false)
   Finished Playing Animation() {is=title K2Node_CustomEvent}
  ]}
 IsReadyForNextXpEvent {nodes [
   IsReadyForNextXpEvent() {is=title K2Node_FunctionEntry}
   > n1 = IsReadyForNextXpReward!(execute, self:Athena Player Level Object Reference=Get PlayerLevelBar() -> (PlayerLevelBar:Athena Player Level Object Reference)) -> (then, Ready:bool)
   > n2 = IsReadyForNextXpReward(self:Xp Reason Feed Object Reference=Get XpReasonFeed())
   > Return Node!(execute, ReturnValue:bool=AND Boolean(A=@n2, B=@n1))
  ]}
 SetPlayerState {nodes [
   n1 = SetPlayerState!() -> (then, PlayerController:Fort Player Controller Athena Object Reference, AthenaLevelInfo:Athena Level Info Structure) {is=title K2Node_FunctionEntry}
   > InitializeFromLevelInfo!(execute, self:Athena Player Level Object Reference=Get PlayerLevelBar(), LevelInfo:Athena Level Info Structure=@n1.AthenaLevelInfo) -> (then)
   > UpdateRewardForLevel(SeasonLevel=Break Athena Level Info!(AthenaLevelInfo:"Athena Level Info Structure (by ref)"=@n1.AthenaLevelInfo) -> (AccountLevel:int, Level:int, MaxLevel:int, LevelXp:int, LevelXpForLevel:int, BookMaxLevel:int, BookLevelXpForLevel:int) {is=title K2Node_BreakStruct}.Level)
  ]}
 TickRewardAnimation {nodes [
   TickRewardAnimation() {is=title K2Node_FunctionEntry}
   > n1 = IsReadyForNextXpEvent!(execute, self) -> (then, ReturnValue:bool) {is=title K2Node_CallFunction}
   > Branch(Condition=AND Boolean(A=@n1, B=Get PlayingLeveling() -> (PlayingLeveling:bool)))
   > n2 = Branch(Condition=Is Valid Index!(
     TargetArray:Array of Athena Match Xp Reward Structures=Get XpRewards() -> (XpRewards:Array of Athena Match Xp Reward Structures)
     IndexToTest:int=Get XpRewardIndex() -> (XpRewardIndex:int=0)
    ) -> (ReturnValue:bool))
   n3 = "Get (a ref)"(Array:Array of Athena Match Xp Reward Structures=Get XpRewards(), Dimension 1=Get XpRewardIndex() -> (XpRewardIndex=0)) -> (Output:"Athena Match Xp Reward Structure (by ref)")
   > Play Xp Reward!(execute, self:Athena Player Level Object Reference=Get PlayerLevelBar(), Reward:"Athena Match Xp Reward Structure (by ref)"=@n3) -> (then)
   > Play Xp Reward(self:Xp Reason Feed Object Reference=Get XpReasonFeed(), Reward=@n3)
   > Increment Int(Value=Get XpRewardIndex() -> (XpRewardIndex=0))
   Set PlayingLeveling(execute=@n2.else, PlayingLeveling=false)
   > Finished Playing Animation() {is=title K2Node_CallFunction}
  ]}
 UpdateRewardForLevel {nodes [
   n1 = UpdateRewardForLevel!() -> (then, SeasonLevel:int) {is=title K2Node_FunctionEntry}
   n2 = GetAthenaProfile!(self:Fort Mcp Context Object Reference=Get FortMcpContext!() -> (ReturnValue:Fort Mcp Context Object Reference) {is=title K2Node_GetBlueprintContextEx}) -> (ReturnValue:Fort Mcp Profile Athena Object Reference)
   > Is Valid(InputObject=@n2)
   > n3 = GetSeasonLevelUpReward!(execute, self:Fort Mcp Profile Athena Object Reference=@n2, SeasonLevel:int=@n1) -> (then, Reward:Fort Item Instance Quantity Pair Structure, RewardLevel:int=0, ReturnValue:bool)
   > SetupReward!(
    execute
    self:Small Level Up Reward Object Reference=Get LevelUpReward() -> (LevelUpReward:Small Level Up Reward Object Reference)
    HasReward:bool=@n3.ReturnValue
    RewardItem:Fort Item Instance Quantity Pair Structure=@n3
    LevelRewarded:int=@n3.RewardLevel
   ) -> (then)
  ]}
]
variables [
 PlayingLeveling {type=boolean}
 Report {type=AthenaPlayerMatchReport, default=null}
 XpRewardIndex {type=integer}
 XpRewards {type="array<AthenaMatchXpReward>", default []}
]
widgets {
 info {inheritedWidgetCount=0, namedSlotCount=3, parentClass=&CommonUserWidget, rootWidgetClass=&Overlay, widgetCount=10}
 widgets [
  Overlay_0 {widgetClassPath=&Overlay, widget=Overlay_0}
  PlayerLevelBar {
   widgetClassPath=/Game/Athena/UI/Progression/AthenaPlayerLevel
   bIsVariable=true
   parent=Overlay_0
   widget=PlayerLevelBar
   slot {class=OverlaySlot, properties {horizontalAlignment=HAlign_Fill, verticalAlignment=VAlign_Fill}}
  }
  SizeBox_0 {widgetClassPath=&SizeBox, namedSlotHost=PlayerLevelBar, widget=SizeBox_0}
  GridPanel_0 {
   widgetClassPath=&GridPanel
   parent=SizeBox_0
   widget=GridPanel_0
   slot {class=SizeBoxSlot, properties {verticalAlignment=VAlign_Center, padding=<ltrb 20 0 0 0>}}
  }
  TextBlock_0 {
   widgetClassPath=&TextBlock
   parent=GridPanel_0
   widget=TextBlock_0
   properties {
    text=Total Score
    font {
     bForceMonospaced=false
     bMaterialIsStencil=false
     fontMaterial=null
     fontObject=&NotoSans
     letterSpacing=0
     monospacedWidth=1
     size=18
     skewAmount=0
     typefaceFontName=NotoSans-Bold
     outlineSettings {
      bApplyOutlineToDropShadows=true
      bMiteredCorners=false
      bSeparateFillAlpha=false
      outlineMaterial=null
      outlineSize=0
      outlineColor=<rgba 0 0 0 1>
     }
    }
   }
   slot {class=GridSlot, properties {verticalAlignment=VAlign_Center}}
  }
  TotalScore {
   widgetClassPath=/Script/CommonUI.CommonNumericTextBlock
   bIsVariable=true
   parent=GridPanel_0
   widget=TotalScore
   properties {
    text=Text Block
    font {
     bForceMonospaced=false
     bMaterialIsStencil=false
     fontMaterial=null
     fontObject=&NotoSans
     letterSpacing=0
     monospacedWidth=1
     size=18
     skewAmount=0
     typefaceFontName=NotoSans-Bold
     outlineSettings {
      bApplyOutlineToDropShadows=true
      bMiteredCorners=false
      bSeparateFillAlpha=false
      outlineMaterial=null
      outlineSize=0
      outlineColor=<rgba 0 0 0 1>
     }
    }
   }
   slot {class=GridSlot, properties {column=1, padding=<ltrb 10 0 0 0>}}
  }
  TextBlock_1 {
   widgetClassPath=&TextBlock
   parent=GridPanel_0
   widget=TextBlock_1
   properties {
    text=Total XP
    font {
     bForceMonospaced=false
     bMaterialIsStencil=false
     fontMaterial=null
     fontObject=&NotoSans
     letterSpacing=0
     monospacedWidth=1
     size=18
     skewAmount=0
     typefaceFontName=NotoSans-Bold
     outlineSettings {
      bApplyOutlineToDropShadows=true
      bMiteredCorners=false
      bSeparateFillAlpha=false
      outlineMaterial=null
      outlineSize=0
      outlineColor=<rgba 0 0 0 1>
     }
    }
   }
   slot {class=GridSlot, properties {row=1, verticalAlignment=VAlign_Center}}
  }
  TotalXp {
   widgetClassPath=&CommonNumericTextBlock
   bIsVariable=true
   parent=GridPanel_0
   widget=TotalXp
   properties {
    text=Text Block
    font {
     bForceMonospaced=false
     bMaterialIsStencil=false
     fontMaterial=null
     fontObject=&NotoSans
     letterSpacing=0
     monospacedWidth=1
     size=18
     skewAmount=0
     typefaceFontName=NotoSans-Bold
     outlineSettings {
      bApplyOutlineToDropShadows=true
      bMiteredCorners=false
      bSeparateFillAlpha=false
      outlineMaterial=null
      outlineSize=0
      outlineColor=<rgba 0 0 0 1>
     }
    }
   }
   slot {class=GridSlot, properties {column=1, row=1, padding=<ltrb 10 0 0 0>}}
  }
  XpReasonFeed {widgetClassPath=&/XpReasonFeed, bIsVariable=true, namedSlotHost=PlayerLevelBar, widget=XpReasonFeed}
  LevelUpReward {widgetClassPath=/Game/Athena/UI/Progression/SmallLevelUpReward, bIsVariable=true, namedSlotHost=PlayerLevelBar, widget=LevelUpReward}
 ]
}
