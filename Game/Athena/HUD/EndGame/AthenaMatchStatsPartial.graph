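# End-of-game Athena HUD panel showing match stats, team placement and XP progression, or a "Team Still Alive" throbber until rewards arrive.
# SetMatchReport rebinds the report's stats, rewards and team-stats delegates, then replays whatever is available; Destruct clears it.
# Handlers fill StatsGrid, format ordinal placement text, and switch to XP progression, playing it after a retriggerable delay.
# The placement border and rewards switcher start collapsed, and this graph never reveals them.
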WidgetBlueprint /Game/Athena/HUD/EndGame/AthenaMatchStatsPartial
graphs [
 EventGraph {nodes [
   Play XP Rewards() {is=title K2Node_CustomEvent}
   > RetriggerableDelay(Duration=0.3)
   > Play Xp Rewards!(
    execute
    self:Xp Match Progression Object Reference=Get XpMatchProgression() -> (XpMatchProgression:Xp Match Progression Object Reference)
    Report:Athena Player Match Report Object Reference=Get MatchReport() -> (MatchReport:Athena Player Match Report Object Reference)
   ) -> (then)
   Event Destruct()
   > SetMatchReport!(execute, self, Report:Athena Player Match Report Object Reference) -> (then)
  ]}
 OnRewardsAvailable {nodes [
   OnRewardsAvailable() {is=title K2Node_FunctionEntry}
   > SetActiveWidget(
    self:Widget Switcher Object Reference=Get RewardsSwitcher() -> (RewardsSwitcher:Widget Switcher Object Reference)
    Widget=Get XpMatchProgression()
   )
   > Play XP Rewards() {is=title K2Node_CallFunction}
  ]}
 OnStatsAvailable {nodes [
   OnStatsAvailable() {is=title K2Node_FunctionEntry}
   > n1 = CreateCurrentMatchStatView!(
    execute
    self:Athena Profile Stat Context Object Reference=Get AthenaProfileStatContext!() -> (ReturnValue:Athena Profile Stat Context Object Reference) {is=title K2Node_GetBlueprintContextEx}
   ) -> (then, ReturnValue:Athena Match Stat View Object Reference)
   > Is Valid(InputObject=@n1)
   > SetStats!(
    execute
    self:Match Stats Grid Object Reference=Get StatsGrid() -> (StatsGrid:Match Stats Grid Object Reference)
    AthenaMatchStats:Athena Match Stat View Object Reference=@n1
   ) -> (then)
  ]}
 OnTeamStatsAvailable {nodes [
   OnTeamStatsAvailable() {is=title K2Node_FunctionEntry}
   n1 = Break Athena Match Team Stats!(AthenaMatchTeamStats:"Athena Match Team Stats Structure (by ref)"=Get TeamStats!(self:Athena Player Match Report Object Reference=Get MatchReport()) -> (TeamStats:Athena Match Team Stats Structure)) -> (Place:int, TotalPlayers:int) {is=title K2Node_BreakStruct}
   > "SetText (Text)"(
    self=Get PlacementText() -> (PlacementText:Common Text Object Reference)
    InText=Format Text!(
     Format="LOCTEXT(\"You came in {Place}{Place}|ordinal(one=st,two=nd,few=rd,other=th) out of {Players} Players\")"
     Place:int=@n1
     Players:int=@n1.TotalPlayers
    ) -> (Result)
   )
   > SetVisibility(self=Get PlacementLine() -> (PlacementLine:Horizontal Box Object Reference), InVisibility=Visible)
  ]}
 SetMatchReport {nodes [
   SetMatchReport!() -> (then, Report:Athena Player Match Report Object Reference) {is=title K2Node_FunctionEntry}
   > n1 = Is Valid(InputObject=Get MatchReport())
   > Unbind Event from On Stats Available(self:Athena Player Match Report Object Reference=Get MatchReport(), Delegate=Create Event()) {is=title K2Node_RemoveDelegate}
   > Unbind Event from On Rewards Available(self:Athena Player Match Report Object Reference=Get MatchReport(), Delegate=Create Event()) {is=title K2Node_RemoveDelegate}
   > n2 = Unbind Event from On Team Stats Available(self:Athena Player Match Report Object Reference=Get MatchReport(), Delegate=Create Event()) {is=title K2Node_RemoveDelegate}
   n3 = Set MatchReport(
    execute=[@n2.then, @n1.Is Not Valid]
    MatchReport:Athena Player Match Report Object Reference=Get Report() -> (Report:Athena Player Match Report Object Reference)
   ) -> (Output_Get:Athena Player Match Report Object Reference)
   > Is Valid(InputObject=@n3)
   > Bind Event to On Stats Available(self:Athena Player Match Report Object Reference=Get MatchReport(), Delegate=Create Event()) {is=title K2Node_AddDelegate}
   > Bind Event to On Rewards Available(self:Athena Player Match Report Object Reference=Get MatchReport(), Delegate=Create Event()) {is=title K2Node_AddDelegate}
   > Bind Event to On Team Stats Available(self:Athena Player Match Report Object Reference=Get MatchReport(), Delegate=Create Event()) {is=title K2Node_AddDelegate}
   > n4 = Sequence!(execute) -> (then_0, then_1, then_2:Exec, then_3:Exec)
   > n5 = Cast To FortPlayerControllerAthena(Object=GetOwningPlayer())
   > SetPlayerState!(
    execute
    self:Xp Match Progression Object Reference=Get XpMatchProgression()
    PlayerController:Fort Player Controller Athena Object Reference=@n5
    AthenaLevelInfo:Athena Level Info Structure=Get InitialLevelInfo!(self:Athena Player Match Report Object Reference=Get MatchReport()) -> (InitialLevelInfo:Athena Level Info Structure)
   ) -> (then)
   OnStatsAvailable(execute=@n4.then_1) {is=title K2Node_CallFunction}
   n6 = Branch(execute=@n4.then_2, Condition=Get bHasRewards!(self:Athena Player Match Report Object Reference=Get MatchReport()) -> (bHasRewards:bool))
   > OnRewardsAvailable() {is=title K2Node_CallFunction}
   SetActiveWidget(execute=@n6.else, self=Get RewardsSwitcher(), Widget=Get WaitingForMatchToFinish() -> (WaitingForMatchToFinish:Overlay Object Reference))
   Branch(execute=@n4.then_3, Condition=Get bHasTeamStats!(self:Athena Player Match Report Object Reference=Get MatchReport()) -> (bHasTeamStats:bool))
   > OnTeamStatsAvailable() {is=title K2Node_CallFunction}
  ]}
]
variables [MatchReport {type=AthenaPlayerMatchReport, default=null}]
widgets {
 info {inheritedWidgetCount=0, parentClass=&CommonUserWidget, rootWidgetClass=&SizeBox, widgetCount=17}
 widgets [
  SizeBox_0 {widgetClassPath=&SizeBox, widget=SizeBox_0, properties {bOverride_MinDesiredWidth=true, minDesiredWidth=250}}
  VerticalBox_0 {
   widgetClassPath=&VerticalBox
   parent=SizeBox_0
   widget=VerticalBox_0
   slot {class=SizeBoxSlot, properties {verticalAlignment=VAlign_Bottom}}
  }
  CommonTextBlock_0 {
   widgetClassPath=&CommonTextBlock
   parent=VerticalBox_0
   widget=CommonTextBlock_0
   properties {
    style=/Game/Athena/HUD/TextStyles/TextStyle-HUD-MapScreen_Title.TextStyle-HUD-MapScreen_Title_C
    text=Match Stats
    textTransformPolicy=ToUpper
    font {
     bForceMonospaced=false
     bMaterialIsStencil=false
     fontMaterial=null
     fontObject=&Burbank
     letterSpacing=0
     monospacedWidth=1
     size=18
     skewAmount=0
     typefaceFontName=BurbankRegular-Black
     outlineSettings {
      bApplyOutlineToDropShadows=true
      bMiteredCorners=false
      bSeparateFillAlpha=true
      outlineMaterial=&M-FontShadow
      outlineSize=3
      outlineColor=<rgba 0 0 0 0.5>
     }
    }
    renderTransform {angle=0, scale=<1 1>, shear=<-8 0>, translation=<0 0>}
    shadowOffset=<0 0>
   }
   slot {class=VerticalBoxSlot}
  }
  CommonBorder_0 {
   widgetClassPath=&CommonBorder
   parent=VerticalBox_0
   widget=CommonBorder_0
   properties {visibility=Collapsed, padding=<ltrb 0 0 0 0>}
   slot {class=VerticalBoxSlot, properties {padding=<ltrb 0 10 0 0>}}
  }
  PlacementLine {
   widgetClassPath=&HorizontalBox
   bIsVariable=true
   parent=CommonBorder_0
   widget=PlacementLine
   properties {visibility=Collapsed}
   slot {class=BorderSlot, properties {padding=<ltrb 0 0 0 0>}}
  }
  Image_0 {
   widgetClassPath=&Image
   parent=PlacementLine
   widget=Image_0
   properties {brush {
     bIsDynamicallyLoaded=false
     drawAs=Image
     imageType=NoImage
     mirroring=NoMirror
     resourceName=None
     resourceObject=/Game/UI/Fort_HUD/ChatLog_Icon_Star
     tiling=NoTile
     imageSize=<24 24>
     margin=<ltrb 0 0 0 0>
     outlineSettings {
      bUseBrushTransparency=false
      roundingType=HalfHeightRadius
      width=0
      color {colorUseRule=UseColor_Specified, specifiedColor=<rgba 0 0 0 0>}
      cornerRadii=<0 0 0 0>
     }
     tintColor {colorUseRule=UseColor_Specified, specifiedColor=<rgba 1 1 1 1>}
     uVRegion {bIsValid=false, max=<0 0>, min=<0 0>}
    }}
   slot {class=HorizontalBoxSlot, properties {horizontalAlignment=HAlign_Center, verticalAlignment=VAlign_Center, padding=<ltrb 0 2 0 0>}}
  }
  PlacementText {
   widgetClassPath=&CommonTextBlock
   bIsVariable=true
   parent=PlacementLine
   widget=PlacementText
   properties {
    style=/FortUILibrary/Styles/Typography/Foundation/TextStyle-Base-S.TextStyle-Base-S_C
    text=You Placed 1st out of 100 Players
    font {
     bForceMonospaced=false
     bMaterialIsStencil=false
     fontMaterial=null
     fontObject=&NotoSans
     letterSpacing=0
     monospacedWidth=1
     size=18
     skewAmount=0
     typefaceFontName=NotoSans-Regular
     outlineSettings {
      bApplyOutlineToDropShadows=false
      bMiteredCorners=false
      bSeparateFillAlpha=false
      outlineMaterial=null
      outlineSize=0
      outlineColor=<rgba 0 0 0 1>
     }
    }
    shadowOffset=<0 0>
   }
   slot {class=HorizontalBoxSlot}
  }
  Overlay_1 {widgetClassPath=&Overlay, parent=VerticalBox_0, widget=Overlay_1, slot {class=VerticalBoxSlot}}
  Image_FlatBG {
   widgetClassPath=&Image
   bIsVariable=true
   parent=Overlay_1
   widget=Image_FlatBG
   properties {colorAndOpacity=<rgba 0.0052 0.0052 0.0056 0>}
   slot {class=OverlaySlot, properties {horizontalAlignment=HAlign_Fill, verticalAlignment=VAlign_Fill}}
  }
  StatsGrid {
   widgetClassPath=&/MatchStatsGrid
   bIsVariable=true
   parent=Overlay_1
   widget=StatsGrid
   slot {class=OverlaySlot, properties {padding=<ltrb 10 20 10 20>}}
  }
  RewardsSwitcher {
   widgetClassPath=&WidgetSwitcher
   bIsVariable=true
   parent=VerticalBox_0
   widget=RewardsSwitcher
   properties {visibility=Collapsed}
   slot {class=VerticalBoxSlot, properties {padding=<ltrb 0 30 0 0>}}
  }
  XpMatchProgression {
   widgetClassPath=&/XpMatchProgression
   bIsVariable=true
   parent=RewardsSwitcher
   widget=XpMatchProgression
   properties {visibility=Collapsed}
   slot {class=WidgetSwitcherSlot, properties {verticalAlignment=VAlign_Top}}
  }
  WaitingForMatchToFinish {
   widgetClassPath=&Overlay
   bIsVariable=true
   parent=RewardsSwitcher
   widget=WaitingForMatchToFinish
   slot {class=WidgetSwitcherSlot}
  }
  HorizontalBox_0 {
   widgetClassPath=&HorizontalBox
   parent=WaitingForMatchToFinish
   widget=HorizontalBox_0
   slot {class=OverlaySlot, properties {horizontalAlignment=HAlign_Center}}
  }
  SizeBox_5069 {
   widgetClassPath=&SizeBox
   parent=HorizontalBox_0
   widget=SizeBox_5069
   properties {bOverride_HeightOverride=true, bOverride_WidthOverride=true, heightOverride=100, widthOverride=100}
   slot {class=HorizontalBoxSlot, properties {horizontalAlignment=HAlign_Center, verticalAlignment=VAlign_Center}}
  }
  ProgressSpinner {
   widgetClassPath=&Image
   bIsVariable=true
   parent=SizeBox_5069
   widget=ProgressSpinner
   properties {brush {
     bIsDynamicallyLoaded=false
     drawAs=Image
     imageType=NoImage
     mirroring=NoMirror
     resourceName=None
     resourceObject=/FortUILibrary/UIKit/Throbber/MI_UIKit_Monochrome_Throbber
     tiling=NoTile
     imageSize=<64 64>
     margin=<ltrb 0 0 0 0>
     outlineSettings {
      bUseBrushTransparency=false
      roundingType=HalfHeightRadius
      width=0
      color {colorUseRule=UseColor_Specified, specifiedColor=<rgba 0 0 0 0>}
      cornerRadii=<0 0 0 0>
     }
     tintColor {colorUseRule=UseColor_Specified, specifiedColor=<rgba 1 1 1 1>}
     uVRegion {bIsValid=false, max=<0 0>, min=<0 0>}
    }}
   slot {class=SizeBoxSlot, properties {horizontalAlignment=HAlign_Center, verticalAlignment=VAlign_Center}}
  }
  CommonTextBlock_1 {
   widgetClassPath=&CommonTextBlock
   parent=HorizontalBox_0
   widget=CommonTextBlock_1
   properties {
    style=/FortUILibrary/Styles/Typography/Foundation/TextStyle-Base-M-B.TextStyle-Base-M-B_C
    text=Team Still Alive
    font {
     bForceMonospaced=false
     bMaterialIsStencil=false
     fontMaterial=null
     fontObject=&NotoSans
     letterSpacing=0
     monospacedWidth=1
     size=24
     skewAmount=0
     typefaceFontName=NotoSans-Bold
     outlineSettings {
      bApplyOutlineToDropShadows=false
      bMiteredCorners=false
      bSeparateFillAlpha=false
      outlineMaterial=null
      outlineSize=0
      outlineColor=<rgba 0 0 0 1>
     }
    }
    shadowOffset=<0 0>
   }
   slot {class=HorizontalBoxSlot, properties {verticalAlignment=VAlign_Center}}
  }
 ]
}
