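# Translucent UI material in the Athena HUD that computes a texture alpha mask faded toward the sides.
# A DrawLine-2D function draws a soft vertical band down the horizontal center, tuned by Width and Falloff scalar parameters.
# That band multiplies the alpha of the Texture parameter, so the mask weakens away from the center without needing a separate gradient texture.
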Material /Game/Athena/HUD/Materials/M_UI_SideFade
graph {
 properties {blendMode=BLEND_Translucent, materialDomain=MD_UI, pixelDepthOffsetMode=PDOM_Legacy, usedShadingModels=Default Lit}
 nodes [Multiply(
   A=TextureSampleParameter2D() {parameterName=Texture}.A
   B=MaterialFunctionCall!(
    Coordinates
    P0=Constant2Vector() {r=0.5}
    P1=Constant2Vector() {g=1, r=0.5}
    Line Width=ScalarParameter() {defaultValue=0.42, parameterName=Width}
    Edge Falloff=ScalarParameter() {defaultValue=0.1789, parameterName=Falloff}
    Square Corners?
   ) {materialFunction=/Engine/Functions/Engine_MaterialFunctions01/Coordinates/DrawLine-2D}.Result
  )]
}
