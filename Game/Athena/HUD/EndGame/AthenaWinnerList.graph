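# End-game HUD widget (child of AthenaWinnersList) that shows a pluralized winner-count title and the winners' names in up to five columns.
# OnWinnersAnnounced stores the names, reveals the widget and starts a retriggerable delay that collapses it after OnScreenDuration.
# UpdateWinnerList appends names four per column via index/4, unhiding each column's parent SizeBox; its <=20 check lets index 20 overrun the columns.
# A side-fade scrolling ticker layout exists but the graph never drives it.
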WidgetBlueprint /Game/Athena/HUD/EndGame/AthenaWinnerList
graphs [
 EventGraph {nodes [
   UpdateWinnerList() {is=title K2Node_CustomEvent}
   > ClearList()
   > n1 = For Each Loop(Array=Get LocalWinnerList() -> (LocalWinnerList:Array of Strings))
   > Branch(Condition="integer <= integer"!(A:int=@n1.Array Index, B:int=20) -> (ReturnValue:bool) {is=title K2Node_CallFunction})
   n2 = "Get (a copy)"(
    Array:Array of Common Text Object References=Get WinnerListTextBoxes() -> (WinnerListTextBoxes:Array of Common Text Object References)
    Dimension 1=int / int!(A:int=@n1.Array Index, B:int=4) -> (ReturnValue:int=0) {is=title K2Node_CallFunction}
   ) -> (Output:Common Text Object Reference)
   > "SetText (Text)"(
    self=@n2
    InText=Format Text!(
     Format="LOCTEXT(\"{0}\\r\\n{1}\")"
     0:Text="GetText (Text)"!(self:Text Object Reference=@n2) -> (ReturnValue:Text)
     1:Text="To Text (String)"(InString=@n1)
    ) -> (Result)
   )
   > n3 = FindParentWidgetOfType!(execute, StartingWidget:Widget Object Reference=@n2, Type:Widget Class Reference=&SizeBox) -> (then, ReturnValue:Size Box Object Reference)
   > SetVisibility(self=@n3, InVisibility=HitTestInvisible)
   n4 = TimeoutDuration!() -> (OutputDelegate, then, Duration:double=~) {is=title K2Node_CustomEvent}
   > Branch(Condition="float > float"(A=@n4.Duration, B=0) {is=title K2Node_CallFunction})
   > RetriggerableDelay(Duration=@n4.Duration)
   n5 = Get OnScreenDuration() -> (OnScreenDuration:double)
   > SetVisibility(InVisibility=Select(
     Option 0=Collapsed
     Option 1=HitTestInvisible
     Index="Equal (Float)"!(A:double=@n5, B:double=0) -> (ReturnValue:bool) {is=title K2Node_CallFunction}
    ))
   ClearList() {is=title K2Node_CustomEvent}
   > n6 = For Each Loop(Array:Array of Common Text Object References=Get WinnerListTextBoxes()) -> (Array Element:Common Text Object Reference)
   > "SetText (Text)"(self=@n6)
   Event Construct()
   > Set WinnerListTextBoxes(WinnerListTextBoxes:Array of Common Text Object References=Make Array!(
     "[0]":Common Text Object Reference=Get RichTextBlock_VerticalWinners() -> (RichTextBlock_VerticalWinners:Common Text Object Reference)
     "[1]":Common Text Object Reference=Get RichTextBlock_VerticalWinners_column2() -> (RichTextBlock_VerticalWinners_column2:Common Text Object Reference)
     "[2]":Common Text Object Reference=Get RichTextBlock_VerticalWinners_column3() -> (RichTextBlock_VerticalWinners_column3:Common Text Object Reference)
     "[3]":Common Text Object Reference=Get RichTextBlock_VerticalWinners_column4() -> (RichTextBlock_VerticalWinners_column4:Common Text Object Reference)
     "[4]":Common Text Object Reference=Get RichTextBlock_VerticalWinners_column5() -> (RichTextBlock_VerticalWinners_column5:Common Text Object Reference)
    ) -> (Array:Array of Common Text Object References)) -> (Output_Get:Array of Common Text Object References)
   n7 = Event OnWinnersAnnounced!() -> (OutputDelegate, then, WinnerInfo:"Athena Winner Info Structure (by ref)") {is=title K2Node_Event}
   > n8 = Set OnScreenDuration(OnScreenDuration:double=10) -> (Output_Get:double)
   > TimeoutDuration!(execute, self, Duration:double=@n8) -> (then) {is=title K2Node_CallFunction}
   n9 = Break Athena Winner Info!(AthenaWinnerInfo:"Athena Winner Info Structure (by ref)"=@n7.WinnerInfo) -> (BigNameWinnerName:String, WinnerNames:Array of Strings) {is=title K2Node_BreakStruct}
   > Branch(Condition="integer > integer"(A=Length(TargetArray:Array of Strings=@n9.WinnerNames), B=0) {is=title K2Node_CallFunction})
   > SetVisibility(InVisibility=HitTestInvisible)
   > Set LocalWinnerList(LocalWinnerList:Array of Strings=@n9.WinnerNames) -> (Output_Get:Array of Strings)
   > UpdateWinnerList() {is=title K2Node_CallFunction}
   n10 = Get LocalWinnerList()
   > "SetText (Text)"(
    self=Get CommonTextBlock_WinnersTitle() -> (CommonTextBlock_WinnersTitle:Common Text Object Reference)
    InText=Format Text!(
     Format=Select(
      Option 0:Text="LOCTEXT(\"{0} {0}|plural(one=WINNER,other=WINNERS)\")"
      Option 1:Text="LOCTEXT(\"{0} OTHER {0}|plural(one=WINNER,other=WINNERS)\")"
      Index=false
     ) -> (ReturnValue:Text)
     0:int=Length(TargetArray=@n10)
    ) -> (Result)
   )
   Get isWinner() -> (isWinner:bool)
  ]}
 OnRep_isWinner {nodes [OnRep_isWinner() {is=title K2Node_FunctionEntry}]}
]
variables [
 isWinner {type=boolean}
 LocalWinnerList {
  type="array<string>"
  default [Lorem1,Ipsum2,Dolor3,Amit4,Sequit5,DolorAmit6,SequitIpsum7,Steve8,Lorem9,Lorem10,Lorem11,Lorem12,Lorem13,Lorem14]
 }
 OnScreenDuration {type=number, default=10}
 WinnerListTextBoxes {type="array<CommonTextBlock>", default []}
]
widgets {
 info {inheritedWidgetCount=0, parentClass=/Script/FortniteUI.AthenaWinnersList, rootWidgetClass=&Overlay, widgetCount=18}
 widgets [
  Overlay_0 {widgetClassPath=&Overlay, widget=Overlay_0}
  VerticalBoxWinnerlist {
   widgetClassPath=&VerticalBox
   parent=Overlay_0
   widget=VerticalBoxWinnerlist
   properties {visibility=HitTestInvisible}
   slot {class=OverlaySlot, properties {horizontalAlignment=HAlign_Center}}
  }
  CommonTextBlock_WinnersTitle {
   widgetClassPath=&CommonTextBlock
   bIsVariable=true
   parent=VerticalBoxWinnerlist
   widget=CommonTextBlock_WinnersTitle
   properties {
    text=10 WINNERS:
    font {
     bForceMonospaced=false
     bMaterialIsStencil=false
     fontMaterial=null
     fontObject=&Burbank
     letterSpacing=0
     monospacedWidth=1
     size=34
     skewAmount=0
     typefaceFontName=Burbank-Black
     outlineSettings {
      bApplyOutlineToDropShadows=false
      bMiteredCorners=false
      bSeparateFillAlpha=false
      outlineMaterial=null
      outlineSize=1
      outlineColor=<rgba 0.0052 0.0052 0.0056 0.149>
     }
    }
    shadowColorAndOpacity=<rgba 0.0052 0.0052 0.0056 0.8235>
    shadowOffset=<3 3>
   }
   slot {class=VerticalBoxSlot, properties {horizontalAlignment=HAlign_Center}}
  }
  CommonWidgetSwitcher_WinnerListStyle {
   widgetClassPath=&CommonWidgetSwitcherLegacy
   bIsVariable=true
   parent=VerticalBoxWinnerlist
   widget=CommonWidgetSwitcher_WinnerListStyle
   slot {class=VerticalBoxSlot, properties {verticalAlignment=VAlign_Top}}
  }
  HorizontalBox_1 {
   widgetClassPath=&HorizontalBox
   parent=CommonWidgetSwitcher_WinnerListStyle
   widget=HorizontalBox_1
   slot {class=WidgetSwitcherSlot, properties {horizontalAlignment=HAlign_Center, verticalAlignment=VAlign_Top, padding=<ltrb 0 0 0 -220>}}
  }
  SizeBox_1 {
   widgetClassPath=&SizeBox
   parent=HorizontalBox_1
   widget=SizeBox_1
   properties {bOverride_MinDesiredWidth=true, minDesiredWidth=200, visibility=Collapsed}
   slot {class=HorizontalBoxSlot, properties {padding=<ltrb 5 0 5 0>}}
  }
  RichTextBlock_VerticalWinners {
   widgetClassPath=&CommonTextBlock
   bIsVariable=true
   parent=SizeBox_1
   widget=RichTextBlock_VerticalWinners
   properties {
    justification=Center
    text="LoremDolor\r\nDolorIpsum\r\nDolorLorem\r\nDolorIpsum \r\nLoremDolor"
    font {
     bForceMonospaced=false
     bMaterialIsStencil=false
     fontMaterial=null
     fontObject=&BurbankSmall
     letterSpacing=0
     monospacedWidth=1
     size=24
     skewAmount=0
     typefaceFontName=Bold
     outlineSettings {
      bApplyOutlineToDropShadows=false
      bMiteredCorners=false
      bSeparateFillAlpha=false
      outlineMaterial=null
      outlineSize=1
      outlineColor=<rgba 0 0 0 1>
     }
    }
    shadowOffset=<0 0>
   }
   slot {class=SizeBoxSlot}
  }
  SizeBox_2 {
   widgetClassPath=&SizeBox
   parent=HorizontalBox_1
   widget=SizeBox_2
   properties {bOverride_MinDesiredWidth=true, minDesiredWidth=200, visibility=Collapsed}
   slot {class=HorizontalBoxSlot, properties {padding=<ltrb 5 0 5 0>}}
  }
  RichTextBlock_VerticalWinners_column2 {
   widgetClassPath=&CommonTextBlock
   bIsVariable=true
   parent=SizeBox_2
   widget=RichTextBlock_VerticalWinners_column2
   properties {
    justification=Center
    text="LoremDolor\r\nDolorIpsum\r\nDolorLorem\r\nDolorIpsum \r\nLoremDolor"
    font {
     bForceMonospaced=false
     bMaterialIsStencil=false
     fontMaterial=null
     fontObject=&BurbankSmall
     letterSpacing=0
     monospacedWidth=1
     size=24
     skewAmount=0
     typefaceFontName=Bold
     outlineSettings {
      bApplyOutlineToDropShadows=false
      bMiteredCorners=false
      bSeparateFillAlpha=false
      outlineMaterial=null
      outlineSize=1
      outlineColor=<rgba 0 0 0 1>
     }
    }
    shadowOffset=<0 0>
   }
   slot {class=SizeBoxSlot}
  }
  SizeBox_3 {
   widgetClassPath=&SizeBox
   parent=HorizontalBox_1
   widget=SizeBox_3
   properties {bOverride_MinDesiredWidth=true, minDesiredWidth=200, visibility=Collapsed}
   slot {class=HorizontalBoxSlot, properties {padding=<ltrb 5 0 5 0>}}
  }
  RichTextBlock_VerticalWinners_column3 {
   widgetClassPath=&CommonTextBlock
   bIsVariable=true
   parent=SizeBox_3
   widget=RichTextBlock_VerticalWinners_column3
   properties {
    justification=Center
    text="LoremDolor\r\nDolorIpsum\r\nDolorLorem\r\nDolorIpsum \r\nLoremDolor"
    font {
     bForceMonospaced=false
     bMaterialIsStencil=false
     fontMaterial=null
     fontObject=&BurbankSmall
     letterSpacing=0
     monospacedWidth=1
     size=24
     skewAmount=0
     typefaceFontName=Bold
     outlineSettings {
      bApplyOutlineToDropShadows=false
      bMiteredCorners=false
      bSeparateFillAlpha=false
      outlineMaterial=null
      outlineSize=1
      outlineColor=<rgba 0 0 0 1>
     }
    }
    shadowOffset=<0 0>
   }
   slot {class=SizeBoxSlot}
  }
  SizeBox_4 {
   widgetClassPath=&SizeBox
   parent=HorizontalBox_1
   widget=SizeBox_4
   properties {bOverride_MinDesiredWidth=true, minDesiredWidth=200, visibility=Collapsed}
   slot {class=HorizontalBoxSlot, properties {padding=<ltrb 5 0 5 0>}}
  }
  RichTextBlock_VerticalWinners_column4 {
   widgetClassPath=&CommonTextBlock
   bIsVariable=true
   parent=SizeBox_4
   widget=RichTextBlock_VerticalWinners_column4
   properties {
    justification=Center
    text="LoremDolor\r\nDolorIpsum\r\nDolorLorem\r\nDolorIpsum \r\nLoremDolor"
    font {
     bForceMonospaced=false
     bMaterialIsStencil=false
     fontMaterial=null
     fontObject=&BurbankSmall
     letterSpacing=0
     monospacedWidth=1
     size=24
     skewAmount=0
     typefaceFontName=Bold
     outlineSettings {
      bApplyOutlineToDropShadows=false
      bMiteredCorners=false
      bSeparateFillAlpha=false
      outlineMaterial=null
      outlineSize=1
      outlineColor=<rgba 0 0 0 1>
     }
    }
    shadowOffset=<0 0>
   }
   slot {class=SizeBoxSlot}
  }
  SizeBox_5 {
   widgetClassPath=&SizeBox
   parent=HorizontalBox_1
   widget=SizeBox_5
   properties {bOverride_MinDesiredWidth=true, minDesiredWidth=200, visibility=Collapsed}
   slot {class=HorizontalBoxSlot, properties {padding=<ltrb 5 0 0 0>}}
  }
  RichTextBlock_VerticalWinners_column5 {
   widgetClassPath=&CommonTextBlock
   bIsVariable=true
   parent=SizeBox_5
   widget=RichTextBlock_VerticalWinners_column5
   properties {
    justification=Center
    text="LoremDolor\r\nDolorIpsum\r\nDolorLorem\r\nDolorIpsum \r\nLoremDolor"
    font {
     bForceMonospaced=false
     bMaterialIsStencil=false
     fontMaterial=null
     fontObject=&BurbankSmall
     letterSpacing=0
     monospacedWidth=1
     size=24
     skewAmount=0
     typefaceFontName=Bold
     outlineSettings {
      bApplyOutlineToDropShadows=false
      bMiteredCorners=false
      bSeparateFillAlpha=false
      outlineMaterial=null
      outlineSize=1
      outlineColor=<rgba 0 0 0 1>
     }
    }
    shadowOffset=<0 0>
   }
   slot {class=SizeBoxSlot}
  }
  RetainerBox_Winnerlist {
   widgetClassPath=/Script/UMG.RetainerBox
   bIsVariable=true
   parent=CommonWidgetSwitcher_WinnerListStyle
   widget=RetainerBox_Winnerlist
   properties {effectMaterial=/Game/Athena/HUD/Materials/M_UI_SideFade}
   slot {class=WidgetSwitcherSlot}
  }
  SizeBox_0 {
   widgetClassPath=&SizeBox
   parent=RetainerBox_Winnerlist
   widget=SizeBox_0
   properties {bOverride_MaxDesiredWidth=true, maxDesiredWidth=954.911}
   slot {class=PanelSlot}
  }
  CommonTextBlock_0 {
   widgetClassPath=&CommonTextBlock
   parent=SizeBox_0
   widget=CommonTextBlock_0
   properties {
    scrollStyle=&TextScrollStyle-Base_C
    text="                    LoremDolor (4 elims), DolorIpsum (4 elims), DolorLorem (0 elims), LoremDolor (4 elims), DolorIpsum (4 elims), DolorLorem (0 elims)                    "
    font {
     bForceMonospaced=false
     bMaterialIsStencil=false
     fontMaterial=null
     fontObject=&NotoSans
     letterSpacing=0
     monospacedWidth=1
     size=24
     skewAmount=0
     typefaceFontName=NotoSans-Bold
     outlineSettings {
      bApplyOutlineToDropShadows=false
      bMiteredCorners=false
      bSeparateFillAlpha=false
      outlineMaterial=null
      outlineSize=1
      outlineColor=<rgba 0.0052 0.0052 0.0056 0.149>
     }
    }
    shadowColorAndOpacity=<rgba 0.0052 0.0052 0.0056 0.8235>
    shadowOffset=<3 3>
   }
   slot {class=SizeBoxSlot}
  }
 ]
}
