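# Killer info widget for the end-game screen, showing who eliminated the player with name, ID, health and shield, kills and weapon.
# Self-eliminations switch to a message mapped from death cause tags.
# It registers as host for three gameplay-tag slots (status, name addition, weapon info) and forwards death reports to slotted widgets implementing the extension interface.
# Health info hides for configured killer pawn tags, and status insertion adjusts layout for tactical overshield.
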WidgetBlueprint /Game/Athena/HUD/EndGame/KillerScreen/KillerInfoWidget
graphs [
 AddContentToHostWidget {nodes [
   n1 = AddContentToHostWidget!() -> (then, InputPin:"Gameplay Tag Structure (by ref)", Content:Widget Object Reference, Padding:Margin Structure, Fill:bool=~) {is=title K2Node_FunctionEntry}
   > n2 = Sequence!(execute) -> (then_0, then_1, then_2:Exec)
   > Branch(Condition="Equal (GameplayTag)"(A=@n1, B=Get LastKillerStatusSlotTag() -> (LastKillerStatusSlotTag:Gameplay Tag Structure)))
   > n3 = AddChildToOverlay!(
    execute
    self:Overlay Object Reference=Get Overlay_Status() -> (Overlay_Status:Overlay Object Reference)
    Content:Widget Object Reference=Get Content() -> (Content:Widget Object Reference)
   ) -> (then, ReturnValue:Overlay Slot Object Reference)
   > Set OverlaySlot(OverlaySlot:Overlay Slot Object Reference=@n3) -> (Output_Get:Overlay Slot Object Reference)
   > n4 = IsTacticalOvershieldEnabled!(execute) -> (then, ReturnValue:bool)
   > Branch(Condition=@n4)
   > SetWidthOverride(self=Get SizeBox_HealthAndShield() -> (SizeBox_HealthAndShield:Size Box Object Reference), InWidthOverride=550)
   > SetPadding!(
    execute
    self:Overlay Slot Object Reference=Get OverlaySlot() -> (OverlaySlot:Overlay Slot Object Reference)
    InPadding:Margin Structure=Make Margin(Left=-25, Top=0, Right=0, Bottom=0)
   ) -> (then)
   Branch(
    execute=@n2.then_2
    Condition="Equal (GameplayTag)"(A=@n1, B=Get LastKillerNameAdditionSlotTag() -> (LastKillerNameAdditionSlotTag:Gameplay Tag Structure))
   )
   > AddChildToHorizontalBox!(
    execute
    self:Horizontal Box Object Reference=Get HBox_KillerName() -> (HBox_KillerName:Horizontal Box Object Reference)
    Content:Widget Object Reference=Get Content()
   ) -> (then, ReturnValue:Horizontal Box Slot Object Reference)
   Branch(
    execute=@n2.then_1
    Condition="Equal (GameplayTag)"(A=@n1, B=Get LastKillerWeaponInfoSlotTag() -> (LastKillerWeaponInfoSlotTag:Gameplay Tag Structure))
   )
   > AddChildToVerticalBox(self=Get VBox_KillerWeaponHolder() -> (VBox_KillerWeaponHolder:Vertical Box Object Reference), Content=Get Content())
  ]}
 DeathCauseToText_0 {nodes [
   n1 = DeathCauseToText_0!() -> (then, DeathCause:EDeathCause Enum=OutsideSafeZone) {is=title K2Node_FunctionEntry}
   > Return Node!(
    execute
    ReturnValue:Text=Select!(
     OutsideSafeZone:Text="LOCTEXT(\"ELIMINATED BY THE STORM\")"
     FallDamage:Text="LOCTEXT(\"YOU DIDN\\'T STICK THE LANDING\")"
     Pistol:Text
     Shotgun:Text
     Rifle:Text
     SMG:Text
     Sniper:Text
     SniperNoScope:Text
     Melee:Text
     InfinityBlade:Text
     Grenade:Text
     C4:Text
     GrenadeLauncher:Text
     RocketLauncher:Text
     Minigun:Text
     Bow:Text
     Trap:Text
     DBNOTimeout:Text
     Banhammer:Text
     RemovedFromGame:Text
     MassiveMelee:Text
     MassiveDiveBomb:Text
     MassiveRanged:Text
     Vehicle:Text
     ShoppingCart:Text
     ATK:Text
     QuadCrasher:Text
     Biplane:Text
     BiplaneGun:Text
     LMG:Text
     GasGrenade:Text
     InstantEnvironmental:Text
     InstantEnvironmentalFellOutOfWorld:Text
     InstantEnvironmentalUnderLandscape:Text
     Turret:Text
     ShipCannon:Text
     Cube:Text
     Balloon:Text
     StormSurge:Text
     Lava:Text
     BasicFiend:Text="LOCTEXT(\"ELIMINATED BY A CUBE FIEND\")"
     EliteFiend:Text="LOCTEXT(\"ELIMINATED BY AN ELITE CUBE FIEND\")"
     RangedFiend:Text
     BasicBrute:Text="LOCTEXT(\"ELIMINATED BY A CUBE BRUTE\")"
     EliteBrute:Text="LOCTEXT(\"ELIMINATED BY AN ELITE CUBE BRUTE\")"
     MegaBrute:Text="LOCTEXT(\"ELIMINATED BY A MEGA CUBE BRUTE\")"
     SilentSwitchingToSpectate:Text="LOCTEXT(\"SPECTATING\")"
     LoggedOut:Text
     TeamSwitchSuicide:Text
     WonMatch:Text
     Unspecified:Text
     Index:EDeathCause Enum=@n1
    ) -> (ReturnValue:Text)
   )
  ]}
 EventGraph {nodes [
   n1 = Event SetKillerHealthAndShield!() -> (OutputDelegate, then, HealthPercent:float, ShieldPercent:float, OvershieldPercent:float) {is=title K2Node_Event}
   > Set Health And Shield!(
    execute
    self:Killer Health And Shield Object Reference=Get WBP_KillerHealthAndShield() -> (WBP_KillerHealthAndShield:Killer Health And Shield Object Reference)
    Health PCT:double=@n1.HealthPercent
    Shield PCT:double=@n1.ShieldPercent
   ) -> (then) {is=title K2Node_CallFunction}
   n2 = Event SetNumOfKillerKills!() -> (OutputDelegate, then, NumOfKillers:int=0) {is=title K2Node_Event}
   > Set Num Of Kills!(
    execute
    self:Killer Kills Widget Object Reference=Get WBP_KillerKillsWidget() -> (WBP_KillerKillsWidget:Killer Kills Widget Object Reference)
    Num Of Kills:int=@n2.NumOfKillers
   ) -> (then) {is=title K2Node_CallFunction}
   n3 = Event OnSetDeathReport!() -> (OutputDelegate, then, InDeathReport:"Fort Player Death Report Structure (by ref)") {is=title K2Node_Event}
   > n4 = For Each Loop(Array:Array of Widget Object References=Get SlottedWidgetChildren() -> (SlottedWidgetChildren:Array of Widget Object References)) -> (Array Element:Widget Object Reference)
   > Branch(Condition=Does Object Implement Interface!(TestObject:Object Reference=@n4, Interface:Interface Interface=/Script/FortniteUI.LastKillerInfoExtensionWidgetInterface) -> (ReturnValue:bool))
   > SetDeathReport!(execute:Exec, self:Object Reference=@n4, InDeathReport:"Fort Player Death Report Structure (by ref)"=@n3.InDeathReport) -> (then:Exec) {is=title K2Node_Message}
   Event Construct()
   n5 = Get LastKillerNameAdditionSlotTag()
   > RegisterWidgetHost!(execute, WidgetHost:Widget Object Reference, WidgetHostTag:"Gameplay Tag Structure (by ref)"=@n5, IsPlayerSpecific:bool=true) -> (then)
   n6 = Get LastKillerStatusSlotTag()
   > RegisterWidgetHost(WidgetHostTag=@n6, IsPlayerSpecific=true)
   n7 = Get LastKillerWeaponInfoSlotTag()
   > RegisterWidgetHost(WidgetHostTag=@n7, IsPlayerSpecific=true)
   n8 = Event SetKillerName!() -> (OutputDelegate, then, PlayerName:String, PlayerID:String) {is=title K2Node_Event}
   > "SetText (Text)"(self=Get Text_KillerName() -> (Text_KillerName:Common Text Object Reference), InText="To Text (String)"(InString=@n8.PlayerName))
   > "SetText (Text)"(self=Get Text_UserID() -> (Text_UserID:Text Object Reference), InText="To Text (String)"(InString=@n8.PlayerID))
   > SetVisibility(self=Get VBox_KillerName() -> (VBox_KillerName:Vertical Box Object Reference), InVisibility=HitTestInvisible)
   n9 = Event OnKillerPlayerStateSet!() -> (OutputDelegate, then, InKillerPlayerState:Fort Player State Object Reference) {is=title K2Node_Event}
   n10 = Get TagsToHideHealthInfoFor() -> (TagsToHideHealthInfoFor:Array of Gameplay Tag Structures)
   > Branch(Condition=HasAnyMatchingGameplayTags!(
     self:Gameplay Tag Asset Interface Interface=GetCurrentPawn!(self:Fort Player State Object Reference=@n9.InKillerPlayerState) -> (ReturnValue:Fort Player Pawn Object Reference)
     TagContainer:"Gameplay Tag Container Structure (by ref)"=MakeGameplayTagContainerFromArray!(GameplayTags:Array of Gameplay Tag Structures=@n10) -> (ReturnValue:Gameplay Tag Container Structure)
    ) -> (ReturnValue:bool))
   > SetVisibility(self=Get SizeBox_HealthAndShield(), InVisibility=Hidden)
   n11 = Event SetKillerWeaponIconNameAndRarity!() -> (OutputDelegate, then, IconTexture:Texture 2D Object Reference, ItemName:"Text (by ref)", ItemRarity:EFortRarity Enum=Common) {is=title K2Node_Event}
   > Set Icon And Name!(
    execute
    self:Killer Info Weapon Widget Object Reference=Get WBP_KillerInfoWeaponWidget() -> (WBP_KillerInfoWeaponWidget:Killer Info Weapon Widget Object Reference)
    Weapon Icon:Texture 2D Object Reference=@n11.IconTexture
    Weapon Name:Text=@n11.ItemName
    Weapon Rarity:EFortRarity Enum=@n11.ItemRarity
   ) -> (then) {is=title K2Node_CallFunction}
   n12 = Event OnSelfElimination!() -> (OutputDelegate, then, InDeathReport:"Fort Player Death Report Structure (by ref)") {is=title K2Node_Event}
   > n13 = ToDeathCause!(
    execute
    InTags:"Gameplay Tag Container Structure (by ref)"=Break Fort Player Death Report!(FortPlayerDeathReport:"Fort Player Death Report Structure (by ref)"=@n12.InDeathReport) -> (
     ServerTimeForRespawn:float=~
     ServerTimeForResurrect:float=~
     LethalDamage:float=~
     KillerPlayerState:Fort Player State Object Reference
     KillerPawn:Fort Pawn Weak object
     KillerHealthPercent:float=~
     KillerShieldPercent:float=~
     KillerOvershieldPercent:float=~
     KillerWeapon:Fort Item Entry Structure
     DamageCauser:Actor Weak object
     bDroppedBackpack:bool=~
     bNotifyUI:bool=~
     Tags:Gameplay Tag Container Structure
     ViewLocationAtTimeOfDeath:Vector=~
     ViewRotationAtTimeOfDeath:Rotator=~
    ) {is=title K2Node_BreakStruct}.Tags
    bWasDBNO:bool=false
   ) -> (then, ReturnValue:EDeathCause Enum=OutsideSafeZone)
   n14 = DeathCauseToText_0!(DeathCause:EDeathCause Enum=@n13) -> (ReturnValue:Text) {is=title K2Node_CallFunction}
   > n15 = Branch(Condition=TextIsEmpty(InText=@n14))
   SetActiveWidgetIndex(
    execute=[@n15.then, @n16.then]
    self:Common Visibility Switcher Object Reference=Get Switcher_EliminationSource() -> (Switcher_EliminationSource:Common Visibility Switcher Object Reference)
    Index=1
   )
   n16 = Set Text(execute=@n15.else, Text=@n14, self=Get Text_EliminatedSelf() -> (Text_EliminatedSelf:Common Text Object Reference))
   n17 = Event SetEliminationText!() -> (OutputDelegate, then, EliminationText:String) {is=title K2Node_Event}
   > Set Text(Text="To Text (String)"(InString=@n17.EliminationText), self=Get Text_EliminatedBy() -> (Text_EliminatedBy:Common Text Object Reference))
   Event Destruct()
   > UnregisterWidgetHost!(execute, WidgetHost:Widget Object Reference, WidgetHostTag:"Gameplay Tag Structure (by ref)"=@n5) -> (then)
   > UnregisterWidgetHost(WidgetHostTag=@n6)
   > UnregisterWidgetHost(WidgetHostTag=@n7)
   > Clear(TargetArray:Array of Widget Object References=Get SlottedWidgetChildren())
   n18 = Event Add Content To Host Widget!() -> (
    OutputDelegate
    then
    WidgetHostTag:"Gameplay Tag Structure (by ref)"
    ContentWidget:Widget Object Reference
    Padding:"Margin Structure (by ref)"
    bFill:bool
   ) {is=title K2Node_Event}
   > AddContentToHostWidget!(
    execute
    self
    InputPin:"Gameplay Tag Structure (by ref)"=@n18.WidgetHostTag
    Content:Widget Object Reference=@n18.ContentWidget
    Padding:Margin Structure=@n18.Padding
    Fill:bool=@n18.bFill
   ) -> (then) {is=title K2Node_CallFunction}
   > SetSlottedWidgetDeathReport!(execute, self, SlottedWidget:Widget Object Reference=@n18.ContentWidget) -> (then)
   > Branch(Condition=Does Object Implement Interface(TestObject=@n18.ContentWidget, Interface=&LastKillerInfoExtensionWidgetInterface))
   > Add(TargetArray:Array of Widget Object References=Get SlottedWidgetChildren(), NewItem:"Widget Object Reference (by ref)"=@n18.ContentWidget) {is=title K2Node_CallArrayFunction}
  ]}
 RemoveContentFromWidgetHost {nodes [
   n1 = Remove Content From Host Widget!() -> (
    then
    WidgetHostTag:"Gameplay Tag Structure (by ref)"
    ContentWidget:Widget Object Reference
    RemovalType:EWidgetSlottingRemovalType Enum=Unparent
   ) {is=title K2Node_FunctionEntry}
   > n2 = Sequence()
   > Branch(Condition="Equal (GameplayTag)"(A=@n1, B=Get LastKillerStatusSlotTag()))
   > n3 = RemoveChild(self=Get Overlay_Status(), Content=Get ContentWidget() -> (ContentWidget:Widget Object Reference))
   Return Node(execute=[@n3.then, @n4.then, @n5.then], ReturnValue:bool=false)
   Branch(execute=@n2.then_2, Condition="Equal (GameplayTag)"(A=@n1, B=Get LastKillerNameAdditionSlotTag()))
   > n4 = RemoveChild(self=Get HBox_KillerName(), Content=Get ContentWidget())
   Branch(execute=@n2.then_1, Condition="Equal (GameplayTag)"(A=@n1, B=Get LastKillerWeaponInfoSlotTag()))
   > n5 = RemoveChild(self=Get VBox_KillerWeaponHolder(), Content=Get ContentWidget())
  ]}
 SetSlottedWidgetDeathReport {nodes [
   n1 = SetSlottedWidgetDeathReport!() -> (then, SlottedWidget:Widget Object Reference) {is=title K2Node_FunctionEntry}
   > Branch(Condition=Does Object Implement Interface(TestObject=@n1, Interface=&LastKillerInfoExtensionWidgetInterface))
   > n2 = Cast To FortPlayerControllerZone!(execute, Object=GetOwningPlayer()) -> (then, CastFailed, AsFort Player Controller Zone:Fort Player Controller Zone Object Reference) {is=title K2Node_DynamicCast}
   > SetDeathReport(
    self=@n1
    InDeathReport=Get PlayerDeathReport!(self:Fort Player Controller Zone Object Reference=@n2) -> (PlayerDeathReport:Fort Player Death Report Structure)
   )
  ]}
]
variables [
 LastKillerNameAdditionSlotTag {type=GameplayTag, default {tagName=DynamicUI.HUD.Fort.Athena.LastKillerNameAddition.Slot}}
 LastKillerStatusSlotTag {type=GameplayTag, default {tagName=DynamicUI.HUD.Fort.Athena.LastKillerStatus.Slot}}
 LastKillerWeaponInfoSlotTag {type=GameplayTag, default {tagName=DynamicUI.HUD.Fort.Athena.LastKillerWeaponInfo.Slot}}
 SlottedWidgetChildren {type="array<Widget>", default []}
 TagsToHideHealthInfoFor {type="array<GameplayTag>", default [{tagName=Granted.IcicleLarva.Equipped}]}
]
widgets {
 info {inheritedWidgetCount=0, parentClass=/Script/FortniteUI.AthenaLastKillerInfoWidget, rootWidgetClass=&Overlay, widgetCount=22}
 widgets [
  Overlay_0 {widgetClassPath=&Overlay, widget=Overlay_0}
  Switcher_EliminationSource {
   widgetClassPath=&CommonVisibilitySwitcher
   bIsVariable=true
   parent=Overlay_0
   widget=Switcher_EliminationSource
   slot {class=OverlaySlot, properties {horizontalAlignment=HAlign_Fill, verticalAlignment=VAlign_Fill}}
  }
  Overlay_KillerInfo {widgetClassPath=&Overlay, parent=Switcher_EliminationSource, widget=Overlay_KillerInfo, slot {class=CommonVisibilitySwitcherSlot}}
  HorizontalBox_0 {
   widgetClassPath=&HorizontalBox
   parent=Overlay_KillerInfo
   widget=HorizontalBox_0
   slot {class=OverlaySlot, properties {horizontalAlignment=HAlign_Center}}
  }
  SizeBox_HealthAndShield {
   widgetClassPath=&SizeBox
   bIsVariable=true
   parent=HorizontalBox_0
   widget=SizeBox_HealthAndShield
   properties {bOverride_WidthOverride=true, widthOverride=450}
   slot {class=HorizontalBoxSlot, properties {horizontalAlignment=HAlign_Right}}
  }
  HorizontalBox_2 {
   widgetClassPath=&HorizontalBox
   parent=SizeBox_HealthAndShield
   widget=HorizontalBox_2
   slot {class=SizeBoxSlot, properties {horizontalAlignment=HAlign_Right, verticalAlignment=VAlign_Bottom}}
  }
  Overlay_1 {
   widgetClassPath=&Overlay
   parent=HorizontalBox_2
   widget=Overlay_1
   properties {renderTransform {angle=0, scale=<1 1>, shear=<0 0>, translation=<-8 -15>}}
   slot {class=HorizontalBoxSlot, properties {horizontalAlignment=HAlign_Right, size {sizeRule=Fill, value=1}}}
  }
  Overlay_Status {
   widgetClassPath=&Overlay
   bIsVariable=true
   parent=Overlay_1
   widget=Overlay_Status
   slot {class=OverlaySlot, properties {horizontalAlignment=HAlign_Fill, verticalAlignment=VAlign_Fill}}
  }
  WBP_KillerHealthAndShield {
   widgetClassPath=&/KillerHealthAndShield
   bIsVariable=true
   parent=Overlay_1
   widget=WBP_KillerHealthAndShield
   properties {visibility=Hidden}
   slot {class=OverlaySlot, properties {verticalAlignment=VAlign_Center}}
  }
  VBox_KillerName {
   widgetClassPath=&VerticalBox
   bIsVariable=true
   parent=HorizontalBox_0
   widget=VBox_KillerName
   properties {visibility=Hidden}
   slot {class=HorizontalBoxSlot, properties {horizontalAlignment=HAlign_Center, verticalAlignment=VAlign_Bottom, padding=<ltrb 35 0 35 0>}}
  }
  Text_EliminatedBy {
   widgetClassPath=&CommonTextBlock
   bIsVariable=true
   parent=VBox_KillerName
   widget=Text_EliminatedBy
   properties {
    justification=Center
    text=Eliminated By
    textTransformPolicy=ToUpper
    font {
     bForceMonospaced=false
     bMaterialIsStencil=false
     fontMaterial=null
     fontObject=&Burbank
     letterSpacing=0
     monospacedWidth=1
     size=33
     skewAmount=0
     typefaceFontName=Burbank-PlayerPower
     outlineSettings {
      bApplyOutlineToDropShadows=true
      bMiteredCorners=false
      bSeparateFillAlpha=true
      outlineMaterial=&M-FontShadow
      outlineSize=3
      outlineColor=<rgba 0 0 0 0.5>
     }
    }
    renderTransform {angle=0, scale=<1 1>, shear=<-12 0>, translation=<0 0>}
    shadowOffset=<0 0>
   }
   slot {class=VerticalBoxSlot, properties {horizontalAlignment=HAlign_Center, padding=<ltrb 0 0 0 -3>}}
  }
  SizeBox_KillerName {
   widgetClassPath=&SizeBox
   parent=VBox_KillerName
   widget=SizeBox_KillerName
   properties {bOverride_MaxDesiredWidth=true, maxDesiredWidth=515}
   slot {class=VerticalBoxSlot, properties {horizontalAlignment=HAlign_Center, verticalAlignment=VAlign_Center, padding=<ltrb 0 -18 0 -15>}}
  }
  ScaleBox_1 {
   widgetClassPath=&ScaleBox
   parent=SizeBox_KillerName
   widget=ScaleBox_1
   properties {stretchDirection=DownOnly}
   slot {class=SizeBoxSlot, properties {verticalAlignment=VAlign_Center}}
  }
  HBox_KillerName {widgetClassPath=&HorizontalBox, bIsVariable=true, parent=ScaleBox_1, widget=HBox_KillerName, slot {class=ScaleBoxSlot}}
  Text_KillerName {
   widgetClassPath=&CommonTextBlock
   bIsVariable=true
   parent=HBox_KillerName
   widget=Text_KillerName
   properties {
    justification=Center
    text=player name
    textTransformPolicy=ToUpper
    font {
     bForceMonospaced=false
     bMaterialIsStencil=false
     fontMaterial=null
     fontObject=&Burbank
     letterSpacing=0
     monospacedWidth=1
     size=62
     skewAmount=0
     typefaceFontName=Burbank-Black
     outlineSettings {
      bApplyOutlineToDropShadows=true
      bMiteredCorners=false
      bSeparateFillAlpha=true
      outlineMaterial=&M-FontShadow
      outlineSize=3
      outlineColor=<rgba 0 0 0 0.5>
     }
    }
    shadowOffset=<0 0>
   }
   slot {class=HorizontalBoxSlot}
  }
  SizeBox_1 {
   widgetClassPath=&SizeBox
   parent=HorizontalBox_0
   widget=SizeBox_1
   properties {widthOverride=450}
   slot {class=HorizontalBoxSlot, properties {horizontalAlignment=HAlign_Left, verticalAlignment=VAlign_Bottom}}
  }
  HorizontalBox_1 {widgetClassPath=&HorizontalBox, parent=SizeBox_1, widget=HorizontalBox_1, slot {class=SizeBoxSlot}}
  WBP_KillerKillsWidget {
   widgetClassPath=&/KillerKillsWidget
   bIsVariable=true
   parent=HorizontalBox_1
   widget=WBP_KillerKillsWidget
   properties {visibility=Hidden}
   slot {class=HorizontalBoxSlot, properties {verticalAlignment=VAlign_Bottom}}
  }
  VBox_KillerWeaponHolder {
   widgetClassPath=&VerticalBox
   bIsVariable=true
   parent=HorizontalBox_1
   widget=VBox_KillerWeaponHolder
   slot {class=HorizontalBoxSlot, properties {horizontalAlignment=HAlign_Left, verticalAlignment=VAlign_Bottom, padding=<ltrb 60 0 0 0>}}
  }
  WBP_KillerInfoWeaponWidget {
   widgetClassPath=&/KillerInfoWeaponWidget
   bIsVariable=true
   parent=VBox_KillerWeaponHolder
   widget=WBP_KillerInfoWeaponWidget
   properties {visibility=Hidden}
   slot {class=VerticalBoxSlot}
  }
  Text_UserID {
   widgetClassPath=&TextBlock
   bIsVariable=true
   parent=Overlay_KillerInfo
   widget=Text_UserID
   properties {
    justification=Center
    text=USER ID
    colorAndOpacity {colorUseRule=UseColor_Specified, specifiedColor=<rgba 1 1 1 0.4>}
    font {
     bForceMonospaced=false
     bMaterialIsStencil=false
     fontMaterial=null
     fontObject=&NotoSans
     letterSpacing=0
     monospacedWidth=1
     size=10
     skewAmount=0
     typefaceFontName=NotoSans-Regular
     outlineSettings {
      bApplyOutlineToDropShadows=false
      bMiteredCorners=false
      bSeparateFillAlpha=false
      outlineMaterial=null
      outlineSize=0
      outlineColor=<rgba 0 0 0 1>
     }
    }
    renderTransform {angle=0, scale=<1 1>, shear=<0 0>, translation=<0 21>}
   }
   slot {class=OverlaySlot, properties {horizontalAlignment=HAlign_Center, verticalAlignment=VAlign_Bottom}}
  }
  Text_EliminatedSelf {
   widgetClassPath=&CommonTextBlock
   bIsVariable=true
   parent=Switcher_EliminationSource
   widget=Text_EliminatedSelf
   properties {
    justification=Center
    text=YOU ELIMINATED YOURSELF
    textTransformPolicy=ToUpper
    font {
     bForceMonospaced=false
     bMaterialIsStencil=false
     fontMaterial=null
     fontObject=&Burbank
     letterSpacing=0
     monospacedWidth=1
     size=33
     skewAmount=0
     typefaceFontName=Burbank-PlayerPower
     outlineSettings {
      bApplyOutlineToDropShadows=true
      bMiteredCorners=false
      bSeparateFillAlpha=true
      outlineMaterial=&M-FontShadow
      outlineSize=3
      outlineColor=<rgba 0 0 0 0.5>
     }
    }
    renderTransform {angle=0, scale=<1 1>, shear=<-12 0>, translation=<0 0>}
    shadowOffset=<0 0>
   }
   slot {class=CommonVisibilitySwitcherSlot, properties {horizontalAlignment=HAlign_Center}}
  }
 ]
}
