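# End-of-game elimination screen widget, built on FortEliminationScreen, showing killer info above Report Player and Continue buttons bound to input actions.
# PlayIntroAnimation runs Intro, which fades and scales in the killer panel, then reveals the initially collapsed Continue button.
# On activation, Continue is disabled until a configurable delay elapses, then re-enabled.
# SetContinueButtonText casts Continue to the UIKit CTA button base to set its label.
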WidgetBlueprint /Game/Athena/HUD/EndGame/FortEliminationScreen
animations [Intro {
  binding_count=2
  length_seconds=1.4766
  object=&self:Intro
  sequence {bindings [
    Widget_KillerInfo {
     id=#1
     possessed_class=KillerInfoWidget_C
     tracks [
      {
       type=MovieSceneFloatTrack
       display_name=Render Opacity
       sections [{type=MovieSceneFloatSection, channels [None {type=MovieSceneScriptingFloatChannel, keys [[0,0],[15.2343,0],[29.5313,1]]}]}]
      }
      {
       type=MovieScene2DTransformTrack
       sections [{
         type=MovieScene2DTransformSection
         channels [
          Scale.X {type=MovieSceneScriptingFloatChannel, keys [[15.2343,0.9],[29.5313,1]]}
          Scale.Y {type=MovieSceneScriptingFloatChannel, keys [[15.2343,0.9],[29.5313,1]]}
         ]
        }]
      }
     ]
    }
    Button_Continue {
     id=#2
     possessed_class=InputReflectorButton_C
     tracks [{
       type=MovieSceneByteTrack
       display_name=Visibility
       sections [{type=MovieSceneByteSection, channels [None {type=MovieSceneScriptingByteChannel, keys [[0,1],[29.5313,0]]}]}]
      }]
    }
   ]}
 }]
graphs [EventGraph {nodes [
   n1 = Event SetContinueButtonText!() -> (OutputDelegate, then, InText:"Text (by ref)") {is=title K2Node_Event}
   > n2 = Cast To WBP_UIKit_ButtonCTA_Base!(execute, Object=Get Button_Continue() -> (Button_Continue:Common Button Base Object Reference)) -> (then, CastFailed, AsWBP UIKit Button CTA Base:WBP UIKit Button CTA Base Object Reference) {is=title K2Node_DynamicCast}
   > SetText!(execute, self:Fort CTAButton Object Reference=@n2, ButtonText:"Text (by ref)"=@n1.InText) -> (then)
   Event PlayIntroAnimation() {is=title K2Node_Event}
   > PlayAnimation(InAnimation=Get Intro())
   Event On Activated()
   > Set bIsEnabled!(execute, bIsEnabled:bool=false, self:Widget Object Reference=Get Button_Continue()) -> (then, Output_Get:bool)
   > Delay(Duration=Get DisableContinueButtonDuration() -> (DisableContinueButtonDuration:double))
   > Set bIsEnabled(bIsEnabled=true, self=Get Button_Continue())
  ]}]
variables [DisableContinueButtonDuration {type=number, default=2}]
widgets {
 info {inheritedWidgetCount=3, parentClass=/Script/FortniteUI.FortEliminationScreen, rootWidgetClass=&Overlay, widgetCount=6}
 widgets [
  Overlay_0 {widgetClassPath=&Overlay, widget=Overlay_0}
  Widget_KillerInfo {
   widgetClassPath=/Game/Athena/HUD/EndGame/KillerScreen/KillerInfoWidget
   bInherited=true
   bIsVariable=true
   parent=Overlay_0
   widget=Widget_KillerInfo
   properties {bAutoActivate=true, renderTransform {angle=0, scale=<1 1>, shear=<0 0>, translation=<0 -60>}, renderTransformPivot=<0.5 1>}
   slot {class=OverlaySlot, properties {horizontalAlignment=HAlign_Fill, verticalAlignment=VAlign_Bottom, padding=<ltrb 0 0 0 32>}}
  }
  SizeBox_0 {
   widgetClassPath=&SizeBox
   parent=Overlay_0
   widget=SizeBox_0
   properties {bOverride_MinDesiredHeight=true, heightOverride=45, minDesiredHeight=48}
   slot {class=OverlaySlot, properties {horizontalAlignment=HAlign_Fill, verticalAlignment=VAlign_Bottom, padding=<ltrb 32 0 32 16>}}
  }
  Overlay_1 {
   widgetClassPath=&Overlay
   parent=SizeBox_0
   widget=Overlay_1
   slot {class=SizeBoxSlot, properties {verticalAlignment=VAlign_Center, padding=<ltrb 10 0 10 0>}}
  }
  Button_ReportPlayer {
   widgetClassPath=/FortUILibrary/UIKit/Buttons/WBP_UIKit_Button_Quiet
   bInherited=true
   bIsVariable=true
   parent=Overlay_1
   widget=Button_ReportPlayer
   properties {text=Report Player, triggeringInputAction {dataTable=&InputActionDataTable, rowName=ReportPlayer}}
   slot {class=OverlaySlot}
  }
  Button_Continue {
   widgetClassPath=&WBP_UIKit_Button_Quiet
   bInherited=true
   bIsVariable=true
   parent=Overlay_1
   widget=Button_Continue
   properties {text=Continue, visibility=Collapsed, triggeringInputAction {dataTable=&InputActionDataTable, rowName=EoMNext}}
   slot {class=OverlaySlot, properties {horizontalAlignment=HAlign_Right}}
  }
 ]
}
