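# PlacementOverlay is the post-game Widget Blueprint that shows Victory Royale, losing placement rank, or tie.
# Show events switch panels and play intro animations, with sound and victory-placement HUD context update on wins.
# Gameplay-tag slot hosts accept extension widgets, styled and animated through BPI_VictoryExtensionSlot; losing content can replace the default panel.
# Placement changes toggle a desktop/mobile XP-screen layout, while a stateful event overrides victory text inside an invalidation box.
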WidgetBlueprint /Game/Athena/HUD/EndGame/PlacementOverlay
animations [
 Anim_SpinningSpikeball {
  binding_count=1
  length_seconds=35.7931
  object=&self:Anim_SpinningSpikeball
  sequence {bindings [Image_SpinningSpikeball {
     id=#1
     possessed_class=Image
     tracks [{
       type=MovieScene2DTransformTrack
       sections [{
         type=MovieScene2DTransformSection
         end_frame=715
         start_frame=0
         channels [
          Translation.X {type=MovieSceneScriptingFloatChannel}
          Translation.Y {type=MovieSceneScriptingFloatChannel}
          Angle {type=MovieSceneScriptingFloatChannel, keys [[0,0],[715.8623,360]]}
          Scale.X {type=MovieSceneScriptingFloatChannel}
          Scale.Y {type=MovieSceneScriptingFloatChannel}
          Shear.X {type=MovieSceneScriptingFloatChannel}
          Shear.Y {type=MovieSceneScriptingFloatChannel}
         ]
        }]
      }]
    }]}
 }
 Anim_PlacementIntro {
  binding_count=2
  length_seconds=0.6167
  object=&self:Anim_PlacementIntro
  sequence {bindings [
    Text_Placement {
     id=#2
     possessed_class=CommonTextBlock
     tracks [
      {
       type=MovieScene2DTransformTrack
       sections [{
         type=MovieScene2DTransformSection
         channels [
          Translation.X {type=MovieSceneScriptingFloatChannel}
          Translation.Y {type=MovieSceneScriptingFloatChannel}
          Angle {type=MovieSceneScriptingFloatChannel}
          Scale.X {type=MovieSceneScriptingFloatChannel, keys [[4,4],[8.1663,0.9751],[12.333,1]]}
          Scale.Y {type=MovieSceneScriptingFloatChannel, keys [[4,4],[8.1663,0.9751],[12.333,1]]}
          Shear.X {type=MovieSceneScriptingFloatChannel}
          Shear.Y {type=MovieSceneScriptingFloatChannel}
         ]
        }]
      }
      {
       type=MovieSceneFloatTrack
       display_name=Render Opacity
       sections [{type=MovieSceneFloatSection, channels [None {type=MovieSceneScriptingFloatChannel, keys [[4,0],[6.7773,1]]}]}]
      }
     ]
    }
    Text_YouPlaced {
     id=#3
     possessed_class=CommonTextBlock
     tracks [
      {
       type=MovieScene2DTransformTrack
       sections [{
         type=MovieScene2DTransformSection
         channels [
          Translation.X {type=MovieSceneScriptingFloatChannel}
          Translation.Y {type=MovieSceneScriptingFloatChannel}
          Angle {type=MovieSceneScriptingFloatChannel}
          Scale.X {type=MovieSceneScriptingFloatChannel, keys [[0,5],[3.4717,0.7023],[4.1663,1]]}
          Scale.Y {type=MovieSceneScriptingFloatChannel, keys [[0,5],[3.4717,0.7023],[4.1663,1]]}
          Shear.X {type=MovieSceneScriptingFloatChannel}
          Shear.Y {type=MovieSceneScriptingFloatChannel}
         ]
        }]
      }
      {
       type=MovieSceneFloatTrack
       display_name=Render Opacity
       sections [{type=MovieSceneFloatSection, channels [None {type=MovieSceneScriptingFloatChannel, keys [[0,0],[2.7773,1]]}]}]
      }
     ]
    }
   ]}
 }
 Anim_XPScreenLayout {
  binding_count=2
  length_seconds=1
  object=&self:Anim_XPScreenLayout
  sequence {bindings [
    Overlay_LayoutContainer {
     id=#4
     possessed_class=Overlay
     tracks [{
       type=MovieScene2DTransformTrack
       sections [{
         type=MovieScene2DTransformSection
         channels [
          Scale.X {type=MovieSceneScriptingFloatChannel, keys [[0,1],[19.3093,0.76],[20,0.78]]}
          Scale.Y {type=MovieSceneScriptingFloatChannel, keys [[0,1],[19.3093,0.76],[20,0.78]]}
         ]
        }]
      }]
    }
    Overlay_Losing {
     id=#5
     display_name=LosingOverlay
     possessed_class=Overlay
     tracks [{
       type=MovieScene2DTransformTrack
       sections [{
         type=MovieScene2DTransformSection
         channels [
          Translation.X {type=MovieSceneScriptingFloatChannel}
          Translation.Y {type=MovieSceneScriptingFloatChannel, keys [[0,0],[19,-60.1699],[20,-63]]}
         ]
        }]
      }]
    }
   ]}
 }
 Anim_XPScreenLayout_Mobile {
  binding_count=2
  length_seconds=1
  object=&self:Anim_XPScreenLayout_Mobile
  sequence {bindings [
    Overlay_Losing {
     id=#6
     possessed_class=Overlay
     tracks [{
       type=MovieScene2DTransformTrack
       sections [{
         type=MovieScene2DTransformSection
         channels [
          Translation.X {type=MovieSceneScriptingFloatChannel}
          Translation.Y {type=MovieSceneScriptingFloatChannel, keys [[0,0],[20,-93.0787]]}
          Scale.X {type=MovieSceneScriptingFloatChannel, keys [[0,1],[20,0.78]]}
          Scale.Y {type=MovieSceneScriptingFloatChannel, keys [[0,1],[20,0.78]]}
         ]
        }]
      }]
    }
    Overlay_Victory {
     id=#7
     possessed_class=Overlay
     tracks [{
       type=MovieScene2DTransformTrack
       sections [{
         type=MovieScene2DTransformSection
         channels [
          Translation.X {type=MovieSceneScriptingFloatChannel}
          Translation.Y {type=MovieSceneScriptingFloatChannel, keys [[0,0],[20,-108.0067]]}
          Scale.X {type=MovieSceneScriptingFloatChannel, keys [[0,1],[20,0.7]]}
          Scale.Y {type=MovieSceneScriptingFloatChannel, keys [[0,1],[20,0.7]]}
         ]
        }]
      }]
    }
   ]}
 }
 Anim_NewItem_FadeOut {
  binding_count=1
  length_seconds=0.25
  object=&self:Anim_NewItem_FadeOut
  sequence {bindings [Overlay_Fade {
     id=#8
     possessed_class=Overlay
     tracks [{
       type=MovieSceneFloatTrack
       display_name=Render Opacity
       sections [{type=MovieSceneFloatSection, end_frame=5, start_frame=0, channels [None {type=MovieSceneScriptingFloatChannel, keys [[0,1],[5,0]]}]}]
      }]
    }]}
 }
 Anim_NewItem_FadeIn {
  binding_count=1
  length_seconds=0.25
  object=&self:Anim_NewItem_FadeIn
  sequence {bindings [Overlay_Fade {
     id=#8
     possessed_class=Overlay
     tracks [{
       type=MovieSceneFloatTrack
       display_name=Render Opacity
       sections [{type=MovieSceneFloatSection, end_frame=5, start_frame=0, channels [None {type=MovieSceneScriptingFloatChannel, keys [[0,0],[5,1]]}]}]
      }]
    }]}
 }
 Anim_VictoryRoyale {
  binding_count=13
  length_seconds=2.8333
  object=&self:Anim_VictoryRoyale
  sequence {
   bindings [
    Image_Lightshaft {
     id=#9
     possessed_class=Image
     tracks [
      {
       type=MovieSceneWidgetMaterialTrack
       display_name=Brush Material
       sections [{
         type=MovieSceneParameterSection
         end_frame=56
         start_frame=0
         channels [
          ShaftMax {type=MovieSceneScriptingFloatChannel, const [1.0018,2]}
          ShaftMin {type=MovieSceneScriptingFloatChannel, keys [[7.3333,0.9713],[10,-2.7486],[14.6667,0.9713]]}
         ]
        }]
      }
      {
       type=MovieSceneFloatTrack
       display_name=Render Opacity
       sections [{
         type=MovieSceneFloatSection
         end_frame=56
         start_frame=0
         channels [None {type=MovieSceneScriptingFloatChannel, keys [[7.3337,0],[8,1],[14.6667,0]]}]
        }]
      }
     ]
    }
    RichTextBlock_BottomLineShadow {
     id=#10
     possessed_class=RichTextBlock
     tracks [{
       type=MovieScene2DTransformTrack
       sections [{
         type=MovieScene2DTransformSection
         end_frame=49
         start_frame=0
         channels [
          Translation.X {type=MovieSceneScriptingFloatChannel}
          Translation.Y {type=MovieSceneScriptingFloatChannel}
          Angle {type=MovieSceneScriptingFloatChannel}
          Scale.X {type=MovieSceneScriptingFloatChannel, keys [[0.0197,0],[3,0.001],[6.1283,1.222],[7.2563,1],[9.3653,1]]}
          Scale.Y {type=MovieSceneScriptingFloatChannel, keys [[0.0197,0],[3,0.001],[6.1283,1.222],[7.2563,1],[9.3653,1]]}
          Shear.X {type=MovieSceneScriptingFloatChannel}
          Shear.Y {type=MovieSceneScriptingFloatChannel}
         ]
        }]
      }]
    }
    RichTextBlock_BottomLine {
     id=#11
     possessed_class=RichTextBlock
     tracks [
      {
       type=MovieSceneWidgetMaterialTrack
       display_name=Font Material
       sections [{
         type=MovieSceneParameterSection
         end_frame=53
         start_frame=0
         channels [ShineOffset {type=MovieSceneScriptingFloatChannel, keys [[29.9397,0.6],[43.6063,-0.7262]]}]
        }]
      }
      {
       type=MovieScene2DTransformTrack
       sections [{
         type=MovieScene2DTransformSection
         end_frame=37
         start_frame=0
         channels [
          Translation.X {type=MovieSceneScriptingFloatChannel}
          Translation.Y {type=MovieSceneScriptingFloatChannel}
          Angle {type=MovieSceneScriptingFloatChannel}
          Scale.X {type=MovieSceneScriptingFloatChannel, keys [[0.0197,0],[3,0.001],[6.1283,1.222],[7.2563,1],[9.3653,1]]}
          Scale.Y {type=MovieSceneScriptingFloatChannel, keys [[0.0197,0],[3,0.001],[6.1283,1.222],[7.2563,1],[9.3653,1]]}
          Shear.X {type=MovieSceneScriptingFloatChannel}
          Shear.Y {type=MovieSceneScriptingFloatChannel}
         ]
        }]
      }
     ]
    }
    ImageBackplate {
     id=#12
     possessed_class=Image
     tracks [{
       type=MovieSceneWidgetMaterialTrack
       display_name=Brush Material
       sections [{
         type=MovieSceneParameterSection
         end_frame=51
         start_frame=0
         channels [
          RevealDensity {type=MovieSceneScriptingFloatChannel, keys [[0.02,0.0632],[6.824,1.3382],[7.147,10]]}
          RevealRadius {type=MovieSceneScriptingFloatChannel, keys [[0.02,0],[6.824,1.3973],[7.147,3.2234]]}
          ShineOffset {type=MovieSceneScriptingFloatChannel, keys [[13.2353,0.6576],[28.4263,-1.0328]]}
          PatternRep {type=MovieSceneScriptingFloatChannel, const [0.9,1]}
         ]
        }]
      }]
    }
    Border_VictoryPlacementNumber {
     id=#13
     possessed_class=Border
     tracks [
      {
       type=MovieSceneFloatTrack
       display_name=Render Opacity
       sections [{
         type=MovieSceneFloatSection
         end_frame=52
         start_frame=0
         channels [None {type=MovieSceneScriptingFloatChannel, keys [[8.995,0],[11.0607,1]]}]
        }]
      }
      {
       type=MovieScene2DTransformTrack
       sections [{
         type=MovieScene2DTransformSection
         end_frame=56
         start_frame=0
         channels [
          Translation.X {type=MovieSceneScriptingFloatChannel}
          Translation.Y {type=MovieSceneScriptingFloatChannel}
          Angle {type=MovieSceneScriptingFloatChannel}
          Scale.X {type=MovieSceneScriptingFloatChannel, keys [[8.995,0.3925],[11.5353,0.9925],[12.8417,1.3915],[13.5083,0.9925],[14.7823,0.9925]]}
          Scale.Y {type=MovieSceneScriptingFloatChannel, keys [[8.995,0.3925],[11.5353,0.9925],[12.8417,1.3915],[13.5083,0.9925],[14.7823,0.9925]]}
          Shear.X {type=MovieSceneScriptingFloatChannel}
          Shear.Y {type=MovieSceneScriptingFloatChannel}
         ]
        }]
      }
     ]
    }
    RichTextBlock_TopLine {
     id=#14
     possessed_class=RichTextBlock
     tracks [
      {
       type=MovieSceneWidgetMaterialTrack
       display_name=Font Material
       sections [{
         type=MovieSceneParameterSection
         end_frame=53
         start_frame=6
         channels [ShineOffset {type=MovieSceneScriptingFloatChannel, keys [[27.5083,0.59],[41.175,-0.7]]}]
        }]
      }
      {
       type=MovieScene2DTransformTrack
       sections [{
         type=MovieScene2DTransformSection
         end_frame=53
         start_frame=0
         channels [
          Translation.X {type=MovieSceneScriptingFloatChannel}
          Translation.Y {type=MovieSceneScriptingFloatChannel}
          Angle {type=MovieSceneScriptingFloatChannel}
          Scale.X {type=MovieSceneScriptingFloatChannel, keys [[0.0197,0],[3,0.001],[6.1283,1.222],[7.2563,1],[9.3653,1]]}
          Scale.Y {type=MovieSceneScriptingFloatChannel, keys [[0.0197,0],[3,0.001],[6.1283,1.222],[7.2563,1],[9.3653,1]]}
          Shear.X {type=MovieSceneScriptingFloatChannel}
          Shear.Y {type=MovieSceneScriptingFloatChannel}
         ]
        }]
      }
     ]
    }
    Image_Spikeball2 {
     id=#15
     possessed_class=Image
     tracks [{
       type=MovieSceneFloatTrack
       display_name=Render Opacity
       sections [{
         type=MovieSceneFloatSection
         end_frame=55
         start_frame=0
         channels [None {type=MovieSceneScriptingFloatChannel, keys [[23.929,0],[26.6607,0.8097],[32.2627,1],[44.984,0]]}]
        }]
      }]
    }
    RichTextBlock_TopLineShadow {
     id=#16
     possessed_class=RichTextBlock
     tracks [{
       type=MovieScene2DTransformTrack
       sections [{
         type=MovieScene2DTransformSection
         end_frame=50
         start_frame=0
         channels [
          Translation.X {type=MovieSceneScriptingFloatChannel}
          Translation.Y {type=MovieSceneScriptingFloatChannel}
          Angle {type=MovieSceneScriptingFloatChannel}
          Scale.X {type=MovieSceneScriptingFloatChannel, keys [[0.0197,0],[3,0.001],[6.1283,1.222],[7.2563,1],[9.3653,1]]}
          Scale.Y {type=MovieSceneScriptingFloatChannel, keys [[0.0197,0],[3,0.001],[6.1283,1.222],[7.2563,1],[9.3653,1]]}
          Shear.X {type=MovieSceneScriptingFloatChannel}
          Shear.Y {type=MovieSceneScriptingFloatChannel}
         ]
        }]
      }]
    }
    Image_FlareBG {
     id=#17
     possessed_class=Image
     tracks [
      {
       type=MovieSceneFloatTrack
       display_name=Render Opacity
       sections [{
         type=MovieSceneFloatSection
         end_frame=42
         start_frame=0
         channels [None {type=MovieSceneScriptingFloatChannel, keys [[7.497,0],[10.0523,0.8097],[15.6543,1],[20.3557,0]]}]
        }]
      }
      {
       type=MovieScene2DTransformTrack
       sections [{
         type=MovieScene2DTransformSection
         end_frame=38
         start_frame=0
         channels [
          Translation.X {type=MovieSceneScriptingFloatChannel}
          Translation.Y {type=MovieSceneScriptingFloatChannel}
          Angle {type=MovieSceneScriptingFloatChannel}
          Scale.X {type=MovieSceneScriptingFloatChannel, keys [[7.576,1],[10.3473,1.2],[25.7797,1]]}
          Scale.Y {type=MovieSceneScriptingFloatChannel, keys [[7.576,0.14],[10.3473,1.2],[25.7797,1]]}
          Shear.X {type=MovieSceneScriptingFloatChannel}
          Shear.Y {type=MovieSceneScriptingFloatChannel}
         ]
        }]
      }
     ]
    }
    ImageBackplateShadow {
     id=#18
     possessed_class=Image
     tracks [{
       type=MovieSceneWidgetMaterialTrack
       display_name=Brush Material
       sections [{
         type=MovieSceneParameterSection
         end_frame=51
         start_frame=0
         channels [
          RevealDensity {type=MovieSceneScriptingFloatChannel, keys [[0.02,0.0632],[6.824,1.3382],[7.147,10]]}
          RevealRadius {type=MovieSceneScriptingFloatChannel, keys [[0.02,0],[6.824,1.3973],[7.147,3.2234]]}
          ShineOffset {type=MovieSceneScriptingFloatChannel}
          PatternRep {type=MovieSceneScriptingFloatChannel, const [0.9,1]}
         ]
        }]
      }]
    }
    Image_Cross {
     id=#19
     possessed_class=Image
     tracks [{
       type=MovieSceneFloatTrack
       display_name=Render Opacity
       sections [{
         type=MovieSceneFloatSection
         end_frame=55
         start_frame=0
         channels [None {type=MovieSceneScriptingFloatChannel, keys [[29.9967,0],[34.5327,1],[47.254,0]]}]
        }]
      }]
    }
    Image_Spikeball1 {
     id=#20
     possessed_class=Image
     tracks [{
       type=MovieSceneFloatTrack
       display_name=Render Opacity
       sections [{
         type=MovieSceneFloatSection
         end_frame=53
         start_frame=0
         channels [None {type=MovieSceneScriptingFloatChannel, keys [[15.7473,0],[18.479,0.8097],[24.081,1],[36.8023,0]]}]
        }]
      }]
    }
    Image_SpinningSpikeball {
     id=#21
     possessed_class=Image
     tracks [{
       type=MovieSceneFloatTrack
       display_name=Render Opacity
       sections [{
         type=MovieSceneFloatSection
         end_frame=50
         start_frame=0
         channels [None {type=MovieSceneScriptingFloatChannel, keys [[29.5753,0],[32.2597,0.8097],[37.8617,1]]}]
        }]
      }]
    }
   ]
   tracks [{
     type=MovieSceneEventTrack
     sections [{
       type=MovieSceneEventTriggerSection
       channels [None {type=MovieSceneScriptingEventChannel, const ["<Struct 'MovieSceneEvent' (0x467a5161e00) {}>", 1]}]
      }]
    }]
  }
 }
]
graphs [
 EventGraph {nodes [
   n1 = EventXPLayout!() -> (OutputDelegate, then, Placement:Name) {is=title K2Node_CustomEvent}
   n2 = Get isXPLayout() -> (isXPLayout:bool)
   n3 = "Equal (Name)"!(A:Name=@n1.Placement, B:Name=EoMSpectateEnd) -> (ReturnValue:bool) {is=title K2Node_CallFunction}
   n4 = "Equal (Name)"(A=@n1.Placement, B=XPScreen)
   > n5 = Branch(Condition=AND Boolean(A=NOT Boolean(A=@n2), B=OR Boolean(A=@n3, B=@n4)))
   n6 = Select(
    Option 0:Widget Animation Object Reference=Get Anim_XPScreenLayout() -> (Anim_XPScreenLayout:Widget Animation Object Reference)
    Option 1:Widget Animation Object Reference=Get Anim_XPScreenLayout_Mobile() -> (Anim_XPScreenLayout_Mobile:Widget Animation Object Reference)
    Index=IsMobileGame()
   ) -> (ReturnValue:Widget Animation Object Reference)
   > PlayAnimation(InAnimation=@n6)
   > Set isXPLayout(isXPLayout:bool=true) -> (Output_Get:bool)
   Branch(execute=@n5.else, Condition=AND Boolean(A="Equal (Name)"(A=@n1.Placement, B=RecoverFromPostGame), B=Get isXPLayout()))
   > PlayAnimation(InAnimation=@n6, PlayMode=Reverse)
   > Set isXPLayout(isXPLayout=false)
   n7 = Show Tie() {is=title K2Node_CustomEvent}
   SetActiveWidget(
    execute=[@n7.then, @n30.then]
    self:Widget Switcher Object Reference=Get Switcher_Placement() -> (Switcher_Placement:Widget Switcher Object Reference)
    Widget=Get Overlay_Tie() -> (Overlay_Tie:Overlay Object Reference)
   )
   > SetVisibility(self=Get Overlay_Root() -> (Overlay_Root:Overlay Object Reference), InVisibility=HitTestInvisible)
   New Item Anim Start() {is=title K2Node_CustomEvent}
   > PlayAnimation(InAnimation=Get Anim_NewItem_FadeOut() -> (Anim_NewItem_FadeOut:Widget Animation Object Reference))
   New Item Anim End() {is=title K2Node_CustomEvent}
   > PlayAnimation(InAnimation=Get Anim_NewItem_FadeIn() -> (Anim_NewItem_FadeIn:Widget Animation Object Reference))
   n8 = Show Win() {is=title K2Node_CustomEvent}
   n9 = Sequence(execute=[@n8.then, @n28.then])
   > SetActiveWidget(self=Get Switcher_Placement(), Widget=Get Overlay_Victory() -> (Overlay_Victory:Overlay Object Reference))
   > SetPostGameWidgetPlacement!(
    execute
    self:Athena HUDContext Object Reference=Get AthenaHUDContext!() -> (ReturnValue:Athena HUDContext Object Reference) {is=title K2Node_GetBlueprintContextEx}
    WidgetPlacement:Name=PlacementVictory
   ) -> (then)
   > SetVisibility(self=Get Overlay_Root(), InVisibility=HitTestInvisible)
   > PlaySound2D(Sound=/Game/Athena/Sounds/Character/LastPlayerKilled_Cue)
   PlayAnimation(execute=@n9.then_1, InAnimation=Get Anim_VictoryRoyale() -> (Anim_VictoryRoyale:Widget Animation Object Reference), PlaybackSpeed=0.5)
   > PlayAnimation(InAnimation=Get Anim_SpinningSpikeball() -> (Anim_SpinningSpikeball:Widget Animation Object Reference), NumLoopsToPlay=0)
   n10 = Show Loss() {is=title K2Node_CustomEvent}
   n11 = Get FortHUDContext!() -> (ReturnValue:Fort HUDContext Object Reference) {is=title K2Node_GetBlueprintContextEx}
   "SetText (Text)"(
    execute=[@n10.then, @n29.then]
    self=Get Text_Placement() -> (Text_Placement:Common Text Object Reference)
    InText=Format Text!(Format="LOCTEXT(\"##{0}\")", 0:int=GetLocalPlayerPlacement!(self:Fort HUDContext Object Reference=@n11) -> (ReturnValue:int=0)) -> (Result)
   )
   n12 = Get WidgetSlottingHostOverlay_LosingOverride() -> (WidgetSlottingHostOverlay_LosingOverride:Overlay Object Reference)
   n13 = HasAnyChildren!(self:Panel Widget Object Reference=@n12) -> (ReturnValue:bool)
   > SetActiveWidget(
    self=Get Switcher_Placement()
    Widget=Select(
     Option 0:Widget Object Reference=Get Overlay_Losing() -> (Overlay_Losing:Overlay Object Reference)
     Option 1:Widget Object Reference=@n12
     Index=@n13
    ) -> (ReturnValue:Widget Object Reference)
   )
   > n14 = Branch(Condition=@n13)
   SetVisibility(execute=[@n15.then, @n14.then], self=Get Overlay_Root(), InVisibility=HitTestInvisible)
   PlayAnimation(execute=@n14.else, InAnimation=Get Anim_PlacementIntro() -> (Anim_PlacementIntro:Widget Animation Object Reference))
   > n15 = PlayLosingExtensionAnimation()
   n16 = Event PreConstruct()
   LocalizationLayoutUpdates(execute=[@n17.then_0, @n16.then])
   Event Construct()
   > BindToAnimationFinished(Animation=Get Anim_XPScreenLayout(), Delegate=Create Event())
   > Set DEVTestLocSpace(DEVTestLocSpace:bool=false) -> (Output_Get:bool)
   > RegisterDynamicUISlotTags!(execute, self, Register:bool=true) -> (then)
   > n17 = Sequence()
   Bind Event to On Post Game Widget Placement Change(execute=@n17.then_1, self:Athena HUDContext Object Reference=Get AthenaHUDContext(), Delegate=Create Event()) {is=title K2Node_AddDelegate}
   > n18 = GetEventRouter!(execute, Scope:Object Class Reference=/Script/FortniteGame.FortGameState, ContextActor:Actor Object Reference=GetOwningPlayer()) -> (then, ReturnValue:Gameplay Event Router Component Object Reference)
   > n19 = StartListeningToStatefulEvent!(
    execute:Exec
    Target:Gameplay Event Router Component Object Reference=@n18
    EventType:Script Struct Object Reference=&/Struct_UI_VictoryTextOverrideBroadcastEvent
    Context:Object Reference
    EventBubblingRule:EEventBubblingRule Enum
   ) -> (
    then:Exec
    AsyncTaskProxy:Async Action Start Listening To Stateful Event Object Reference
    OnEventReceived:Exec
    OnSavedEventStateExists:Exec
    OnEventStateCleared:Exec
    Payload:Struct UI Victory Text Override Broadcast Event Structure
   ) {is=title K2Node_StartListeningToStatefulEventAsyncAction}
   n20 = Break Struct UI Victory Text Override Broadcast Event!(Struct_UI_VictoryTextOverrideBroadcastEvent:"Struct UI Victory Text Override Broadcast Event Structure (by ref)"=@n19.Payload) -> (VictoryRoyaleTopLine_2_#22:Text, VictoryRoyaleBottomLine_4_#23:Text) {is=title K2Node_BreakStruct}
   SetText(
    execute=@n19.OnSavedEventStateExists
    self=[
     Get RichTextBlock_TopLine() -> (RichTextBlock_TopLine:Rich Text Block Object Reference)
     Get RichTextBlock_TopLineShadow() -> (RichTextBlock_TopLineShadow:Rich Text Block Object Reference)
    ]
    InText=@n20
   )
   > SetText(
    self=[
     Get RichTextBlock_BottomLine() -> (RichTextBlock_BottomLine:Rich Text Block Object Reference)
     Get RichTextBlock_BottomLineShadow() -> (RichTextBlock_BottomLineShadow:Rich Text Block Object Reference)
    ]
    InText=@n20.VictoryRoyaleBottomLine_4_#23
   )
   n21 = Get InvalidationBox_VictoryText() -> (InvalidationBox_VictoryText:Invalidation Box Object Reference)
   > SetCanCache!(execute, self:Invalidation Box Object Reference=@n21, CanCache:bool=false) -> (then)
   > DelayUntilNextFrame!(execute) -> (then)
   > SetCanCache(self=@n21, CanCache=true)
   n22 = Event Add Content To Host Widget!() -> (
    OutputDelegate
    then
    WidgetHostTag:"Gameplay Tag Structure (by ref)"
    ContentWidget:Widget Object Reference
    Padding:"Margin Structure (by ref)"
    bFill:bool
   ) {is=title K2Node_Event}
   > n23 = Branch(Condition="Equal (GameplayTag)"(A=@n22.WidgetHostTag, B=Get BelowVictoryPlacement_DynamicSlottingTag() -> (BelowVictoryPlacement_DynamicSlottingTag:Gameplay Tag Structure)))
   > AddChildToOverlay!(
    execute
    self:Overlay Object Reference=Get WidgetSlottingHostOverlay_VictoryPlacementSlot() -> (WidgetSlottingHostOverlay_VictoryPlacementSlot:Overlay Object Reference)
    Content:Widget Object Reference=@n22.ContentWidget
   ) -> (then, ReturnValue:Overlay Slot Object Reference)
   > n24 = Cast To BPI_VictoryExtensionSlot!(execute, Object=@n22.ContentWidget) -> (then, CastFailed, AsBPI Victory Extension Slot:BPI Victory Extension Slot Interface) {is=title K2Node_DynamicCast}
   > OverrideVisualData!(
    execute
    self:BPI Victory Extension Slot Interface=@n24
    ExtensionSlotData:BP Victory Extension Slot Data Type Object Reference=&/DA_VictoryExtensionSlot_Default
   ) -> (then)
   n25 = Branch(
    execute=@n23.else
    Condition="Equal (GameplayTag)"(A=@n22.WidgetHostTag, B=Get BelowLosingPlacement_DynamicSlottingTag() -> (BelowLosingPlacement_DynamicSlottingTag:Gameplay Tag Structure))
   )
   > AddChildToOverlay(
    self=Get WidgetSlottingHostOverlay_LosingPlacementSlot() -> (WidgetSlottingHostOverlay_LosingPlacementSlot:Overlay Object Reference)
    Content=@n22.ContentWidget
   )
   > n26 = Cast To BPI_VictoryExtensionSlot(Object=@n22.ContentWidget)
   > OverrideVisualData(self=@n26, ExtensionSlotData=&DA_VictoryExtensionSlot_Default)
   Branch(
    execute=@n25.else
    Condition="Equal (GameplayTag)"(
     A=@n22.WidgetHostTag
     B=Get LosingPlacementOverride_DynamicSlottingTag() -> (LosingPlacementOverride_DynamicSlottingTag:Gameplay Tag Structure)
    )
   )
   n27 = Get WidgetSlottingHostOverlay_LosingOverride()
   > AddChildToOverlay(self=@n27, Content=@n22.ContentWidget)
   > SetActiveWidget(self=Get Switcher_Placement(), Widget=@n27)
   n28 = Event ShowWinningPlacement() {is=title K2Node_Event}
   n29 = Event ShowLosingPlacement() {is=title K2Node_Event}
   Event Destruct()
   > RegisterDynamicUISlotTags(Register=false)
   n30 = Event ShowTiePlacement() {is=title K2Node_Event}
   XP Layout Anim Finished() {is=title K2Node_CustomEvent}
  ]}
 HandleExtensionWidgetCreated {nodes [
   Handle Extension Widget Created!() -> (then, Widget:User Widget Object Reference, SlotType:EUIExtensionSlot Enum=Invalid) {is=title K2Node_FunctionEntry}
   > n1 = Branch(Condition="Equal (Enum)"(
     A:EUIExtensionSlot Enum=Get SlotType() -> (SlotType:EUIExtensionSlot Enum=Invalid)
     B:EUIExtensionSlot Enum=Get PostGamePlacementOverlayVictory() -> (PostGamePlacementOverlayVictory:EUIExtensionSlot Enum=Invalid)
    ))
   > AddChildToVerticalBox(
    self=Get VBox_VictoryExtension() -> (VBox_VictoryExtension:Vertical Box Object Reference)
    Content=Get Widget() -> (Widget:User Widget Object Reference)
   )
   > n2 = Cast To BPI_VictoryExtensionSlot(Object=Get Widget())
   > OverrideVisualData(self=@n2, ExtensionSlotData=&DA_VictoryExtensionSlot_Default)
   Branch(
    execute=@n1.else
    Condition="Equal (Enum)"(
     A=Get SlotType() -> (SlotType=Invalid)
     B=Get PostGamePlacementOverlayLosing() -> (PostGamePlacementOverlayLosing:EUIExtensionSlot Enum=Invalid)
    )
   )
   > AddChildToVerticalBox(self=Get VBox_LosingExtension() -> (VBox_LosingExtension:Vertical Box Object Reference), Content=Get Widget())
   > n3 = Cast To BPI_VictoryExtensionSlot(Object=Get Widget())
   > OverrideVisualData(self=@n3, ExtensionSlotData=&DA_VictoryExtensionSlot_Default)
  ]}
 LocalizationLayoutUpdates {nodes [
   LocalizationLayoutUpdates() {is=title K2Node_FunctionEntry}
   n1 = SlotAsVerticalBoxSlot(Widget=Get RichTextBlock_BottomLine())
   n2 = Break Margin!(Margin:"Margin Structure (by ref)"=Get Padding!(self:Vertical Box Slot Object Reference=@n1) -> (Padding:Margin Structure)) -> (Left:float=~, Top:float=~, Right:float=~, Bottom:float=~) {is=title K2Node_BreakStruct}
   n3 = GetCurrentLanguage!() -> (ReturnValue:String)
   n4 = Select(
    Option 0:double=0
    Option 1:double=30
    Index=OR Boolean!(
     A=Contains(SearchIn=@n3, Substring=ko-KR)
     B=Contains(SearchIn=@n3, Substring=zh-hant)
     C:bool=Get DEVTestLocSpace() -> (DEVTestLocSpace:bool)
    ) -> (ReturnValue)
   ) -> (ReturnValue:double)
   > SetPadding!(
    execute
    self:Vertical Box Slot Object Reference=[SlotAsVerticalBoxSlot(Widget=Get RichTextBlock_BottomLineShadow()), @n1]
    InPadding:Margin Structure=Make Margin(Left=@n2, Top=float + float(A=@n2.Top, B=@n4) {is=title K2Node_CommutativeAssociativeBinaryOperator}, Right=@n2.Right, Bottom=@n2.Bottom)
   ) -> (then)
  ]}
 PlayLosingExtensionAnimation {nodes [
   PlayLosingExtensionAnimation() {is=title K2Node_FunctionEntry}
   > n1 = Sequence()
   > n2 = For Each Loop(Array:Array of Widget Object References=GetAllChildren(self=Get WidgetSlottingHostOverlay_LosingPlacementSlot())) -> (Array Element:Widget Object Reference)
   > n3 = Cast To BPI_VictoryExtensionSlot(Object=@n2)
   > PlayIntroAnimation(self:BPI Victory Extension Slot Interface=@n3)
   n4 = For Each Loop(Exec=@n1.then_1, Array=GetAllChildren(self=Get VBox_LosingExtension()))
   > n5 = Cast To BPI_VictoryExtensionSlot(Object=@n4)
   > PlayIntroAnimation(self=@n5)
  ]}
 PlayVictoryExtensionAnimation {nodes [
   PlayVictoryExtensionAnimation() {is=title K2Node_FunctionEntry}
   > n1 = Sequence()
   > n2 = For Each Loop(Array=GetAllChildren(self=Get WidgetSlottingHostOverlay_VictoryPlacementSlot()))
   > n3 = Cast To BPI_VictoryExtensionSlot(Object=@n2)
   > PlayIntroAnimation(self=@n3)
   n4 = For Each Loop(Exec=@n1.then_1, Array=GetAllChildren(self=Get VBox_VictoryExtension()))
   > n5 = Cast To BPI_VictoryExtensionSlot(Object=@n4)
   > PlayIntroAnimation(self=@n5)
  ]}
 RegisterDynamicUISlotTags {nodes [
   n1 = RegisterDynamicUISlotTags!() -> (then, Register:bool=~) {is=title K2Node_FunctionEntry}
   > n2 = Branch(Condition=@n1)
   n3 = Get BelowVictoryPlacement_DynamicSlottingTag()
   > RegisterWidgetHost!(execute, WidgetHost:Widget Object Reference, WidgetHostTag:"Gameplay Tag Structure (by ref)"=@n3, IsPlayerSpecific:bool=true) -> (then)
   n4 = Get BelowLosingPlacement_DynamicSlottingTag()
   > RegisterWidgetHost(WidgetHostTag=@n4, IsPlayerSpecific=true)
   n5 = Get LosingPlacementOverride_DynamicSlottingTag()
   > RegisterWidgetHost(WidgetHostTag=@n5, IsPlayerSpecific=true)
   UnregisterWidgetHost!(execute=@n2.else, WidgetHost:Widget Object Reference, WidgetHostTag:"Gameplay Tag Structure (by ref)"=@n3) -> (then)
   > UnregisterWidgetHost(WidgetHostTag=@n4)
   > UnregisterWidgetHost(WidgetHostTag=@n5)
   Get WidgetSlottingHostOverlay_LosingOverride()
   Get WidgetSlottingHostOverlay_VictoryPlacementSlot()
   Get WidgetSlottingHostOverlay_LosingPlacementSlot()
  ]}
 RemoveContentFromWidgetHost {nodes [
   n1 = Remove Content From Host Widget!() -> (
    then
    WidgetHostTag:"Gameplay Tag Structure (by ref)"
    ContentWidget:Widget Object Reference
    RemovalType:EWidgetSlottingRemovalType Enum=Unparent
   ) {is=title K2Node_FunctionEntry}
   > n2 = Branch(Condition="Equal (GameplayTag)"(A=@n1, B=Get BelowVictoryPlacement_DynamicSlottingTag()))
   > n3 = RemoveChild(self=Get WidgetSlottingHostOverlay_VictoryPlacementSlot(), Content=@n1.ContentWidget)
   > Return Node!(execute, ReturnValue:bool=@n3)
   n4 = Branch(execute=@n2.else, Condition="Equal (GameplayTag)"(A=@n1, B=Get BelowLosingPlacement_DynamicSlottingTag()))
   > n5 = RemoveChild(self=Get WidgetSlottingHostOverlay_LosingPlacementSlot(), Content=@n1.ContentWidget)
   > Return Node(ReturnValue=@n5)
   Return Node(execute=@n4.else, ReturnValue=false)
  ]}
 Sequencer Events {nodes [
   AnimSequence_PlayVictoryExtensionAnim() {is=title K2Node_CustomEvent}
   > PlayVictoryExtensionAnimation() {is=title K2Node_CallFunction}
  ]}
 TEMP_PreConstructTestBusExfilVictoryText {nodes [
   TEMP_PreConstructTestBusExfilVictoryText() {is=title K2Node_FunctionEntry}
   > SetText(self=[Get RichTextBlock_TopLine(), Get RichTextBlock_TopLineShadow()], InText="LOCTEXT(\"<cap>B</>ATTLE <cap>B</>US\")")
   > SetText(self=[Get RichTextBlock_BottomLine(), Get RichTextBlock_BottomLineShadow()], InText="LOCTEXT(\"<cap>V</>ICTORY <cap>R</>OYALE \")")
  ]}
]
variables [
 AnimationDelay {type=number, default=0.2}
 BelowLosingPlacement_DynamicSlottingTag {type=GameplayTag, default {tagName=DynamicUI.HUD.Fort.Athena.PostMatch.BelowVictoryPlacement}}
 BelowVictoryPlacement_DynamicSlottingTag {type=GameplayTag, default {tagName=DynamicUI.HUD.Fort.Athena.PostMatch.BelowVictoryPlacement}}
 DEVTestLocSpace {type=boolean}
 isXPLayout {type=boolean}
 LosingPlacementOverride_DynamicSlottingTag {type=GameplayTag, default {tagName=DynamicUI.HUD.Fort.Athena.PostMatch.LosingPlacementOverride}}
 PostGamePlacementOverlayLosing {type=EUIExtensionSlot, default=Invalid}
 PostGamePlacementOverlayVictory {type=EUIExtensionSlot, default=Invalid}
]
widgets {
 info {inheritedWidgetCount=1, parentClass=/Script/FortniteUI.PostGamePlacementOverlay, rootWidgetClass=&Overlay, widgetCount=43}
 widgets [
  Overlay_Root {widgetClassPath=&Overlay, bIsVariable=true, widget=Overlay_Root, properties {visibility=Collapsed}}
  Overlay_Fade {
   widgetClassPath=&Overlay
   parent=Overlay_Root
   widget=Overlay_Fade
   slot {class=OverlaySlot, properties {horizontalAlignment=HAlign_Fill, verticalAlignment=VAlign_Fill, padding=<ltrb 0 50 0 0>}}
  }
  Overlay_LayoutContainer {
   widgetClassPath=&Overlay
   parent=Overlay_Fade
   widget=Overlay_LayoutContainer
   properties {visibility=HitTestInvisible, renderTransformPivot=<0.5 -0.2317>}
   slot {class=OverlaySlot, properties {horizontalAlignment=HAlign_Fill, verticalAlignment=VAlign_Fill}}
  }
  Switcher_Placement {
   widgetClassPath=&WidgetSwitcher
   bIsVariable=true
   parent=Overlay_LayoutContainer
   widget=Switcher_Placement
   properties {visibility=HitTestInvisible}
   slot {class=OverlaySlot, properties {horizontalAlignment=HAlign_Fill, verticalAlignment=VAlign_Fill}}
  }
  Overlay_Victory {
   widgetClassPath=&Overlay
   bIsVariable=true
   parent=Switcher_Placement
   widget=Overlay_Victory
   slot {class=WidgetSwitcherSlot, properties {horizontalAlignment=HAlign_Center, verticalAlignment=VAlign_Top, padding=<ltrb 0 85 0 0>}}
  }
  Image_FlareBG {
   widgetClassPath=&Image
   parent=Overlay_Victory
   widget=Image_FlareBG
   properties {
    brush {
     bIsDynamicallyLoaded=false
     drawAs=Image
     imageType=NoImage
     mirroring=NoMirror
     resourceName=None
     resourceObject=/Game/Athena/HUD/Materials/Victory/M_UI_LightFlare
     tiling=NoTile
     imageSize=<1371.1171 946.7446>
     margin=<ltrb 0 0 0 0>
     outlineSettings {
      bUseBrushTransparency=false
      roundingType=HalfHeightRadius
      width=0
      color {colorUseRule=UseColor_Specified, specifiedColor=<rgba 0 0 0 0>}
      cornerRadii=<0 0 0 0>
     }
     tintColor {colorUseRule=UseColor_Specified, specifiedColor=<rgba 1 1 1 1>}
     uVRegion {bIsValid=false, max=<0 0>, min=<0 0>}
    }
    renderTransform {angle=0, scale=<1.192 1.192>, shear=<0 0>, translation=<37 -86>}
   }
   slot {class=OverlaySlot, properties {horizontalAlignment=HAlign_Center, padding=<ltrb 0 -340 0 0>}}
  }
  Overlay_VictoryBody {
   widgetClassPath=&Overlay
   parent=Overlay_Victory
   widget=Overlay_VictoryBody
   properties {renderTransform {angle=0, scale=<0.9 0.9>, shear=<0 0>, translation=<0 0>}}
   slot {class=OverlaySlot, properties {horizontalAlignment=HAlign_Center}}
  }
  Image_Lightshaft {
   widgetClassPath=&Image
   parent=Overlay_VictoryBody
   widget=Image_Lightshaft
   properties {
    renderOpacity=0
    brush {
     bIsDynamicallyLoaded=false
     drawAs=Image
     imageType=NoImage
     mirroring=NoMirror
     resourceName=None
     resourceObject=/Game/Athena/HUD/Materials/Victory/M_Victory_LightShafts_Inst
     tiling=NoTile
     imageSize=<1210.3491 446.3852>
     margin=<ltrb 0 0 0 0>
     outlineSettings {
      bUseBrushTransparency=false
      roundingType=HalfHeightRadius
      width=0
      color {colorUseRule=UseColor_Specified, specifiedColor=<rgba 0 0 0 0>}
      cornerRadii=<0 0 0 0>
     }
     tintColor {colorUseRule=UseColor_Specified, specifiedColor=<rgba 1 1 1 1>}
     uVRegion {bIsValid=false, max=<0 0>, min=<0 0>}
    }
    renderTransform {angle=0, scale=<1 1>, shear=<0 0>, translation=<0 -82>}
   }
   slot {class=OverlaySlot, properties {horizontalAlignment=HAlign_Center, verticalAlignment=VAlign_Center, padding=<ltrb -1000 -1000 -1000 -1000>}}
  }
  Overlay_VictoryBackplate {
   widgetClassPath=&Overlay
   parent=Overlay_VictoryBody
   widget=Overlay_VictoryBackplate
   properties {renderTransform {angle=-4.11, scale=<1 1>, shear=<-36 0>, translation=<-10 -90>}}
   slot {class=OverlaySlot, properties {horizontalAlignment=HAlign_Center}}
  }
  ImageBackplateShadow {
   widgetClassPath=&Image
   parent=Overlay_VictoryBackplate
   widget=ImageBackplateShadow
   properties {
    brush {
     bIsDynamicallyLoaded=false
     drawAs=Image
     imageType=NoImage
     mirroring=NoMirror
     resourceName=None
     resourceObject=/Game/Athena/HUD/Materials/Victory/M_VictoryBG2_Inst
     tiling=NoTile
     imageSize=<740 270>
     margin=<ltrb 0 0 0 0>
     outlineSettings {
      bUseBrushTransparency=false
      roundingType=HalfHeightRadius
      width=0
      color {colorUseRule=UseColor_Specified, specifiedColor=<rgba 0 0 0 0>}
      cornerRadii=<0 0 0 0>
     }
     tintColor {colorUseRule=UseColor_Specified, specifiedColor=<rgba 1 1 1 1>}
     uVRegion {bIsValid=false, max=<0 0>, min=<0 0>}
    }
    colorAndOpacity=<rgba 0 0.0103 0.1301 0.8>
    renderTransform {angle=0, scale=<1 1>, shear=<0 0>, translation=<12 10>}
   }
   slot {class=OverlaySlot}
  }
  ImageBackplate {
   widgetClassPath=&Image
   parent=Overlay_VictoryBackplate
   widget=ImageBackplate
   properties {brush {
     bIsDynamicallyLoaded=false
     drawAs=Image
     imageType=NoImage
     mirroring=NoMirror
     resourceName=None
     resourceObject=&M_VictoryBG2_Inst
     tiling=NoTile
     imageSize=<740 270>
     margin=<ltrb 0 0 0 0>
     outlineSettings {
      bUseBrushTransparency=false
      roundingType=HalfHeightRadius
      width=0
      color {colorUseRule=UseColor_Specified, specifiedColor=<rgba 0 0 0 0>}
      cornerRadii=<0 0 0 0>
     }
     tintColor {colorUseRule=UseColor_Specified, specifiedColor=<rgba 1 1 1 1>}
     uVRegion {bIsValid=false, max=<0 0>, min=<0 0>}
    }}
   slot {class=OverlaySlot}
  }
  InvalidationBox_VictoryText {
   widgetClassPath=/Script/UMG.InvalidationBox
   bIsVariable=true
   parent=Overlay_VictoryBody
   widget=InvalidationBox_VictoryText
   slot {class=OverlaySlot, properties {horizontalAlignment=HAlign_Center, verticalAlignment=VAlign_Center}}
  }
  SizeBox_1 {
   widgetClassPath=&SizeBox
   parent=InvalidationBox_VictoryText
   widget=SizeBox_1
   properties {
    bOverride_WidthOverride=true
    maxDesiredWidth=420
    widthOverride=460
    renderTransform {angle=0, scale=<1 1>, shear=<0 0>, translation=<74 -99>}
   }
   slot {class=PanelSlot}
  }
  ScaleBox_1 {
   widgetClassPath=&ScaleBox
   parent=SizeBox_1
   widget=ScaleBox_1
   properties {stretchDirection=DownOnly}
   slot {class=SizeBoxSlot, properties {horizontalAlignment=HAlign_Left}}
  }
  Overlay_VictoryText {
   widgetClassPath=&Overlay
   parent=ScaleBox_1
   widget=Overlay_VictoryText
   properties {renderTransform {angle=-4.11, scale=<1.05 1>, shear=<0 0>, translation=<0 0>}}
   slot {class=ScaleBoxSlot, properties {horizontalAlignment=HAlign_Left}}
  }
  Border_1 {
   widgetClassPath=&Border
   parent=Overlay_VictoryText
   widget=Border_1
   properties {
    horizontalAlignment=HAlign_Left
    background {
     bIsDynamicallyLoaded=false
     drawAs=NoDrawType
     imageType=NoImage
     mirroring=NoMirror
     resourceName=None
     resourceObject=null
     tiling=NoTile
     imageSize=<32 32>
     margin=<ltrb 0 0 0 0>
     outlineSettings {
      bUseBrushTransparency=false
      roundingType=HalfHeightRadius
      width=0
      color {colorUseRule=UseColor_Specified, specifiedColor=<rgba 0 0 0 0>}
      cornerRadii=<0 0 0 0>
     }
     tintColor {colorUseRule=UseColor_Specified, specifiedColor=<rgba 1 1 1 1>}
     uVRegion {bIsValid=false, max=<0 0>, min=<0 0>}
    }
    contentColorAndOpacity=<rgba 0.0307 0.0307 0.1812 0.9>
    padding=<ltrb 0 0 0 0>
   }
   slot {class=OverlaySlot}
  }
  VerticalBox_2 {
   widgetClassPath=&VerticalBox
   parent=Border_1
   widget=VerticalBox_2
   properties {renderTransform {angle=0, scale=<1 1>, shear=<-4 0>, translation=<0 0>}}
   slot {class=BorderSlot, properties {horizontalAlignment=HAlign_Left, padding=<ltrb 0 0 0 0>}}
  }
  RichTextBlock_TopLineShadow {
   widgetClassPath=/Script/UMG.RichTextBlock
   bIsVariable=true
   parent=VerticalBox_2
   widget=RichTextBlock_TopLineShadow
   properties {text="<cap>V</>ICTORY", textStyleSet=&/VictoryRichTextStyles}
   slot {class=VerticalBoxSlot, properties {horizontalAlignment=HAlign_Left}}
  }
  RichTextBlock_BottomLineShadow {
   widgetClassPath=&RichTextBlock
   bIsVariable=true
   parent=VerticalBox_2
   widget=RichTextBlock_BottomLineShadow
   properties {
    text="<cap>R</>OYALE"
    textStyleSet=&VictoryRichTextStyles
    decoratorClasses [/Script/UMG.RichTextBlockImageDecorator]
   }
   slot {class=VerticalBoxSlot, properties {padding=<ltrb 85 -65 0 0>}}
  }
  VerticalBox_1 {
   widgetClassPath=&VerticalBox
   parent=Overlay_VictoryText
   widget=VerticalBox_1
   properties {renderTransform {angle=0, scale=<1 1>, shear=<-4 0>, translation=<-5 -8>}}
   slot {class=OverlaySlot}
  }
  RichTextBlock_TopLine {
   widgetClassPath=&RichTextBlock
   bIsVariable=true
   parent=VerticalBox_1
   widget=RichTextBlock_TopLine
   properties {text="<cap>V</>ICTORY", textStyleSet=&VictoryRichTextStyles}
   slot {class=VerticalBoxSlot, properties {horizontalAlignment=HAlign_Left}}
  }
  RichTextBlock_BottomLine {
   widgetClassPath=&RichTextBlock
   bIsVariable=true
   parent=VerticalBox_1
   widget=RichTextBlock_BottomLine
   properties {
    text="<cap>R</>OYALE"
    textStyleSet=&VictoryRichTextStyles
    decoratorClasses [&RichTextBlockImageDecorator]
   }
   slot {class=VerticalBoxSlot, properties {horizontalAlignment=HAlign_Left, padding=<ltrb 85 -65 0 0>}}
  }
  Border_VictoryPlacementNumber {
   widgetClassPath=&Border
   parent=Overlay_VictoryBody
   widget=Border_VictoryPlacementNumber
   properties {
    horizontalAlignment=HAlign_Right
    verticalAlignment=VAlign_Center
    background {
     bIsDynamicallyLoaded=false
     drawAs=NoDrawType
     imageType=NoImage
     mirroring=NoMirror
     resourceName=None
     resourceObject=null
     tiling=NoTile
     imageSize=<32 32>
     margin=<ltrb 0 0 0 0>
     outlineSettings {
      bUseBrushTransparency=false
      roundingType=HalfHeightRadius
      width=0
      color {colorUseRule=UseColor_Specified, specifiedColor=<rgba 0 0 0 0>}
      cornerRadii=<0 0 0 0>
     }
     tintColor {colorUseRule=UseColor_Specified, specifiedColor=<rgba 1 1 1 1>}
     uVRegion {bIsValid=false, max=<0 0>, min=<0 0>}
    }
    padding=<ltrb 0 0 0 0>
    renderTransform {angle=0, scale=<0.9925 0.9925>, shear=<0 0>, translation=<-565 -66>}
    renderTransformPivot=<1 0.7>
   }
   slot {class=OverlaySlot, properties {horizontalAlignment=HAlign_Right, verticalAlignment=VAlign_Center, padding=<ltrb 0 -45 0 -45>}}
  }
  Overlay_5 {
   widgetClassPath=&Overlay
   parent=Border_VictoryPlacementNumber
   widget=Overlay_5
   properties {renderTransform {angle=-15, scale=<1 1>, shear=<-10 0>, translation=<0 0>}}
   slot {class=BorderSlot, properties {horizontalAlignment=HAlign_Right, verticalAlignment=VAlign_Center, padding=<ltrb 0 0 0 0>}}
  }
  Border_2 {
   widgetClassPath=&Border
   parent=Overlay_5
   widget=Border_2
   properties {
    background {
     bIsDynamicallyLoaded=false
     drawAs=NoDrawType
     imageType=NoImage
     mirroring=NoMirror
     resourceName=None
     resourceObject=null
     tiling=NoTile
     imageSize=<32 32>
     margin=<ltrb 0 0 0 0>
     outlineSettings {
      bUseBrushTransparency=false
      roundingType=HalfHeightRadius
      width=0
      color {colorUseRule=UseColor_Specified, specifiedColor=<rgba 0 0 0 0>}
      cornerRadii=<0 0 0 0>
     }
     tintColor {colorUseRule=UseColor_Specified, specifiedColor=<rgba 1 1 1 1>}
     uVRegion {bIsValid=false, max=<0 0>, min=<0 0>}
    }
    contentColorAndOpacity=<rgba 0.0123 0.0144 0.0395 1>
    padding=<ltrb 0 0 0 0>
   }
   slot {class=OverlaySlot, properties {horizontalAlignment=HAlign_Right}}
  }
  RichTextBlock_PlacementShadow {
   widgetClassPath=&RichTextBlock
   parent=Border_2
   widget=RichTextBlock_PlacementShadow
   properties {text="<PlacementNumberSymbol>##</><PlacementValue>1</>", textStyleSet=&VictoryRichTextStyles}
   slot {class=BorderSlot, properties {padding=<ltrb 0 0 0 0>}}
  }
  RichTextBlock_Placement {
   widgetClassPath=&RichTextBlock
   parent=Overlay_5
   widget=RichTextBlock_Placement
   properties {
    justification=Right
    text="<PlacementNumberSymbol>##</><PlacementValue>1</>"
    textStyleSet=&VictoryRichTextStyles
    renderTransform {angle=0, scale=<1 1>, shear=<0 0>, translation=<-8 -9>}
   }
   slot {class=OverlaySlot, properties {horizontalAlignment=HAlign_Right}}
  }
  Image_Spikeball2 {
   widgetClassPath=&Image
   parent=Overlay_Victory
   widget=Image_Spikeball2
   properties {
    brush {
     bIsDynamicallyLoaded=false
     drawAs=Image
     imageType=NoImage
     mirroring=NoMirror
     resourceName=None
     resourceObject=/Game/Athena/HUD/Materials/Victory/M_Flare_Spikeball
     tiling=NoTile
     imageSize=<90 90>
     margin=<ltrb 0 0 0 0>
     outlineSettings {
      bUseBrushTransparency=false
      roundingType=HalfHeightRadius
      width=0
      color {colorUseRule=UseColor_Specified, specifiedColor=<rgba 0 0 0 0>}
      cornerRadii=<0 0 0 0>
     }
     tintColor {colorUseRule=UseColor_Specified, specifiedColor=<rgba 1 1 1 1>}
     uVRegion {bIsValid=false, max=<0 0>, min=<0 0>}
    }
    renderTransform {angle=0, scale=<1 1>, shear=<0 0>, translation=<-311 -206>}
   }
   slot {class=OverlaySlot, properties {horizontalAlignment=HAlign_Center, verticalAlignment=VAlign_Center}}
  }
  Image_Cross {
   widgetClassPath=&Image
   parent=Overlay_Victory
   widget=Image_Cross
   properties {
    brush {
     bIsDynamicallyLoaded=false
     drawAs=Image
     imageType=NoImage
     mirroring=NoMirror
     resourceName=None
     resourceObject=/Game/Athena/HUD/Materials/Victory/M_Cross_Flare
     tiling=NoTile
     imageSize=<128 128>
     margin=<ltrb 0 0 0 0>
     outlineSettings {
      bUseBrushTransparency=false
      roundingType=HalfHeightRadius
      width=0
      color {colorUseRule=UseColor_Specified, specifiedColor=<rgba 0 0 0 0>}
      cornerRadii=<0 0 0 0>
     }
     tintColor {colorUseRule=UseColor_Specified, specifiedColor=<rgba 1 1 1 1>}
     uVRegion {bIsValid=false, max=<0 0>, min=<0 0>}
    }
    renderTransform {angle=0, scale=<1 1>, shear=<0 0>, translation=<-214 -122>}
   }
   slot {class=OverlaySlot, properties {horizontalAlignment=HAlign_Center, padding=<ltrb 0 75 0 0>}}
  }
  Image_SpinningSpikeball {
   widgetClassPath=&Image
   parent=Overlay_Victory
   widget=Image_SpinningSpikeball
   properties {
    brush {
     bIsDynamicallyLoaded=false
     drawAs=Image
     imageType=NoImage
     mirroring=NoMirror
     resourceName=None
     resourceObject=&M_Flare_Spikeball
     tiling=NoTile
     imageSize=<256 256>
     margin=<ltrb 0 0 0 0>
     outlineSettings {
      bUseBrushTransparency=false
      roundingType=HalfHeightRadius
      width=0
      color {colorUseRule=UseColor_Specified, specifiedColor=<rgba 0 0 0 0>}
      cornerRadii=<0 0 0 0>
     }
     tintColor {colorUseRule=UseColor_Specified, specifiedColor=<rgba 1 1 1 1>}
     uVRegion {bIsValid=false, max=<0 0>, min=<0 0>}
    }
    renderTransform {angle=0, scale=<1 1>, shear=<0 0>, translation=<-214 -119>}
   }
   slot {class=OverlaySlot, properties {horizontalAlignment=HAlign_Center}}
  }
  Image_Spikeball1 {
   widgetClassPath=&Image
   parent=Overlay_Victory
   widget=Image_Spikeball1
   properties {
    brush {
     bIsDynamicallyLoaded=false
     drawAs=Image
     imageType=NoImage
     mirroring=NoMirror
     resourceName=None
     resourceObject=&M_Flare_Spikeball
     tiling=NoTile
     imageSize=<180 180>
     margin=<ltrb 0 0 0 0>
     outlineSettings {
      bUseBrushTransparency=false
      roundingType=HalfHeightRadius
      width=0
      color {colorUseRule=UseColor_Specified, specifiedColor=<rgba 0 0 0 0>}
      cornerRadii=<0 0 0 0>
     }
     tintColor {colorUseRule=UseColor_Specified, specifiedColor=<rgba 1 1 1 1>}
     uVRegion {bIsValid=false, max=<0 0>, min=<0 0>}
    }
    renderTransform {angle=0, scale=<1 1>, shear=<0 0>, translation=<-219 -189>}
   }
   slot {class=OverlaySlot, properties {horizontalAlignment=HAlign_Center, verticalAlignment=VAlign_Center}}
  }
  VBox_VictoryExtension {
   widgetClassPath=&VerticalBox
   bIsVariable=true
   parent=Overlay_Victory
   widget=VBox_VictoryExtension
   properties {bOverride_Cursor=true, renderTransform {angle=0, scale=<1 1>, shear=<0 0>, translation=<0 160>}}
   slot {class=OverlaySlot, properties {horizontalAlignment=HAlign_Center}}
  }
  WidgetSlottingHostOverlay_VictoryPlacementSlot {
   widgetClassPath=&Overlay
   bIsVariable=true
   parent=Overlay_Victory
   widget=WidgetSlottingHostOverlay_VictoryPlacementSlot
   properties {bOverride_Cursor=true, renderTransform {angle=0, scale=<1 1>, shear=<0 0>, translation=<0 160>}}
   slot {class=OverlaySlot, properties {horizontalAlignment=HAlign_Center}}
  }
  WidgetSlottingHostOverlay_LosingOverride {
   widgetClassPath=&Overlay
   bIsVariable=true
   parent=Switcher_Placement
   widget=WidgetSlottingHostOverlay_LosingOverride
   slot {class=WidgetSwitcherSlot, properties {horizontalAlignment=HAlign_Center, verticalAlignment=VAlign_Top, padding=<ltrb 0 85 0 0>}}
  }
  Overlay_Losing {
   widgetClassPath=&Overlay
   bIsVariable=true
   parent=Switcher_Placement
   widget=Overlay_Losing
   properties {flowDirectionPreference=Culture}
   slot {class=WidgetSwitcherSlot, properties {horizontalAlignment=HAlign_Center, verticalAlignment=VAlign_Top, padding=<ltrb 0 85 0 0>}}
  }
  HorizontalBox_Placement {
   widgetClassPath=&HorizontalBox
   parent=Overlay_Losing
   widget=HorizontalBox_Placement
   properties {renderTransformPivot=<0.5 0>}
   slot {class=OverlaySlot, properties {horizontalAlignment=HAlign_Center}}
  }
  Text_YouPlaced {
   widgetClassPath=&CommonTextBlock
   parent=HorizontalBox_Placement
   widget=Text_YouPlaced
   properties {
    text=You Placed
    textTransformPolicy=ToUpper
    font {
     bForceMonospaced=false
     bMaterialIsStencil=false
     fontMaterial=null
     fontObject=&Burbank
     letterSpacing=0
     monospacedWidth=1
     size=48
     skewAmount=0
     typefaceFontName=Burbank-Black
     outlineSettings {
      bApplyOutlineToDropShadows=false
      bMiteredCorners=false
      bSeparateFillAlpha=false
      outlineMaterial=null
      outlineSize=0
      outlineColor=<rgba 0 0 0 1>
     }
    }
    renderTransform {angle=0, scale=<1 1>, shear=<-12 0>, translation=<0 0>}
    shadowColorAndOpacity=<rgba 0.0052 0.0052 0.0056 0.8>
    shadowOffset=<4 4>
   }
   slot {class=HorizontalBoxSlot, properties {verticalAlignment=VAlign_Center, padding=<ltrb 0 0 10 0>}}
  }
  Text_Placement {
   widgetClassPath=&CommonTextBlock
   bIsVariable=true
   parent=HorizontalBox_Placement
   widget=Text_Placement
   properties {
    text="##99"
    font {
     bForceMonospaced=false
     bMaterialIsStencil=false
     fontMaterial=null
     fontObject=&Burbank
     letterSpacing=0
     monospacedWidth=1
     size=86
     skewAmount=0
     typefaceFontName=Burbank-Black
     outlineSettings {
      bApplyOutlineToDropShadows=false
      bMiteredCorners=false
      bSeparateFillAlpha=false
      outlineMaterial=null
      outlineSize=0
      outlineColor=<rgba 0 0 0 1>
     }
    }
    renderTransform {angle=7, scale=<1 1>, shear=<0 0>, translation=<0 0>}
    shadowColorAndOpacity=<rgba 0.0052 0.0052 0.0056 0.8>
    shadowOffset=<4 4>
   }
   slot {class=HorizontalBoxSlot, properties {verticalAlignment=VAlign_Center}}
  }
  VBox_LosingExtension {
   widgetClassPath=&VerticalBox
   bIsVariable=true
   parent=Overlay_Losing
   widget=VBox_LosingExtension
   properties {bOverride_Cursor=true, renderTransform {angle=0, scale=<1 1>, shear=<0 0>, translation=<0 160>}}
   slot {class=OverlaySlot, properties {horizontalAlignment=HAlign_Center}}
  }
  WidgetSlottingHostOverlay_LosingPlacementSlot {
   widgetClassPath=&Overlay
   bIsVariable=true
   parent=Overlay_Losing
   widget=WidgetSlottingHostOverlay_LosingPlacementSlot
   properties {bOverride_Cursor=true, renderTransform {angle=0, scale=<1 1>, shear=<0 0>, translation=<0 160>}}
   slot {class=OverlaySlot, properties {horizontalAlignment=HAlign_Center}}
  }
  Overlay_Tie {
   widgetClassPath=&Overlay
   bIsVariable=true
   parent=Switcher_Placement
   widget=Overlay_Tie
   slot {class=WidgetSwitcherSlot, properties {horizontalAlignment=HAlign_Center, verticalAlignment=VAlign_Top, padding=<ltrb 0 85 0 0>}}
  }
  Text_Draw {
   widgetClassPath=&CommonTextBlock
   parent=Overlay_Tie
   widget=Text_Draw
   properties {
    text="DRAW!"
    textTransformPolicy=ToUpper
    font {
     bForceMonospaced=false
     bMaterialIsStencil=false
     fontMaterial=null
     fontObject=&Burbank
     letterSpacing=0
     monospacedWidth=1
     size=120
     skewAmount=0
     typefaceFontName=Burbank-Black
     outlineSettings {
      bApplyOutlineToDropShadows=false
      bMiteredCorners=false
      bSeparateFillAlpha=false
      outlineMaterial=null
      outlineSize=0
      outlineColor=<rgba 0 0 0 1>
     }
    }
    renderTransform {angle=0, scale=<1 1>, shear=<-12 0>, translation=<0 0>}
    shadowColorAndOpacity=<rgba 0.0052 0.0052 0.0056 0.8>
    shadowOffset=<4 4>
   }
   slot {class=OverlaySlot, properties {horizontalAlignment=HAlign_Fill, verticalAlignment=VAlign_Fill}}
  }
  TutorialHighlight {widgetClassPath=&TutorialHighlight}
 ]
}
