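# UI material for gold victory text on the Athena HUD, with a moving shine highlight.
# A gradient texture is sampled through offset, center-scaled, rotated UVs, then lerped toward white by WhiteOut.
# MF_Shine's result, multiplied by only the alpha of ShineColor, is color-dodge blended over that base.
# ContinuousShine blends the shine position between a manual ShineOffset and a looping time-driven sweep.
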Material /Game/Athena/HUD/Materials/Victory/M_WinText_Gold
graph {
 properties {materialDomain=MD_UI, pixelDepthOffsetMode=PDOM_Legacy, usedShadingModels=Default Lit}
 nodes [
  n1 = Multiply(A=Time(), B=ScalarParameter() {defaultValue=1, parameterName=ContinuousSpeed})
  n2 = Constant() {r=1}
  n3 = LinearInterpolate(
   A=ScalarParameter() {defaultValue=-1, group=Shine, parameterName=ShineOffset, sliderMax=2, sliderMin=-2}
   B=LinearInterpolate(Alpha=Fmod(A=@n1, B=@n2)) {constA=-1}
   Alpha=ScalarParameter() {parameterName=ContinuousShine}
  )
  n4 = ScalarParameter() {defaultValue=0.0476, group=Shine, parameterName=ShineRotation, sliderMax=1}
  n5 = TextureCoordinate()
  n6 = AppendVector(A=ScalarParameter() {parameterName=OffsetX}, B=ScalarParameter() {defaultValue=0.0667, parameterName=OffsetY})
  n7 = ScalarParameter() {defaultValue=1, parameterName=ScaleX}
  n8 = ScalarParameter() {defaultValue=0.6, parameterName=ScaleY}
  n9 = MaterialFunctionCall(
   UVs=MaterialFunctionCall(UVs=Add(A=@n5, B=@n6), Texture Scale=AppendVector(A=@n7, B=@n8)) {materialFunction=&ScaleUVsByCenter}.UVs
   "Rotation Angle (0-1)"=ScalarParameter() {defaultValue=-0.0415, parameterName=Rotation}
  ) {materialFunction=&CustomRotator}
  MaterialFunctionCall!(
   Blend=Multiply(
    A=VectorParameter() {group=Shine, parameterName=ShineColor, defaultValue=<rgba 1 1 1 1>}.A
    B=MaterialFunctionCall!(ShineOffset=@n3, Rotation=@n4, RotationCenter, UVs) {materialFunction=&/MF_Shine}.Result
   )
   Base=LinearInterpolate(
    A=TextureSampleParameter2D(UVs=@n9.Rotated Values) {parameterName=Texture, samplerSource=SSM_Clamp_WorldGroupSettings, texture=&/GoldGradient}.RGB
    B=Constant3Vector() {constant=<rgba 1 1 1 0>}
    Alpha=ScalarParameter() {parameterName=WhiteOut}
   )
  ) {materialFunction=/Engine/Functions/Engine_MaterialFunctions03/Blends/Blend_ColorDodge}
 ]
}
